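# M_Proto_ColorSaimple is a simple prototype color material in the FX folder of the SilverBellWind female character cosmetic.
# It is unlit and flagged for use on Niagara mesh particles.
# The graph holds only a Particle Color node and a vector parameter defaulting to white, with no connections shown between them or to material outputs.
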Material /BRCosmetics/Characters/Player/Female/Medium/Bodies/F_MED_SilverBellWind/FX/M_Proto_ColorSaimple
graph {
 properties {bUsedWithNiagaraMeshParticles=true, pixelDepthOffsetMode=PDOM_Legacy, shadingModel=MSM_Unlit, usedShadingModels=Unlit}
 nodes [
  ParticleColor()
  VectorParameter() {defaultValue=<rgba 1 1 1 1>}
 ]
}
