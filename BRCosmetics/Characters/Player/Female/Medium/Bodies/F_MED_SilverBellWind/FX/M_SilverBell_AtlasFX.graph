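# Translucent unlit Niagara sprite material for the SilverBellWind character cosmetic's FX, drawing from 3x3 color and normal atlases.
# The transformed normal drives a reflection-vector lookup into a glass cubemap, faking reflections that multiply the partly desaturated atlas color.
# Particle color tints this, and isGlow blends toward flat particle color before MF_FX_EmissiveAdjustments.
# Atlas alpha is multiplied by particle alpha.
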Material /BRCosmetics/Characters/Player/Female/Medium/Bodies/F_MED_SilverBellWind/FX/M_SilverBell_AtlasFX
graph {
 properties {bUsedWithNiagaraSprites=true, blendMode=BLEND_Translucent, pixelDepthOffsetMode=PDOM_Legacy, shadingModel=MSM_Unlit, usedShadingModels=Unlit}
 nodes [
  n1 = TextureSample() {
   samplerType=SAMPLERTYPE_Normal
   texture=/BRCosmetics/Characters/Player/Female/Medium/Bodies/F_MED_SilverBellWind/FX/Textures/T_3x3_SilverBell_FlipbookTest_N
  }
  n2 = Transform(None=@n1.RGB)
  n3 = TextureSample() {texture=/BRCosmetics/Characters/Player/Female/Medium/Bodies/F_MED_SilverBellWind/FX/Textures/T_3x3_SilverBell_Color}
  n4 = Multiply(
   A=TextureSample(UVs=ReflectionVectorWS(CustomWorldNormal=@n2)) {samplerType=SAMPLERTYPE_LinearColor, texture=/Game/Creative/Environments/Meshes/Glass/Textures/T_Glass_Cubemap_01_Desaturated}.RGB
   B=Desaturation(None=@n3.RGB, Fraction=ScalarParameter() {defaultValue=0.2, parameterName=Fraction})
  )
  n5 = ParticleColor()
  MaterialFunctionCall(Emissive=LinearInterpolate(A=Multiply(A=@n4, B=@n5.RGB), B=@n5.RGB, Alpha=ScalarParameter() {parameterName=isGlow})) {materialFunction=&MF_FX_EmissiveAdjustments}
  Multiply(A=@n3.A, B=ParticleColor().A) {constA=1}
 ]
}
