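# Niagara cosmetic FX for the SilverBellWind female character body, attaching glowing hair meshes and sprites and animated emoji particles to the skeletal mesh.
# Emojis burst by probability, travel along one of three splines by partition, and emit death events.
# Hair pieces follow sockets, passing smoothed position to materials, with time-of-day tinting and AlphaFade.
# BodyPart gates spawning; body_spline is disabled, and Emoji_AfterImage's own spawn rate is off.
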NiagaraSystem /BRCosmetics/Characters/Player/Female/Medium/Bodies/F_MED_SilverBellWind/FX/NS_SilverBellWind_Character_FX
dependencies {
 usedDataInterfaces [
  /Script/Niagara.NiagaraDataInterfaceSimpleCounter
  &NiagaraDataInterfaceMeshRendererInfo
  &NiagaraDataInterfaceSpriteRendererInfo
  &NiagaraDataInterfaceCurve
  &NiagaraDataInterfaceArrayFloat
  &NiagaraDataInterfaceArrayFloat3
  /Script/Niagara.NiagaraDataInterfaceSpline
 ]
 usedDynamicInputs [&SimulationPosition]
 usedModules [
  &self:SystemSpawnScript.SystemScriptSource.SystemScriptGraph.NiagaraNodeAssignment_2.SetVariables_#1
  &self:SystemSpawnScript.SystemScriptSource.SystemScriptGraph.NiagaraNodeAssignment_3.SetVariables_#2
  &self:SystemSpawnScript.SystemScriptSource.SystemScriptGraph.NiagaraNodeAssignment_4.SetVariables_#3
  &self:SystemSpawnScript.SystemScriptSource.SystemScriptGraph.NiagaraNodeAssignment_0.SetVariables_#4
  &SystemState
  &FN_VariantManager
  &self:SystemSpawnScript.SystemScriptSource.SystemScriptGraph.NiagaraNodeAssignment_1.SetVariables_#5
 ]
}
emitters {
 Emoji_AfterImage {
  emitterUpdateScript [
   &EmitterState()
   &SpawnRate(SpawnRate=10, Use Spawn Probability=false) {enabled=false}
  ]
  particleSpawnScript [&InitializeParticle(
    Color=&MultiplyLinearColor(
     Linear Color=&SelectLinearColorFromArray(Array Sampling Mode=::NewEnumerator0, Direct Array Index=?, Color Selection Array=$User.Color Array, Random Seed=0)
     Scale Factor=&MakeFloat4FromVectorAndFloat("Vector (RGB)"=&VectorFromFloat(Value=1), "Float (Alpha)"=1)
    )
    Lifetime=&FN_FrontEndFloatModifier(Gameplay=0.15, Front End=0.2)
    Mass Mode=::NewEnumerator1
    Sprite Size Mode=::NewEnumerator3
    Uniform Sprite Size=100
    UsePositionOffset=false
    Write Material Random=true
   ) {module=InitializeParticle001}]
  particleUpdateScript [
   &ParticleState()
   &Drag(Drag=&RandomRangeFloat(Minimum=2, Maximum=4), Use Linear Drag=true, Use Rotational Drag=false)
   &GravityForce(Gravity=&Multiply_VectorByFloat(
     Vector=&RandomRangeVector(Minimum=<0 0 -100>, Maximum=<0 0 -200>)
     Float=&FN_FrontEndFloatModifier(Gameplay=1, Front End=$System.FrontendScale)
    )) {enabled=false}
   &AccelerationForce()
   &SolveForcesAndVelocity(Clamp Velocity=false, Limit Acceleration=false)
   &ScaleSpriteSize(Scale Factor=&Vector2DFromFloat(Value=&FloatFromCurve(
      FloatCurve=DataInterface() {propertyValues {Curve {keys [{tangentMode=RCTM_User, time=0, value=0.5}, {tangentMode=RCTM_User, time=1, value=1}]}}}
      Scale Curve=&FN_FrontEndFloatModifier(Gameplay=2.5, Front End=3)
     )))
   &ScaleSpriteSize(
    Scale Factor=&MakeVector2D(X=&Sine(Normalized Angle=$Emitter.Age, Period=0.7, Scale=1, Bias=0), Y=1)
    Scale Sprite Size Mode=::NewEnumerator2
   ) {enabled=false, module=ScaleSpriteSize001}
   &ScaleColor(
    Scale Alpha=&FloatFromCurve(FloatCurve=DataInterface() {propertyValues {Curve {keys [<at 0 1>,<at 1 0>]}}}, Scale Curve=0.2)
    Scale RGB=&VectorFromFloat(Value=&FN_FrontEndFloatModifier(Gameplay=1, Front End=0.25))
    ScaleA=true
    ScaleRGB=false
   )
  ]
  rendererClasses [&NiagaraSpriteRendererProperties]
  renderers [{rendererClass=&NiagaraSpriteRendererProperties}]
 }
 Emojis {
  emitterUpdateScript [
   &EmitterState()
   &SpawnRate(Spawn Probability=&Lerp_Float(A=0.5, B=0, Alpha=$User.BodyPart), SpawnRate=5, Use Spawn Probability=true) {enabled=false}
   &SpawnBurst_Instantaneous(
    Spawn Group=1
    Spawn Probability=&Lerp_Float(A=0.5, B=0, Alpha=$User.BodyPart)
    Use Spawn Probability=true
    UseLoopCountLimit=false
   ) {module=SpawnBurst_Instantaneous001}
   &SpawnBurst_Instantaneous(
    Spawn Group=1
    Spawn Probability=&Lerp_Float(A=1, B=0, Alpha=$User.BodyPart)
    Spawn Time=0.5
    Use Spawn Probability=true
    UseLoopCountLimit=false
   )
  ]
  particleSpawnScript [
   /SetVariables(Particles.VisibilityTag=$Emitter.SpawnGroup) {isSetParameters=true}
   &self:SpawnGroup1_ParticleCounter(Simple Counter=DataInterface() {propertyValues {GpuSyncMode=SyncCpuToGpu, InitialValue=0}}) {module=SpawnGroup1_ParticleCounter}
   &PartitionParticles(Number Of Partitions=3)
   &InitializeParticle(
    Color=&SelectLinearColorFromArray(
     Array Sampling Mode=::NewEnumerator1
     Direct Array Index=$Particles.PartitionParticles.Partition
     Color Selection Array=DataInterface() {propertyValues {ColorData [<rgba 0.9577 0.3385 1 1>, <rgba 1 0.3 0.3749 1>, <rgba 0 0.1452 4 1>]}}
     Random Seed=?
    )
    Lifetime=1.5
    Mesh Scale Mode=::NewEnumerator3
    Mesh Uniform Scale=&UniformAOrBFloat(A=0.15, B=0.3, Distribution Weight=0.1, Evaluation Type=::NewEnumerator0, Recalculate Random Each Loop=true, Override Seed=false)
    Sprite Size=<1 3>
    Sprite Size Mode=::NewEnumerator4
    UsePositionOffset=false
    Write Material Random=true
   )
   &SubUVAnimation(
    Direct Frame Index:NiagaraFloat=&MakeFloatFromInt(Int=$Particles.SpawnGroup1_ParticleCount)
    Output Alternate Renderer Binding=true
    "Sprite Renderer "=DataInterface() {propertyValues {SpriteRenderer=&self:Emojis_0.NiagaraSpriteRendererProperties_3}}
    SubUV Animation Mode:ENiagara_SubUVLookupModeV2=::NewEnumerator5
   )
  ]
  particleUpdateScript [
   &SolveForcesAndVelocity(Clamp Velocity=false, Limit Acceleration=false)
   &ParticleState()
   &SkeletalMeshLocation(
    Bone / Socket Index=&MakeIntFromFloat(Float=&Lerp_Float(
      A=&MakeFloatFromInt(Int=$Particles.PartitionParticles.Partition)
      B=&MakeFloatFromInt(Int=$Particles.SpawnGroup1_ParticleCount)
      Alpha=&MakeFloatFromInt(Int=$Emitter.SpawnGroup)
     ))
    MaskSpawnGroup=false
    Mesh Sampling Type=::NewEnumerator5
    Orientation Sampling=::NewEnumerator1
    Sampled Position Offset Coordinate Space:ENiagaraSkelMeshTransforms=::NewEnumerator2
    Skeletal Mesh=$User.Skeletal Mesh
    Socket Sampling Mode=::NewEnumerator5
   )
   &SkeletalMeshLocation(
    Bone Sampling Mode=::NewEnumerator4
    MaskSpawnGroup=false
    Orientation Sampling=::NewEnumerator1
    Position Sampling=::NewEnumerator2
    Sampled Position Offset Coordinate Space:ENiagaraSkelMeshTransforms=::NewEnumerator2
    Skeletal Mesh=$User.Skeletal Mesh
   ) {module=SkeletalMeshLocation001}
   &self:AlongSpline(
    Position:NiagaraPosition=$Particles.SkeletalMeshLocation.SampledPosition
    Spline:NiagaraDataInterfaceSpline=$System.Spline
    SplineU:NiagaraFloat=$Particles.NormalizedAge
   ) {module=AlongSpline}
   &self:AlongSpline(
    Position:NiagaraPosition=$Particles.SkeletalMeshLocation.SampledPosition
    Spline:NiagaraDataInterfaceSpline=$System.Spline2
    SplineU:NiagaraFloat=$Particles.NormalizedAge
   ) {module=AlongSpline001}
   &self:AlongSpline(
    Position:NiagaraPosition=$Particles.SkeletalMeshLocation.SampledPosition
    Spline:NiagaraDataInterfaceSpline=$System.Spline3
    SplineU:NiagaraFloat=$Particles.NormalizedAge
   ) {module=AlongSpline002}
   /SetVariables(
    Particles.Position=&self:RecompPos(
     Index:NiagaraInt32=$Particles.PartitionParticles.Partition
     Spline01:NiagaraPosition=$Output.AlongSpline.Position
     Spline02:NiagaraPosition=$Output.AlongSpline001.Position
     Spline03:NiagaraPosition=$Output.AlongSpline002.Position
    )
    Particles.Velocity=&self:RecompVelocity(
     int32:NiagaraInt32=$Particles.PartitionParticles.Partition
     Vector:Vector3f=$Output.AlongSpline.Velocity
     Vector001:Vector3f=$Output.AlongSpline001.Velocity
     Vector002:Vector3f=$Output.AlongSpline002.Velocity
    )
   ) {isSetParameters=true}
   &UpdateMeshOrientation(Rotation Rate=&RandomRangeFloat(Minimum=0.7, Maximum=0.4))
   &ScaleMeshSize(Scale Factor=&VectorFromFloat(Value=&FloatFromCurve(FloatCurve=DataInterface() {propertyValues {Curve {keys [
          {interpMode=RCIM_Cubic, tangentMode=RCTM_User, time=0.0008, value=0}
          {arriveTangent=2.5608, interpMode=RCIM_Cubic, leaveTangent=2.5608, tangentMode=RCTM_User, time=0.1892, value=1.0554}
          {interpMode=RCIM_Cubic, tangentMode=RCTM_User, time=0.5, value=1.7906}
          {interpMode=RCIM_Cubic, tangentMode=RCTM_User, time=0.9, value=0}
         ]}}})))
   &FN_Mesh_Attributes(
    Create Color Attribute=true
    Create Scale Attribute=true
    Scale=&Multiply_Vector(A=$Particles.Scale, B=<0.8 2 0.8>)
    Write to Module Namespace=true
   )
   &ScaleColor(
    Scale Alpha=&FloatFromCurve(
     FloatCurve=DataInterface() {propertyValues {Curve {keys [<at 0 0>,<at 0.5 1>,<at 1 0>]}}}
     CurveIndex=&MakeFloatFromVector(Vector=$Particles.Scale, Channel=::NewEnumerator0)
     Scale Curve=2
    )
    ScaleA=true
    ScaleRGB=true
   )
   &FN_Sprite_Attributes(
    Color=&self:Select_Spline_Color(
     index:NiagaraInt32=$Particles.PartitionParticles.Partition
     Linear Color:LinearColor=<rgba 0.3822 0 1 1>
     Linear Color001:LinearColor=<rgba 1 0.2061 0.7099 1>
     Linear Color002:LinearColor=<rgba 0 0.0595 1 1>
     Linear Color003:LinearColor=<rgba 1 1 1 1>
     Linear Color004:LinearColor=<rgba 1 1 1 1>
    )
    Create Color Attribute=true
    Create SpriteSize Attribute=true
    SpriteSize=&Multiply_Vector2D_ByFloat(Float=1, Vector2D=$Particles.SpriteSize)
    Write to Module Namespace=true
   )
   &FN_Sprite_Attributes(
    Color=&self:Select_Spline_Color(
     index:NiagaraInt32=$Particles.PartitionParticles.Partition
     Linear Color:LinearColor=<rgba 0.3822 0 1 1>
     Linear Color001:LinearColor=<rgba 1 0.2061 0.7099 1>
     Linear Color002:LinearColor=<rgba 0 0.0595 1 1>
     Linear Color003:LinearColor=<rgba 1 1 1 1>
     Linear Color004:LinearColor=<rgba 1 1 1 1>
    )
    Create Color Attribute=true
    Create SpriteSize Attribute=true
    SpriteSize=&Multiply_Vector2D_ByFloat(Float=1, Vector2D=$Particles.SpriteSize)
    Write to Module Namespace=true
   ) {module=FN_Sprite_Attributes001}
   &FN_Sprite_Attributes(
    Color=&MultiplyLinearColor(
     Linear Color=&self:Select_Spline_Color(
      index:NiagaraInt32=$Particles.PartitionParticles.Partition
      Linear Color:LinearColor=<rgba 0.3822 0 1 1>
      Linear Color001:LinearColor=<rgba 1 0.2061 0.7099 1>
      Linear Color002:LinearColor=<rgba 0 0.0595 1 1>
      Linear Color003:LinearColor=<rgba 1 1 1 1>
      Linear Color004:LinearColor=<rgba 1 1 1 1>
     )
     Scale Factor=&MakeVector4(
      X=1
      Y=1
      Z=1
      W=&FloatFromCurve(FloatCurve=DataInterface() {propertyValues {Curve {keys [<at 0 1>,<at 1 0>]}}})
     )
    )
    Create Color Attribute=true
    Create SpriteSize Attribute=true
    SpriteSize=&Vector2DFromFloat(Value=4)
    Write to Module Namespace=true
   ) {module=FN_Sprite_Attributes002}
   &FN_Sprite_Attributes(
    Camera Offset=4
    Color=&LerpLinearColors(
     StartColor=&MultiplyLinearColor(
      Linear Color=<rgba 3 3 3 1>
      Scale Factor=&MakeVector4_FromLinearColor(LinearColor=&FNTimeOfDayColor(
        Night Color=<rgba 0.1169 0.2197 0.5 1>
        Morning Color=<rgba 0.7 0.3444 0.5112 1>
        Day Color=<rgba 0.3385 0.4026 0.65 1>
        Evening Color=<rgba 1 0.6368 0.3801 1>
       ))
     )
     EndColor=<rgba 2 2 2 1>
     LerpFactor=&FN_FrontEndFloatModifier(Gameplay=0, Front End=1)
    )
    Create Camera Offset Attribute=true
    Create Color Attribute=true
    Create Sprite Rotation Attribute=true
    Create SpriteSize Attribute=true
    SpriteSize=&Vector2DFromFloat(Value=&Multiply_Float(
      A=&RandomRangeFloat(Minimum=2, Maximum=2.5)
      B=&FloatFromCurve(
       FloatCurve=DataInterface() {propertyValues {Curve {keys [
           {interpMode=RCIM_Cubic, tangentMode=RCTM_User, time=0.0008, value=0}
           {arriveTangent=2.5608, interpMode=RCIM_Cubic, leaveTangent=2.5608, tangentMode=RCTM_User, time=0.1892, value=1.0554}
           {interpMode=RCIM_Cubic, tangentMode=RCTM_User, time=0.5, value=1.7906}
           {arriveTangent=-5.8674, interpMode=RCIM_Cubic, leaveTangent=-5.8674, tangentMode=RCTM_User, time=1, value=0.5}
          ]}}}
       Scale Curve=&FN_FrontEndFloatModifier(Gameplay=1.5, Front End=2)
      )
     ))
    Write to Module Namespace=true
   ) {module=FN_Sprite_Attributes003}
   &GenerateDeathEvent(
    Event Probability=1
    "Float to Send as Float 1 (NormalizedAge)":NiagaraFloat=$Particles.SubUVAnimation.SubImageIndex
    "Float to Send as Float 2 (RandomNormalizedFloat)":NiagaraFloat=$Particles.SpriteRotation
    "Float to Send as Float 3 (DistanceTraveled)":NiagaraFloat=&MakeFloatFromVector2d(Vector 2D=$Particles.FN_Sprite_Attributes003.SpriteSize)
    Use Event Probability=true
    "Vector to Send as Vector 1 (Position)":NiagaraPosition=$StackContext.Position
   )
  ]
  rendererClasses [&NiagaraMeshRendererProperties, &NiagaraSpriteRendererProperties]
  renderers [
   {rendererClass=&NiagaraMeshRendererProperties, rendererIndex=0}
   {rendererClass=&NiagaraSpriteRendererProperties, rendererIndex=1}
   {rendererClass=&NiagaraSpriteRendererProperties, rendererIndex=2}
   {rendererClass=&NiagaraSpriteRendererProperties, rendererIndex=3}
   {rendererClass=&NiagaraSpriteRendererProperties, rendererIndex=4}
  ]
 }
 Splines_Hair_1 {
  emitterUpdateScript [
   &EmitterState()
   &SpawnBurst_Instantaneous(
    Spawn Count=5
    Spawn Probability=&Lerp_Float(A=1, B=0, Alpha=$User.BodyPart)
    Use Spawn Probability=true
    UseLoopCountLimit=true
   )
  ]
  particleSpawnScript [
   /SetVariables(
    Particles.MeshIndex=&ModuloInt(A=&ReturnExecIndex(), B=3)
    Particles.float:NiagaraFloat=&SelectFloatFromArray(
     Direct Array Index=&ModuloInt(A=&ReturnExecIndex(), B=3)
     Float Selection Array=DataInterface() {propertyValues {FloatData [0,40,10]}}
    )
   ) {isSetParameters=true, module=SetVariables_#6}
   &InitializeParticle(
    Color Mode=::NewEnumerator0
    Lifetime Max=1.5
    Lifetime Mode=::NewEnumerator1
    Mesh Scale=&Multiply_VectorByFloat(Vector=<1 1 1>, Float=1.5)
    Mesh Scale Mode=::NewEnumerator4
    UsePositionOffset=false
    Write Material Random=true
   ) {module=InitializeParticle001}
   &CameraOffset(Camera Offset Amount=1)
  ]
  particleUpdateScript [
   &ParticleState(Kill Particles When Lifetime Has Elapsed=false, Let Infinitely Lived Particles Die When Emitter Deactivates=true)
   &SkeletalMeshLocation(
    Bone / Socket Index=&ReturnExecIndex()
    MaskSpawnGroup=false
    Mesh Sampling Type=::NewEnumerator5
    Sampled Position Offset=&SelectVectorFromArray(
     Direct Array Index=&ReturnExecIndex()
     Vector Selection Array=DataInterface() {propertyValues {FloatData [<0 0 0>, <0 -1.1 0.2>, <0.5 1 0>, <0 0 0>, <0 0 0>]}}
    )
    Sampled Position Offset Coordinate Space:ENiagaraSkelMeshTransforms=::NewEnumerator2
    Skeletal Mesh=$User.Skeletal Mesh
    Socket Sampling Mode=::NewEnumerator5
   )
   &ScaleMeshSize(Scale Factor=&Lerp_Vector(A=<0 0 0>, B=$Particles.SkeletalMeshLocation.SampledScale, Alpha=$System.AlphaFade))
   &InterpolateOverTime(
    Rate Of Change=8
    Target Vector Value:Vector3f=&ConvertPositionToVector(Input Position=$Particles.Position, Convert To Absolute Large World Vector=false)
    Variable Type=::NewEnumerator2
   )
   /SetVariables(Particles.Interpolate Dynamic=&MakeFloat4FromVectorAndFloat(
     "Vector (RGB)"=$StackContext.InterpolateOverTime.InterpolateOverTime_V3_.Moving Average
     "Float (Alpha)"=&Add_Float(A=$User.Interpolate Intensity, B=-0.5)
    )) {isSetParameters=true, module=SetVariables_#7}
   &DynamicMaterialParameters(
    Index 0 Param 1=&RandomRangeFloat(Minimum=0, Maximum=1)
    Index 0 Param 2=&Multiply_Float(
     A=0.3
     B=&SelectFloatFromArray(
      Direct Array Index=&ReturnExecIndex()
      Float Selection Array=DataInterface() {propertyValues {FloatData [5,8,6,4,3]}}
     )
    )
    Module.Param0WriteEnabled=true
    Module.Param1WriteEnabled=true
    Module.Param2WriteEnabled=true
    Module.Param3WriteEnabled=true
   )
   &Color(Color=&MultiplyLinearColor(
     Linear Color=&MultiplyLinearColor(
      Linear Color=&VariantColorSelector(
       Variant Amount=0
       Original Color=&SelectLinearColorFromArray(
        Array Sampling Mode=::NewEnumerator1
        Direct Array Index=&ReturnExecIndex()
        Color Selection Array=DataInterface() {propertyValues {ColorData [<rgba 0.7465 0.1737 2 6>, <rgba 1 0.1539 0.2137 8>, <rgba 0.3979 0.4708 1 8>, <rgba 0.1728 0.1435 0.5 7>, <rgba 0.1083 1 0.0636 4>]}}
        Random Seed=?
       )
       Variant 1 Color=&SelectLinearColorFromArray(
        Array Sampling Mode=::NewEnumerator1
        Direct Array Index=&ReturnExecIndex()
        Color Selection Array=DataInterface() {propertyValues {ColorData [<rgba 0.6593 0 2 6>, <rgba 0 0.0811 2 6>, <rgba 0 0.2809 0.4 8>, <rgba 0.1728 0.1435 0.5 7>, <rgba 0.3026 1 0 4>]}}
        Random Seed=?
       )
       Variant 2 Color=?
      )
      Scale Factor=&MakeVector4_FromLinearColor(LinearColor=&FNTimeOfDayColor(Night Color=<rgba 0.1208 0.2135 1 1>, Morning Color=<rgba 0.3245 0.2331 0.4 1>, Day Color=<rgba 1 1 1 1>, Evening Color=<rgba 1 1 1 1>))
     )
     Scale Factor=&MakeFloat4FromVectorAndFloat("Vector (RGB)"=<1 1 1>, "Float (Alpha)"=$System.AlphaFade)
    ))
   &FN_Mesh_Attributes(
    Color=&MultiplyLinearColor(
     Linear Color=&VariantColorSelector(
      Variant Amount=0
      Morning Color=<rgba 1 1 1 1>
      Day Color=<rgba 1 1 1 1>
      Evening Color=<rgba 1 1 1 1>
      Night Color=<rgba 0 1 0.6785 1>
      Use TOD?=true
      Original Color=&self:Splines_Hair_0.NiagaraScratchPadContainer_1.ScratchDynamicInput_0(
       index:NiagaraInt32=$Particles.MeshIndex
       Linear Color:LinearColor=<rgba 0.3164 0 1 1>
       Linear Color001:LinearColor=<rgba 1 0.1332 0.1332 1>
       Linear Color002:LinearColor=<rgba 0 0.1833 1 1>
       Linear Color003:LinearColor=<rgba 1 1 1 1>
       Linear Color004:LinearColor=<rgba 1 1 1 1>
      )
      Variant 1 Color=&self:Splines_Hair_0.NiagaraScratchPadContainer_1.ScratchDynamicInput_0(
       index:NiagaraInt32=$Particles.MeshIndex
       Linear Color:LinearColor=<rgba 0.3164 0 1 1>
       Linear Color001:LinearColor=<rgba 0 0.1833 1 1>
       Linear Color002:LinearColor=<rgba 0.439 1 0 1>
       Linear Color003:LinearColor=<rgba 1 1 1 1>
       Linear Color004:LinearColor=<rgba 0.1227 1 0.3069 1>
      )
      Variant 2 Color=?
     )
     Scale Factor=&MakeFloat4FromVectorAndFloat("Vector (RGB)"=<2 2 2>, "Float (Alpha)"=&Multiply_Float(A=0.2, B=$System.AlphaFade))
    )
    Create Color Attribute=true
    Write to Module Namespace=true
   )
   &FN_Mesh_Attributes(
    Color=&MultiplyLinearColor(
     Linear Color=&VariantColorSelector(
      Variant Amount=0
      Morning Color=<rgba 1 1 1 0.4>
      Day Color=<rgba 1 1 1 1>
      Evening Color=<rgba 1 0.6303 0.3844 0.4>
      Night Color=<rgba 0.1639 0.174 1 1>
      Use TOD?=true
      Original Color=&self:Splines_Hair_0.NiagaraScratchPadContainer_1.ScratchDynamicInput_0(
       index:NiagaraInt32=$Particles.MeshIndex
       Linear Color:LinearColor=<rgba 0.3822 0 1 0.1>
       Linear Color001:LinearColor=<rgba 1 0.2061 0.7099 0.1>
       Linear Color002:LinearColor=<rgba 0 0.0595 1 0.1>
       Linear Color003:LinearColor=<rgba 1 1 1 1>
       Linear Color004:LinearColor=<rgba 1 1 1 1>
      )
      Variant 1 Color=&self:Splines_Hair_0.NiagaraScratchPadContainer_1.ScratchDynamicInput_0(
       index:NiagaraInt32=$Particles.MeshIndex
       Linear Color:LinearColor=<rgba 0.524 0 2 0.1>
       Linear Color001:LinearColor=<rgba 0.2061 0.2061 1 0.1>
       Linear Color002:LinearColor=<rgba 0 1 0.65 0.1>
       Linear Color003:LinearColor=<rgba 1 1 1 1>
       Linear Color004:LinearColor=<rgba 0.0562 1 0.3343 1>
      )
      Variant 2 Color=?
     )
     Scale Factor=&MakeFloat4FromVectorAndFloat("Vector (RGB)"=<1 1 1>, "Float (Alpha)"=$System.AlphaFade)
    )
    Create Color Attribute=true
    Write to Module Namespace=true
   ) {module=FN_Mesh_Attributes002}
   &FN_Sprite_Attributes(
    Camera Offset=5
    Color=&MultiplyLinearColor(
     Linear Color=&VariantColorSelector(
      Variant Amount=0
      Original Color=&self:ScratchDynamicInput_0(
       index:NiagaraInt32=&ReturnExecIndex()
       Linear Color:LinearColor=<rgba 3.507 0 40 1>
       Linear Color001:LinearColor=<rgba 15 0 0.3808 1>
       Linear Color002:LinearColor=<rgba 0 1.7864 30 1>
       Linear Color003:LinearColor=<rgba 4.5037 2.2153 10 1>
       Linear Color004:LinearColor=<rgba 2.792 5 0 1>
      )
      Variant 1 Color=&self:ScratchDynamicInput_01(
       index:NiagaraInt32=&ReturnExecIndex()
       Linear Color:LinearColor=<rgba 1.9622 0.0253 3 1>
       Linear Color001:LinearColor=<rgba 0.8708 2.8169 15 1>
       Linear Color002:LinearColor=<rgba 0 3 1.1565 1>
       Linear Color003:LinearColor=<rgba 4.5037 2.2153 10 1>
       Linear Color004:LinearColor=<rgba 2.792 5 0 1>
      )
      Variant 2 Color=?
     )
     Scale Factor=&MakeFloat4FromVectorAndFloat("Vector (RGB)"=<1 1 1>, "Float (Alpha)"=$System.AlphaFade)
    )
    Create Camera Offset Attribute=true
    Create Color Attribute=true
    Create SpriteSize Attribute=true
    SpriteSize=&Vector2DFromFloat(Value=&SelectFloatFromArray(
      Direct Array Index=&ReturnExecIndex()
      Float Selection Array=DataInterface() {propertyValues {FloatData [6,6,6,3,4]}}
     ))
    Write to Module Namespace=true
   )
   /SetVariables(
    Particles.Decal_DY0:Vector4f=&MakeVector4(X=0.8, Y=0, Z=1, W=&RandomRangeFloat(Minimum=0, Maximum=1))
    Particles.Decal_DY1:Vector4f=&MakeVector4(X=20, Y=0, Z=0, W=0)
   ) {isSetParameters=true}
  ]
  rendererClasses [&NiagaraMeshRendererProperties, &NiagaraSpriteRendererProperties]
  renderers [
   {rendererClass=&NiagaraMeshRendererProperties, rendererIndex=0}
   {rendererClass=&NiagaraMeshRendererProperties, rendererIndex=1}
   {rendererClass=&NiagaraMeshRendererProperties, rendererIndex=2}
   {rendererClass=&NiagaraSpriteRendererProperties, rendererIndex=3}
  ]
 }
 body_spline {
  bEnabled=false
  emitterUpdateScript [
   &EmitterState()
   &SpawnBurst_Instantaneous(Spawn Probability=&Lerp_Float(A=0, B=1, Alpha=$User.BodyPart), Use Spawn Probability=true, UseLoopCountLimit=true)
  ]
  particleSpawnScript [
   /SetVariables(Particles.MeshIndex=&ReturnExecIndex()) {isSetParameters=true, module=SetVariables_#6}
   &InitializeParticle(
    Color=&MultiplyLinearColor(
     Linear Color=&SelectLinearColorFromArray(
      Array Sampling Mode=::NewEnumerator1
      Direct Array Index=&ReturnExecIndex()
      Color Selection Array=DataInterface() {propertyValues {ColorData [<rgba 0.5 0.3677 0.1542 6>, <rgba 1 0.15 0.2065 8>, <rgba 0.2871 0.2816 1 8>]}}
      Random Seed=?
     )
     Scale Factor=&MakeVector4_FromLinearColor(LinearColor=<rgba 1 1 1 1.5>)
    )
    Lifetime Max=1.5
    Lifetime Mode=::NewEnumerator1
    Mesh Scale=&Multiply_VectorByFloat(Vector=<1 1 1>, Float=1.5)
    Mesh Scale Mode=::NewEnumerator4
    Sprite Size Mode=::NewEnumerator3
    UsePositionOffset=false
    Write Material Random=true
   ) {module=InitializeParticle001}
   &CameraOffset(Camera Offset Amount=2)
   &InitialMeshOrientation(Mesh Orientation Mode=::NewEnumerator7, Rotation=<0.4 0.2 -0.4>, Use Rotation Vector=true) {enabled=false}
   &InitialMeshOrientation(
    Mesh Axis To Orient=<0 0 1>
    Mesh Orientation Mode=::NewEnumerator2
    Orientation Coordinate Space=::World
    Orientation Vector=<0 0 1>
    Rotation=<-0.12 0 0>
    Use Rotation Vector=true
   ) {module=InitialMeshOrientation001}
  ]
  particleUpdateScript [
   &ParticleState(Kill Particles When Lifetime Has Elapsed=false)
   &ScaleMeshSize(Scale Factor=&Lerp_Vector(A=<0 0 0>, B=<1 1 1>, Alpha=$System.AlphaFade))
   &SkeletalMeshLocation(
    MaskSpawnGroup=false
    Mesh Sampling Type=::NewEnumerator5
    Orientation Sampling=::NewEnumerator1
    Sampled Position Offset=<0 0.2 1>
    Sampled Position Offset Coordinate Space:ENiagaraSkelMeshTransforms=::NewEnumerator2
    Skeletal Mesh=$System.Skeletal Mesh
    Socket Sampling Mode=::NewEnumerator5
   )
   &InterpolateOverTime(
    Rate Of Change=8
    Target Vector Value:Vector3f=&ConvertPositionToVector(Input Position=$Particles.Position, Convert To Absolute Large World Vector=false)
    Variable Type=::NewEnumerator2
   )
   /SetVariables(Particles.Interpolate Dynamic=&MakeFloat4FromVectorAndFloat(
     "Vector (RGB)"=$StackContext.InterpolateOverTime.InterpolateOverTime_V3_.Moving Average
     "Float (Alpha)"=&Add_Float(A=$User.Interpolate Intensity, B=-0.5)
    )) {isSetParameters=true, module=SetVariables_#7}
   &DynamicMaterialParameters(
    Index 0 Param 1=&RandomRangeFloat(Minimum=0, Maximum=1)
    Index 0 Param 2=&Multiply_Float(
     A=0.3
     B=&SelectFloatFromArray(Direct Array Index=$Particles.MeshIndex, Float Selection Array=DataInterface() {propertyValues {FloatData [8,8,8]}})
    )
    Module.Param0WriteEnabled=true
    Module.Param1WriteEnabled=true
    Module.Param2WriteEnabled=true
    Module.Param3WriteEnabled=true
   )
   &FN_Mesh_Attributes(
    Color=&MultiplyLinearColor(
     Linear Color=&self:ColorSelector(
      index:NiagaraInt32=$Particles.MeshIndex
      Linear Color:LinearColor=<rgba 4 1.7422 0.0122 1>
      Linear Color001:LinearColor=<rgba 1 0.1332 0.1332 1>
      Linear Color002:LinearColor=<rgba 0 0.1833 1 1>
     )
     Scale Factor=&MakeFloat4FromVectorAndFloat("Vector (RGB)"=<2 2 2>, "Float (Alpha)"=&Multiply_Float(A=0.2, B=$System.AlphaFade))
    )
    Create Color Attribute=true
    Write to Module Namespace=true
   )
   &FN_Mesh_Attributes(
    Color=&MultiplyLinearColor(
     Linear Color=&self:ColorSelector(
      index:NiagaraInt32=$Particles.MeshIndex
      Linear Color:LinearColor=<rgba 0.3822 0 1 0.1>
      Linear Color001:LinearColor=<rgba 1 0.2061 0.7099 0.1>
      Linear Color002:LinearColor=<rgba 0 0.0595 1 0.1>
     )
     Scale Factor=&MakeFloat4FromVectorAndFloat("Vector (RGB)"=<1 1 1>, "Float (Alpha)"=$System.AlphaFade)
    )
    Create Color Attribute=true
    Write to Module Namespace=true
   ) {module=FN_Mesh_Attributes002}
   &FN_Sprite_Attributes(
    Camera Offset=10
    Color=&MultiplyLinearColor(
     Linear Color=<rgba 10 5.9984 2.2601 1>
     Scale Factor=&MakeFloat4FromVectorAndFloat("Vector (RGB)"=<1 1 1>, "Float (Alpha)"=$System.AlphaFade)
    )
    Create Camera Offset Attribute=true
    Create Color Attribute=true
    Create Position Attribute=true
    Create SpriteSize Attribute=true
    Position:NiagaraPosition=&AddVectorToPosition(Position=$Particles.SkeletalMeshLocation.SampledPosition, Vector=<0 0 0>)
    SpriteSize=&Vector2DFromFloat(Value=10)
    Write to Module Namespace=true
   )
   /SetVariables(
    Particles.Decal_DY0:Vector4f=&MakeVector4(X=0.8, Y=0, Z=1, W=&RandomRangeFloat(Minimum=0, Maximum=1))
    Particles.Decal_DY1:Vector4f=&MakeVector4(X=20, Y=0, Z=0, W=0)
   ) {isSetParameters=true}
   &ScaleColor(Scale Alpha=$System.AlphaFade, ScaleA=true, ScaleRGB=true)
  ]
  rendererClasses [&NiagaraMeshRendererProperties, &NiagaraSpriteRendererProperties]
  renderers [
   {rendererClass=&NiagaraMeshRendererProperties, rendererIndex=0}
   {rendererClass=&NiagaraMeshRendererProperties, rendererIndex=1}
   {rendererClass=&NiagaraMeshRendererProperties, rendererIndex=2}
   {rendererClass=&NiagaraSpriteRendererProperties, rendererIndex=3}
  ]
 }
}
summary {
 systemName=NS_SilverBellWind_Character_FX
 emitters [
  {emitterName=Emojis, rendererClasses [&NiagaraMeshRendererProperties, &NiagaraSpriteRendererProperties]}
  {
   bEnabled=false
   emitterName=body_spline
   rendererClasses [&NiagaraMeshRendererProperties, &NiagaraSpriteRendererProperties]
  }
  {emitterName=Splines_Hair_1, rendererClasses [&NiagaraMeshRendererProperties, &NiagaraSpriteRendererProperties]}
  {emitterName=Emoji_AfterImage, rendererClasses [&NiagaraSpriteRendererProperties]}
 ]
 userVariables [
  User.bIsFrontEnd {type=NiagaraBool}
  User.bIsFrontEndPreview {type=NiagaraBool}
  User.BodyPart {type=NiagaraFloat, defaultValue=0, description=""}
  User.Color Array {
   type=NiagaraDataInterfaceArrayColor
   description=""
   defaultValue {
    struct=NiagaraExt_VariableValue_DataInterface
    value {
     dataInterface=&self:NiagaraDataInterfaceArrayColor_0
     dataInterfaceClass=&NiagaraDataInterfaceArrayColor
    }
   }
  }
  User.Interpolate Intensity {type=NiagaraFloat, defaultValue=1, description=""}
  User.Skeletal Mesh {
   type=NiagaraDataInterfaceSkeletalMesh
   defaultValue {
    struct=NiagaraExt_VariableValue_DataInterface
    value {dataInterface=&self:NiagaraDataInterfaceSkeletalMesh_0}
   }
  }
  User.SplineReference {
   type=Object
   description=""
   defaultValue {struct=NiagaraExt_VariableValue_Object, value {object=null, objectClass=&Object}}
  }
  User.SplineReference2 {
   type=Object
   description=""
   defaultValue {struct=NiagaraExt_VariableValue_Object, value {object=null, objectClass=&Object}}
  }
  User.SplineReference3 {
   type=Object
   description=""
   defaultValue {struct=NiagaraExt_VariableValue_Object, value {object=null, objectClass=&Object}}
  }
  User.Variant {type=NiagaraFloat, defaultValue=0, description=""}
 ]
}
