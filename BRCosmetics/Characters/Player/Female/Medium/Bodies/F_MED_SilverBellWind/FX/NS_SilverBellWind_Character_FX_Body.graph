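# Niagara cosmetic effect for the body of the SilverBellWind female character, built from one burst emitter, body_spline, with three mesh renderers and a sprite renderer.
# Particles persist past their lifetime, follow sampled skeletal mesh locations, and use their spawn index to pick per-mesh colors.
# A smoothed moving average of position, packed with the Interpolate Intensity user parameter, is written to a particle attribute.
# System AlphaFade drives mesh scale and opacity.
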NiagaraSystem /BRCosmetics/Characters/Player/Female/Medium/Bodies/F_MED_SilverBellWind/FX/NS_SilverBellWind_Character_FX_Body
dependencies {
 usedDataInterfaces [&NiagaraDataInterfaceArrayFloat, &NiagaraDataInterfaceSkeletalMesh, &NiagaraDataInterfaceCurve]
 usedDynamicInputs [&FloatFromCurve, &MakeCustomFloatFromBool]
 usedModules [
  &self:SystemSpawnScript.SystemScriptSource.SystemScriptGraph.NiagaraNodeAssignment_0.SetVariables_#1
  &SystemState
  &FN_VariantManager
  &self:SystemSpawnScript.SystemScriptSource.SystemScriptGraph.NiagaraNodeAssignment_1.SetVariables_#2
 ]
}
emitters {body_spline {
  emitterUpdateScript [
   &EmitterState()
   &SpawnBurst_Instantaneous(Use Spawn Probability=true, UseLoopCountLimit=true)
  ]
  particleSpawnScript [
   /SetVariables(Particles.MeshIndex=&ReturnExecIndex()) {isSetParameters=true}
   &InitializeParticle(
    Color=&MultiplyLinearColor(
     Linear Color=&SelectLinearColorFromArray(
      Array Sampling Mode=::NewEnumerator1
      Direct Array Index=&ReturnExecIndex()
      Color Selection Array=DataInterface() {propertyValues {ColorData [<rgba 0.5 0.3677 0.1542 6>, <rgba 1 0.15 0.2065 8>, <rgba 0.2871 0.2816 1 8>]}}
      Random Seed=?
     )
     Scale Factor=&MakeVector4_FromLinearColor(LinearColor=<rgba 1 1 1 1.5>)
    )
    Lifetime Max=1.5
    Lifetime Mode=::NewEnumerator1
    Mesh Scale=&Multiply_VectorByFloat(Vector=<1 1 1>, Float=1.5)
    Mesh Scale Mode=::NewEnumerator4
    Sprite Size Mode=::NewEnumerator3
    UsePositionOffset=false
    Write Material Random=true
   ) {module=InitializeParticle001}
   &CameraOffset(Camera Offset Amount=2)
   &InitialMeshOrientation(Mesh Orientation Mode=::NewEnumerator7, Rotation=<0.4 0.2 -0.4>, Use Rotation Vector=true) {enabled=false}
   &InitialMeshOrientation(
    Mesh Axis To Orient=<0 0 1>
    Mesh Orientation Mode=::NewEnumerator2
    Orientation Coordinate Space=::World
    Orientation Vector=<0 0 1>
    Rotation=<-0.12 0 0>
    Use Rotation Vector=true
   ) {module=InitialMeshOrientation001}
  ]
  particleUpdateScript [
   &ParticleState(Kill Particles When Lifetime Has Elapsed=false)
   &ScaleMeshSize(Scale Factor=&Lerp_Vector(A=<0 0 0>, B=<1 1 1>, Alpha=$System.AlphaFade))
   &SkeletalMeshLocation(
    MaskSpawnGroup=false
    Mesh Sampling Type=::NewEnumerator5
    Orientation Sampling=::NewEnumerator1
    Sampled Position Offset=<0 0.2 1>
    Sampled Position Offset Coordinate Space:ENiagaraSkelMeshTransforms=::NewEnumerator2
    Skeletal Mesh=$System.Skeletal Mesh
    Socket Sampling Mode=::NewEnumerator5
   )
   &InterpolateOverTime(
    Rate Of Change=8
    Target Vector Value:Vector3f=&ConvertPositionToVector(Input Position=$Particles.Position, Convert To Absolute Large World Vector=false)
    Variable Type=::NewEnumerator2
   )
   /SetVariables(Particles.Interpolate Dynamic=&MakeFloat4FromVectorAndFloat(
     "Vector (RGB)"=$StackContext.InterpolateOverTime.InterpolateOverTime_V3_.Moving Average
     "Float (Alpha)"=&Add_Float(A=$User.Interpolate Intensity, B=-0.5)
    )) {isSetParameters=true}
   &DynamicMaterialParameters(
    Index 0 Param 1=&RandomRangeFloat(Minimum=0, Maximum=1)
    Index 0 Param 2=&Multiply_Float(
     A=0.3
     B=&SelectFloatFromArray(Direct Array Index=$Particles.MeshIndex, Float Selection Array=DataInterface() {propertyValues {FloatData [8,8,8]}})
    )
    Module.Param0WriteEnabled=true
    Module.Param1WriteEnabled=true
    Module.Param2WriteEnabled=true
    Module.Param3WriteEnabled=true
   )
   &FN_Mesh_Attributes(
    Color=&MultiplyLinearColor(
     Linear Color=&self:ColorSelector(
      index:NiagaraInt32=$Particles.MeshIndex
      Linear Color:LinearColor=<rgba 4 1.7422 0.0122 1>
      Linear Color001:LinearColor=<rgba 1 0.1332 0.1332 1>
      Linear Color002:LinearColor=<rgba 0 0.1833 1 1>
     )
     Scale Factor=&MakeFloat4FromVectorAndFloat("Vector (RGB)"=<2 2 2>, "Float (Alpha)"=&Multiply_Float(A=0.2, B=$System.AlphaFade))
    )
    Create Color Attribute=true
    Write to Module Namespace=true
   )
   &FN_Mesh_Attributes(
    Color=&MultiplyLinearColor(
     Linear Color=&self:ColorSelector(
      index:NiagaraInt32=$Particles.MeshIndex
      Linear Color:LinearColor=<rgba 0.3822 0 1 0.1>
      Linear Color001:LinearColor=<rgba 1 0.2061 0.7099 0.1>
      Linear Color002:LinearColor=<rgba 0 0.0595 1 0.1>
     )
     Scale Factor=&MakeFloat4FromVectorAndFloat("Vector (RGB)"=<1 1 1>, "Float (Alpha)"=$System.AlphaFade)
    )
    Create Color Attribute=true
    Write to Module Namespace=true
   ) {module=FN_Mesh_Attributes002}
   &FN_Sprite_Attributes(
    Camera Offset=10
    Color=&MultiplyLinearColor(
     Linear Color=<rgba 10 5.9984 2.2601 1>
     Scale Factor=&MakeFloat4FromVectorAndFloat("Vector (RGB)"=<1 1 1>, "Float (Alpha)"=$System.AlphaFade)
    )
    Create Camera Offset Attribute=true
    Create Color Attribute=true
    Create Position Attribute=true
    Create SpriteSize Attribute=true
    Position:NiagaraPosition=&AddVectorToPosition(Position=$Particles.SkeletalMeshLocation.SampledPosition, Vector=<0 0 0>)
    SpriteSize=&Vector2DFromFloat(Value=10)
    Write to Module Namespace=true
   )
   /SetVariables(
    Particles.Decal_DY0:Vector4f=&MakeVector4(X=0.8, Y=0, Z=1, W=&RandomRangeFloat(Minimum=0, Maximum=1))
    Particles.Decal_DY1:Vector4f=&MakeVector4(X=20, Y=0, Z=0, W=0)
   ) {isSetParameters=true}
   &ScaleColor(Scale Alpha=$System.AlphaFade, ScaleA=true, ScaleRGB=true)
  ]
  rendererClasses [&NiagaraMeshRendererProperties, &NiagaraSpriteRendererProperties]
  renderers [
   {rendererClass=&NiagaraMeshRendererProperties, rendererIndex=0}
   {rendererClass=&NiagaraMeshRendererProperties, rendererIndex=1}
   {rendererClass=&NiagaraMeshRendererProperties, rendererIndex=2}
   {rendererClass=&NiagaraSpriteRendererProperties, rendererIndex=3}
  ]
 }}
summary {
 systemName=NS_SilverBellWind_Character_FX_Body
 emitters [{emitterName=body_spline, rendererClasses [&NiagaraMeshRendererProperties, &NiagaraSpriteRendererProperties]}]
 userVariables [
  User.bIsFrontEnd {type=NiagaraBool}
  User.bIsFrontEndPreview {type=NiagaraBool}
  User.bisHandHidden {type=NiagaraBool}
  User.BodyPart {type=NiagaraFloat, defaultValue=0, description=""}
  User.Color Array {
   type=NiagaraDataInterfaceArrayColor
   description=""
   defaultValue {
    struct=NiagaraExt_VariableValue_DataInterface
    value {
     dataInterface=&self:NiagaraDataInterfaceArrayColor_0
     dataInterfaceClass=&NiagaraDataInterfaceArrayColor
    }
   }
  }
  User.Interpolate Intensity {type=NiagaraFloat, defaultValue=1, description=""}
  User.Variant {type=NiagaraFloat, defaultValue=0, description=""}
 ]
}
