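# Material function that builds the animated, glowing storm-noise effect for the SilverBellWind female character outfit, plus a Noise mask output.
# A panning distortion texture offsets scrolling UVs shared by lightning, caustics and water-normal samples.
# Fresnel and a light wave from the global sunlight vector, with clamped height, provide view- and light-dependent masks.
# Two color layers, a dark-to-light gradient and a tint, get power-curved brightness boosts in highlights before being summed.
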MaterialFunction /BRCosmetics/Characters/Player/Female/Medium/Bodies/F_MED_SilverBellWind/Materials/MF_StormNoise_SilverBell
graph {nodes [
  FunctionOutput(None=NamedRerouteUsage())
  n1 = NamedRerouteUsage()
  n2 = NamedRerouteUsage()
  n3 = Add(A=Saturate(None=Add(A=@n1, B=@n2)), B=NamedRerouteUsage())
  FunctionOutput(None=NamedRerouteDeclaration(None=Saturate(None=@n3)) {name=Noise_OUT, nodeColor=<rgba 1 0.5156 0 1>}) {outputName=Noise, sortPriority=1}
  n4 = NamedRerouteUsage()
  Multiply(
   A=@n4
   B=LinearInterpolate(
    A=ScalarParameter() {defaultValue=0.5, group=StormNoise, parameterName=NoiseSubtractMax}
    B=ScalarParameter() {defaultValue=0.25, group=StormNoise, parameterName=NoiseSubtractMin}
    Alpha=NamedRerouteUsage()
   )
  )
  n5 = StaticSwitchParameter(True=TextureCoordinate() {coordinateIndex=1}, False=TextureCoordinate()) {group=StormNoise, parameterName=UseUVchannel01}
  n6 = VectorParameter() {group=StormNoise, parameterName=DistortionTiling, channelNames {a="", b="", g=Tile Y, r=Tile X}, defaultValue=<rgba 4 3 2 1>}
  n7 = ComponentMask(None=@n6) {g=true, r=true}
  n8 = VectorParameter() {group=StormNoise, parameterName=DistortionSpeed, channelNames {a="", b="", g=Speed Y, r=Speed X}, defaultValue=<rgba -0.1 0.1 1 1>}
  n9 = TextureSampleParameter2D(UVs=Panner(Coordinate=Multiply(A=@n5, B=@n7), Speed=ComponentMask(None=@n8) {g=true, r=true})) {
   group=StormNoise
   parameterName=DistortionTexture
   samplerType=SAMPLERTYPE_LinearColor
   texture=&T_NoisePack_LinearColor
  }
  NamedRerouteDeclaration(None=ChannelMaskParameter(None=@n9.RGB) {group=StormNoise, parameterName=DistortionMaskChannel}.R) {name=DistortionMask, nodeColor=<rgba 0 1 0.0859 1>}
  NamedRerouteDeclaration(None=ComponentMask(None=ConstantBiasScale(None=@n9.RGB) {bias=-0.5, scale=2}) {g=true, r=true}) {name=UVDistortion, nodeColor=<rgba 1 0 0.75 1>}
  n10 = FunctionInput(Preview=GetMaterialAttributes().MaterialAttributes) {bUsePreviewValueAsDefault=true, inputName=Material, inputType=FunctionInput_MaterialAttributes}
  NamedRerouteDeclaration(None=GetMaterialAttributes(None=NamedRerouteDeclaration(None=@n10) {name=InMaterial, nodeColor=<rgba 1 0.9375 0 1>}) {attributeGetTypes [#1]}.Normal) {name=InNormal, nodeColor=<rgba 0.7578 0 1 1>}
  n11 = CollectionParameter() {
   collection=&FortniteMaterialParameters
   parameterName=SunlightDirectionalVector
  }
  n12 = ComponentMask(None=NamedRerouteDeclaration(None=@n11) {name=LightVecWS, nodeColor=<rgba 0 1 0.9297 1>}) {b=true, g=true, r=true}
  n13 = MaterialFunctionCall(Float3=@n12) {materialFunction=&BreakOutFloat3Components}
  n14 = MaterialFunctionCall(X=@n13.R, Y=@n13.R, Z=Max(A=@n13.B) {constB=0.2}) {materialFunction=&MakeFloat3}
  n15 = MaterialFunctionCall(Float3=@n12) {materialFunction=&BreakOutFloat3Components}
  n16 = Normalize(VectorInput=MaterialFunctionCall(X=@n15.R, Y=@n15.R, Z=Max(A=@n15.B) {constB=0.9}) {materialFunction=&MakeFloat3}.Result)
  n17 = NamedRerouteUsage()
  n18 = DotProduct(A=Normalize(VectorInput=LinearInterpolate(A=@n14.Result, B=@n16, Alpha=@n17)), B=Transform(None=NamedRerouteUsage()))
  NamedRerouteDeclaration(None=Power(Base=Saturate(None=@n18), Exp=ScalarParameter() {defaultValue=5, parameterName=LightPower})) {name=LightWave, nodeColor=<rgba 0 1 0.5078 1>}
  n19 = StaticSwitchParameter(True=TextureCoordinate() {coordinateIndex=1}, False=TextureCoordinate()) {group=StormNoise, parameterName=UseUVchannel01}
  n20 = VectorParameter() {group=StormNoise, parameterName=NoiseTiling, channelNames {a="", b="", g=Tile Y, r=Tile X}, defaultValue=<rgba 1 1 1 1>}
  n21 = ComponentMask(None=@n20) {g=true, r=true}
  n22 = ScalarParameter() {defaultValue=0.02, group=StormNoise, parameterName=DistortionStrength}
  n23 = NamedRerouteUsage()
  n24 = VectorParameter() {group=StormNoise, parameterName=NoiseSpeed, channelNames {a="", b="", g=Speed Y, r=Speed X}, defaultValue=<rgba 0.025 -0.05 1 1>}
  n25 = Add(A=Add(A=Multiply(A=@n19, B=@n21), B=Multiply(A=@n22, B=@n23)), B=Multiply(A=Time(), B=ComponentMask(None=@n24) {g=true, r=true}))
  NamedRerouteDeclaration(None=TextureSampleParameter2D(UVs=@n25) {group=StormNoise, parameterName=NoiseTexture02, texture=&T_ShieldCaustics}.RGB) {name=NoiseSampleTint, nodeColor=<rgba 1 0 0.3281 1>}
  NamedRerouteDeclaration(None=FunctionInput() {bUsePreviewValueAsDefault=true, inputName=FrontEndOverride, inputType=FunctionInput_Scalar, sortPriority=3}) {name=FrontEndOverride, nodeColor=<rgba 0 0.7187 1 1>}
  n26 = TextureSampleParameter2D(UVs=@n25) {
   group=StormNoise
   parameterName=NoiseTexture01
   samplerType=SAMPLERTYPE_LinearGrayscale
   texture=/Game/Characters/Player/Female/Medium/Bodies/F_MED_SilverBellWind/Textures/T_SilverBellWind_Lightning
  }
  NamedRerouteDeclaration(None=@n26.R) {name=NoiseSample, nodeColor=<rgba 0 0.3437 1 1>}
  n27 = TextureSampleParameter2D(UVs=@n25) {
   group=StormNoise
   parameterName=NoiseNormal
   samplerType=SAMPLERTYPE_Normal
   texture=/Game/Characters/Player/Female/Medium/Bodies/F_MED_SilverBellWind/Textures/T_WaterDistortion_N
  }
  NamedRerouteDeclaration(None=@n27.RGB) {name=NoiseNormalSample, nodeColor=<rgba 1 0 0.3516 1>}
  NamedRerouteDeclaration(None=@n4) {name=FresnelMask, nodeColor=<rgba 1 0.7031 0 1>}
  n28 = NamedRerouteUsage()
  n29 = OneMinus(None=DotProduct(A=CameraVectorWS(), B=Transform(None=@n28)))
  NamedRerouteDeclaration(None=Power(Base=ConstantBiasScale(None=@n29))) {name=NoiseNormalFresnel, nodeColor=<rgba 0 1 0.0859 1>}
  n30 = VectorParameter() {group=StormNoise, parameterName=NoiseColorDark, defaultValue=<rgba 0.0088 0.0392 0.0999 1>}
  n31 = VectorParameter() {group=StormNoise, parameterName=NoiseColorLight, defaultValue=<rgba 0.2646 0.5736 0.7011 1>}
  n32 = Saturate(None=Add(A=NamedRerouteUsage(), B=NamedRerouteUsage()))
  n33 = Power(Base=@n32, Exp=ScalarParameter() {defaultValue=1.3, group=StormNoise, parameterName=NoisePower01}) {constExponent=1.5}
  n34 = Multiply(A=@n32) {constB=0.9}
  n35 = ScalarParameter() {defaultValue=6, group=StormNoise, parameterName=HighPower01}
  n36 = Multiply(
   A=Multiply(A=LinearInterpolate(A=@n30.RGB, B=@n31.RGB, Alpha=@n33), B=NamedRerouteUsage())
   B=LinearInterpolate(
    A=ScalarParameter() {defaultValue=1, group=StormNoise, parameterName=Brightness01}
    B=ScalarParameter() {defaultValue=3, group=StormNoise, parameterName=BrightnessHigh01}
    Alpha=Power(Base=@n34, Exp=@n35) {constExponent=4}
   )
  )
  n37 = NamedRerouteUsage()
  n38 = ScalarParameter() {defaultValue=6, group=StormNoise, parameterName=HighPower02}
  n39 = Multiply(
   A=Multiply(A=@n37, B=VectorParameter() {group=StormNoise, parameterName=NoiseTint, defaultValue=<rgba 1 0.1389 0.9332 1>}.RGB)
   B=LinearInterpolate(
    A=ScalarParameter() {defaultValue=2, group=StormNoise, parameterName=Brightness02}
    B=ScalarParameter() {defaultValue=5, group=StormNoise, parameterName=BrightnessHigh02}
    Alpha=Power(Base=@n37, Exp=@n38)
   )
  )
  NamedRerouteDeclaration(None=Multiply(A=Add(A=@n36, B=@n39), B=NamedRerouteUsage())) {name=Output, nodeColor=<rgba 0.0312 0 1 1>}
  n40 = FunctionInput() {bUsePreviewValueAsDefault=true, inputName=Alpha, inputType=FunctionInput_Scalar, sortPriority=1, previewValue=<1 0 0 1>}
  NamedRerouteDeclaration(None=@n40) {name=Alpha, nodeColor=<rgba 0.1484 0 1 1>}
 ]}
