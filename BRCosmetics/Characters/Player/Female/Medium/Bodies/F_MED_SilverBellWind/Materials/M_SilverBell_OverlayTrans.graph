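# Unlit translucent overlay material for the SilverBellWind character body, faking glass on skeletal meshes.
# Character parent function outputs feed tinted base color and a Blinn/Phong-blended highlight from the directional light, falling back to the sun parameter, which also drives opacity.
# Optionally masked storm noise and a switchable Fresnel-scaled thin-film color layer add emissive detail.
# Emissive receives time-of-day tinting, then passes through accessory tech-art overrides.
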Material /BRCosmetics/Characters/Player/Female/Medium/Bodies/F_MED_SilverBellWind/Materials/M_SilverBell_OverlayTrans
graph {
 properties {
  bCastRayTracedShadows=false
  bUseMaterialAttributes=true
  bUsedWithMeshDeformer=true
  bUsedWithMorphTargets=true
  bUsedWithSkeletalMesh=true
  blendMode=BLEND_Translucent
  pixelDepthOffsetMode=PDOM_Legacy
  shadingModel=MSM_Unlit
  usedShadingModels=Unlit
 }
 nodes [
  AtmosphericLightVector()
  n1 = VectorParameter() {group=Toon_Shading, parameterName=ScreenSpaceLightVector, defaultValue=<rgba 1 1 -1 1>}
  Composite!(LightVector=@n1.RGB) {subgraphName=ProjectScreenAlighned}
  n2 = NamedRerouteUsage()
  n3 = Multiply(A=@n2, B=VectorParameter() {group=FakeGlass, parameterName=EveningTint, defaultValue=<rgba 1 0.6167 0.9182 1>}.RGB)
  n4 = Multiply(A=@n2, B=VectorParameter() {group=FakeGlass, parameterName=NightTint, defaultValue=<rgba 0.5417 0.5626 1 1>}.RGB)
  MaterialFunctionCall(In=SetMaterialAttributes!(
    MaterialAttributes
    Emissive Color=MaterialFunctionCall(Color_Morning=@n2, Color_Day=@n2, Color_Evening=@n3, Color_Night=@n4) {materialFunction=&Color_Time_of_Day}.Result
    Opacity=NamedRerouteUsage()
   ) {attributeSetTypes [#1, #2]}) {materialFunction=&MF_TechArtOverrides_accessories}
  n5 = MaterialFunctionCall() {materialFunction=&MF_CharacterParent_2023_PreFX}
  n6 = GetMaterialAttributes(None=@n5.Result) {attributeGetTypes [#3, #4, #1]}
  NamedRerouteDeclaration(None=@n6.Normal) {name=Normal_IN, nodeColor=<rgba 0 1 0.8594 1>}
  n7 = LinearInterpolate(Alpha=Fresnel() {baseReflectFraction=0, exponent=1}) {constA=1, constB=2}
  n8 = NamedRerouteUsage()
  n9 = NamedRerouteUsage()
  n10 = Add(
   A=Multiply(A=ComponentMask(None=@n8) {r=true}, B=ScalarParameter() {defaultValue=0.1, group=SolidColor, parameterName=DistortNoiseIntensity_SolidColor})
   B=Add(A=ScalarParameter() {group=SolidColor, parameterName=SolidThinFilmOffset}, B=@n9)
  )
  n11 = ScalarParameter() {defaultValue=1.2, group=SolidColor, parameterName=SolidColor_UV_Divsior}
  n12 = TextureSampleParameter2D(UVs=AppendVector(A=Divide(A=@n10, B=@n11), B=Constant())) {
   group=SolidColor
   parameterName=Thin Film Texture_RE
   sortPriority=1
   texture=/Game/Global/Textures/Default/Thinfilm/T_ThinFilm_Spectrum_COLOR
  }
  n13 = SmoothStep(
   Min=ScalarParameter() {group=SolidColor, parameterName=NoiseMaskMin01}
   Max=ScalarParameter() {defaultValue=1, group=SolidColor, parameterName=NoiseMaskMax01}
   Value=NamedRerouteUsage()
  )
  n14 = ScalarParameter() {defaultValue=4, group=SolidColor, parameterName=NoiseMaskPower01}
  n15 = SmoothStep(
   Min=ScalarParameter() {group=SolidColor, parameterName=NoiseMaskMin02}
   Max=ScalarParameter() {defaultValue=1, group=SolidColor, parameterName=NoiseMaskMax02}
   Value=OneMinus(None=@n9)
  )
  n16 = ScalarParameter() {defaultValue=2, group=SolidColor, parameterName=NoiseMaskPower02}
  n17 = NamedRerouteUsage()
  n18 = Multiply(A=@n17, B=OneMinus(None=@n17))
  n19 = NamedRerouteUsage()
  n20 = Add(A=Multiply(A=Power(Base=@n13, Exp=@n14), B=Power(Base=@n15, Exp=@n16)), B=Multiply(B=Multiply(A=@n18, B=@n19)) {constA=1})
  n21 = ScalarParameter() {defaultValue=4, group=SolidColor, parameterName=SolidColorIntensity}
  n22 = Multiply(A=@n21)
  n23 = Multiply(
   A=Multiply(A=@n12.RGB, B=NamedRerouteDeclaration(None=@n20) {name=SolidColorCompMask, nodeColor=<rgba 0 1 0.0156 1>})
   B=MaterialFunctionCall!(
    Low _ PC Low and Med=@n21
    "High _ XB1, PS4"=@n21
    "Epic _ PC High, XSX, PS5"=@n21
    --------
    "Mobile Low _ PC Perf, Android"=@n22
    Mobile Med _ Switch=@n22
   ) {materialFunction=/Game/Global/MaterialFunctions/TechArt/MF_Fortnite_QualitySwitch_Scalar}.Result
  )
  NamedRerouteDeclaration(None=StaticSwitchParameter(True=Multiply(A=@n7, B=@n23), False=Constant()) {group=FakeGlass, parameterName=UseSolidFading}) {name=AdditionalFadingEM, nodeColor=<rgba 1 0.8906 0 1>}
  n24 = MaterialFunctionCall!(Material=NamedRerouteUsage(), Alpha=Constant() {r=1}, FrontEndOverride=NamedRerouteUsage()) {materialFunction=/Game/Characters/Player/Female/Medium/Bodies/F_MED_SilverBellWind/Materials/MF_StormNoise_SilverBell}
  NamedRerouteDeclaration(None=@n24.Noise) {name=NoiseDistort, nodeColor=<rgba 0 0.6719 1 1>}
  MaterialExpressionReroute_12 = Reroute(None=@n1.RGB)
  n25 = Normalize(VectorInput=@MaterialExpressionReroute_12.LightVector)
  n26 = ComponentMask(None=@n25) {r=true}
  n27 = Constant3Vector() {constant=<rgba 0 1 0 0>}
  n28 = Normalize(VectorInput=WorldPosition() {worldPositionShaderOffset=WPT_CameraRelative}.XYZ)
  n29 = Normalize(VectorInput=CrossProduct(A=Transform(None=@n27) {transformSourceType=TRANSFORMSOURCE_View}, B=@n28))
  n30 = ComponentMask(None=@n25) {g=true}
  n31 = CrossProduct(A=@n28, B=@n29)
  n32 = Add(A=Add(A=Multiply(A=@n26, B=@n29), B=Multiply(A=@n30, B=@n31)), B=Multiply(A=ComponentMask(None=@n25) {b=true}, B=@n28))
  MaterialExpressionReroute_14 = Reroute(None=@n32)
  n33 = ComponentMask(None=Custom() {code="return View.DirectionalLightColor;", description=DirectionalLightColor, outputType=CMOT_Float4, inputs []}) {b=true, g=true, r=true}
  n34 = Custom() {code="return View.DirectionalLightDirection;", description=DirectionalLightDirection, inputs []}
  n35 = CollectionParameter() {
   collection=&FortniteMaterialParameters
   parameterName=SunlightDirectionalVector
  }
  n36 = ComponentMask(None=@n35) {b=true, g=true, r=true}
  n37 = If(A=DotProduct(A=@n34, B=@n34), "A > B"=@n34, "A == B"=@n36, "A < B"=@n36) {desc=check if DirectionalLight exists}
  n38 = StaticSwitchParameter(
   True=@MaterialExpressionReroute_14.Result
   False=MaterialFunctionCall!(In=If(A=@n33, "A > B"=@n37, "A == B"=@n36, "A < B"=@n36) {desc=check if DirectionalLight color is Black}) {materialFunction=/Game/Packages/Fortress_SharedMaterials/Cinematics/Anime/MF_CINE_AnimeLightDirOverride}.Result
  ) {
   desc="Use a Screen / View alight light vector. \r\nOverrides the UseDirectionalLightFunction!"
   group=Toon_Shading
   parameterName=UseScreenSpaceLightDirection
   sortPriority=0
  }
  n39 = MaterialFunctionCall(Float3=@n38) {materialFunction=&BreakOutFloat3Components}
  n40 = ScalarParameter() {
   desc="0.0 = the light vector is locked on the XY plane.\r\n1.0 = the actual light vector is used  \r\n(for Faces use a blend value of 0,0)"
   group=Toon_Shading
   parameterName=LockLightZ_Blend
   sliderMax=1
  }
  n41 = MaterialFunctionCall(X=@n39.R, Y=@n39.R, Z=Multiply(A=@n39.B, B=@n40)) {materialFunction=&MakeFloat3}
  n42 = StaticSwitchParameter(True=LinearInterpolate(A=@n38, B=Normalize(VectorInput=@n41.Result), Alpha=VertexColor().A), False=@n38) {
   desc="Lock the Z axis of the Light Vector\r\nis driven by the VertexColor A channel (0 = Regular LightVector, 1= Locked LIghtVector)"
   group=Toon_Shading
   parameterName=UseLockLightZVector
   sortPriority=0
  }
  NamedRerouteDeclaration(None=ComponentMask(None=@n42) {b=true, g=true, r=true}) {name=LightVector, nodeColor=<rgba 0 1 0.8634 1>}
  NamedRerouteDeclaration(None=@n6.Emissive Color) {name=EM_IN, nodeColor=<rgba 0 1 0.1797 1>}
  NamedRerouteDeclaration(None=ChannelMaskParameter(None=@n5.Tech Art Mask) {group=FakeGlass, parameterName=UseFakeGlass}.R) {name=TechArtMask, nodeColor=<rgba 1 0.9375 0 1>}
  n43 = ComponentMask(None=NamedRerouteUsage()) {b=true, g=true, r=true}
  n44 = PixelNormalWS()
  NamedRerouteDeclaration(None=Saturate(None=DotProduct(A=@n43, B=@n44))) {name=FakeDiffuse, nodeColor=<rgba 1 0 0.4687 1>}
  NamedRerouteDeclaration(None=@n6.MaterialAttributes) {name=MA_IN, nodeColor=<rgba 0.5703 0 1 1>}
  NamedRerouteDeclaration(None=ChannelMaskParameter(None=@n5.Tech Art Mask) {group=FakeGlass, maskChannel=Green, parameterName=MaskNoise, defaultValue=<rgba 0 1 0 0>}.R) {name=NoiseMask, nodeColor=<rgba 0.4766 1 0 1>}
  n45 = VectorParameter() {group=FakeGlass, parameterName=BaseColorTint, channelNames {a=HighLightIntensity, b="", g="", r=""}, defaultValue=<rgba 1 1 1 1>}
  n46 = Multiply(A=NamedRerouteDeclaration(None=@n6.Base Color) {name=BaseColorIN, nodeColor=<rgba 0 0.0625 1 1>}, B=Multiply(A=@n45, B=@n45))
  n47 = Multiply(A=ScalarParameter() {defaultValue=0.8, parameterName=BaseColorLow}, B=@n46) {constB=1.2}
  n48 = NamedRerouteDeclaration(None=@n46) {name=BaseColorTinted, nodeColor=<rgba 0.125 0 1 1>}
  n49 = NamedRerouteUsage()
  n50 = VectorParameter() {group=FakeGlass, parameterName=HighLight, channelNames {a=HighLightIntensity, b="", g="", r=""}, defaultValue=<rgba 1 1 1 1>}
  n51 = NamedRerouteDeclaration(None=Multiply(A=@n50, B=@n50)) {name=HighLight, nodeColor=<rgba 1 0 0.5859 1>}
  n52 = Multiply(A=@n51) {constB=0.7}
  n53 = NamedRerouteUsage()
  n54 = TextureSampleParameter2D() {parameterName=Flat_Normal, samplerType=SAMPLERTYPE_Normal, texture=&FlatNormalBlue}
  n55 = NamedRerouteUsage()
  n56 = Multiply(
   A=MaterialFunctionCall(Vector 3=Subtract(A=@n54.RGB, B=@n55)) {materialFunction=/Engine/Functions/Engine_MaterialFunctions02/Utility/VectorLength}.V3 Length
   B=ScalarParameter() {defaultValue=0.3, group=FakeGlass, parameterName=NormalEdgingMask, sliderMax=1}
  )
  n57 = CameraVectorWS()
  n58 = ComponentMask(None=Custom() {code="return View.DirectionalLightColor;", description=DirectionalLightColor, outputType=CMOT_Float4, inputs []}) {b=true, g=true, r=true}
  n59 = Custom() {code="return View.DirectionalLightDirection;", description=DirectionalLightDirection, inputs []}
  n60 = CollectionParameter() {
   collection=&FortniteMaterialParameters
   parameterName=SunlightDirectionalVector
  }
  n61 = ComponentMask(None=@n60) {b=true, g=true, r=true}
  n62 = If(A=DotProduct(A=@n59, B=@n59), "A > B"=@n59, "A == B"=@n61, "A < B"=@n61) {desc=check if DirectionalLight exists}
  n63 = ComponentMask(None=NamedRerouteDeclaration(None=If(A=@n58, "A > B"=@n62, "A == B"=@n61, "A < B"=@n61) {desc=check if DirectionalLight color is Black}) {name=LightVec, nodeColor=<rgba 0 1 0.9297 1>}) {b=true, g=true, r=true}
  n64 = MaterialFunctionCall(Float3=@n63) {materialFunction=&BreakOutFloat3Components}
  n65 = MaterialFunctionCall(X=@n64.R, Y=@n64.R, Z=Max(A=@n64.B) {constB=0.2}) {materialFunction=&MakeFloat3}
  n66 = MaterialFunctionCall(Float3=@n63) {materialFunction=&BreakOutFloat3Components}
  n67 = MaterialFunctionCall(X=@n66.R, Y=@n66.R, Z=Max(A=@n66.B) {constB=0.9}) {materialFunction=&MakeFloat3}
  n68 = ScalarParameter() {group=FakeGlass, parameterName=IsFrontEnd}
  n69 = ComponentMask(None=LinearInterpolate(
    A=Normalize(VectorInput=@n65.Result)
    B=Normalize(VectorInput=@n67.Result)
    Alpha=NamedRerouteDeclaration(None=@n68) {name=FrontEnd, nodeColor=<rgba 0.8047 1 0 1>}
   )) {b=true, g=true, r=true}
  n70 = DotProduct(A=Normalize(VectorInput=Add(A=@n57, B=@n69)), B=Normalize(VectorInput=PixelNormalWS())) {desc=Blinn}
  n71 = DotProduct(A=ReflectionVectorWS(), B=Normalize(VectorInput=@n69)) {desc=Phong}
  n72 = ScalarParameter() {defaultValue=1, group=FakeGlass, parameterName=BlinnOrPhong}
  n73 = Power(
   Base=Max(A=LinearInterpolate(A=@n70, B=@n71, Alpha=@n72)) {constB=0}
   Exp=ScalarParameter() {defaultValue=20, group=FakeGlass, parameterName=SpecPow}
  )
  n74 = Add(
   A=NamedRerouteUsage()
   B=LinearInterpolate(A=LinearInterpolate(A=@n47, B=@n48, Alpha=@n49), B=LinearInterpolate(A=@n51, B=@n52, Alpha=@n53), Alpha=Add(A=@n56, B=@n73))
  )
  n75 = NamedRerouteUsage()
  n76 = OneMinus(None=@n75)
  n77 = LinearInterpolate(
   B=@n24.Result
   Alpha=StaticSwitchParameter(True=StaticSwitchParameter(True=@n76, False=@n75) {group=FakeGlass, parameterName=InvertNoiseMask}, False=Constant() {r=1}) {group=FakeGlass, parameterName=UseNoiseMask}
  )
  NamedRerouteDeclaration(None=Add(A=Add(A=@n74, B=@n77), B=NamedRerouteUsage())) {name=EmissiveOUT, nodeColor=<rgba 1 0.2344 0 1>}
  n78 = LinearInterpolate(
   A=ScalarParameter() {defaultValue=0.3, group=FakeGlass, parameterName=OpacityMin}
   B=ScalarParameter() {defaultValue=1, group=FakeGlass, parameterName=OpacityMax}
   Alpha=@n73
  )
  NamedRerouteDeclaration(None=LinearInterpolate(B=Saturate(None=@n78), Alpha=NamedRerouteUsage()) {constB=0}) {name=Opacity_OUT, nodeColor=<rgba 0.6875 0 1 1>}
  PinBase!(Result=@n32) {reroutePins [Result {expression=&self:MaterialExpressionReroute_14}]}
  PinBase!() {reroutePins [LightVector {expression=&self:MaterialExpressionReroute_12}]}
  SkyAtmosphereLightDirection()
 ]
}
