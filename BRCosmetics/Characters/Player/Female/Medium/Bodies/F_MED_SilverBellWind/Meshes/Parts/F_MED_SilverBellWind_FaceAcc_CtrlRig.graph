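# Control Rig for the SilverBellWind female face accessory mesh, procedurally posing its braids and back hair as the character moves and aims.
# Rotation constraints first blend hair bones between upper spine and head or neck, followed by fixed right-braid offsets.
# Spherical pose readers on the head and right upper arm then weight corrective offsets.
# Pistol and rifle targeting variables gate the arm correctives and a final head-driven set, whose offsets are all zero.
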ControlRigBlueprint /BRCosmetics/Characters/Player/Female/Medium/Bodies/F_MED_SilverBellWind/Meshes/Parts/F_MED_SilverBellWind_FaceAcc_CtrlRig
graphs [
 RigVMFunctionLibrary {links [], nodes []}
 RigVMModel {nodes [
   &RigUnit_ModifyTransforms(
    ExecutePin=()
    ItemToModify=((Item=(Type=Bone,Name="dyn_HairBraid_l_MSTR"),Transform=(Rotation=(X=0,Y=0,Z=0.2588,W=0.9659),Translation=(X=0,Y=0,Z=0),Scale3D=(X=1,Y=1,Z=1))))
    Weight=1
   )
   > &RigUnit_ModifyTransforms(
    ItemToModify=((Item=(Type=Bone,Name="dyn_HairBraid_l_02_jnt"),Transform=(Rotation=(X=0,Y=0,Z=0.2588,W=0.9659),Translation=(X=0,Y=0,Z=0),Scale3D=(X=1,Y=1,Z=1))))
    Weight=1
   )
   RigUnit_BeginExecution = &RigUnit_BeginExecution()
   &RigUnit_RotationConstraintLocalSpaceOffset(
    AdvancedSettings=(RotationOrderForFilter=XZY)
    Child=(Type=Bone,Name="dyn_Hair_bk_l_MSTR")
    Filter=(bX=True,bY=True,bZ=True)
    Parents=((Item=(Type=Bone,Name="spine_05"),Weight=0.5),(Item=(Type=Bone,Name="head"),Weight=0.5))
    Weight=1
    ExecutePin=@RigUnit_BeginExecution.ExecutePin
   )
   > &RigUnit_RotationConstraintLocalSpaceOffset(
    AdvancedSettings=(RotationOrderForFilter=XZY)
    Child=(Type=Bone,Name="dyn_Hair_bk_c_MSTR")
    Filter=(bX=True,bY=True,bZ=True)
    Parents=((Item=(Type=Bone,Name="spine_05"),Weight=0.5),(Item=(Type=Bone,Name="head"),Weight=0.5))
    Weight=1
   )
   > &RigUnit_RotationConstraintLocalSpaceOffset(
    AdvancedSettings=(RotationOrderForFilter=XZY)
    Child=(Type=Bone,Name="dyn_Hair_bk_r_MSTR")
    Filter=(bX=True,bY=True,bZ=True)
    Parents=((Item=(Type=Bone,Name="spine_05"),Weight=0.5),(Item=(Type=Bone,Name="head"),Weight=0.5))
    Weight=1
    bMaintainOffset=true
   )
   > &RigUnit_RotationConstraintLocalSpaceOffset(
    AdvancedSettings=(RotationOrderForFilter=XZY)
    Child=(Type=Bone,Name="dyn_Hair_bk_l_02_jnt")
    Filter=(bX=True,bY=True,bZ=True)
    Parents=((Item=(Type=Bone,Name="spine_05"),Weight=0.5),(Item=(Type=Bone,Name="neck_01"),Weight=0.5))
    Weight=1
   )
   > &RigUnit_RotationConstraintLocalSpaceOffset(
    AdvancedSettings=(RotationOrderForFilter=XZY)
    Child=(Type=Bone,Name="dyn_Hair_bk_c_02_jnt")
    Filter=(bX=True,bY=True,bZ=True)
    Parents=((Item=(Type=Bone,Name="spine_05"),Weight=0.5),(Item=(Type=Bone,Name="neck_01"),Weight=0.5))
    Weight=1
   )
   > &RigUnit_RotationConstraintLocalSpaceOffset(
    AdvancedSettings=(RotationOrderForFilter=XZY)
    Child=(Type=Bone,Name="dyn_Hair_bk_r_02_jnt")
    Filter=(bX=True,bY=True,bZ=True)
    Parents=((Item=(Type=Bone,Name="spine_05"),Weight=0.5),(Item=(Type=Bone,Name="neck_01"),Weight=0.5))
    Weight=1
    bMaintainOffset=true
   )
   > &RigUnit_RotationConstraintLocalSpaceOffset(
    AdvancedSettings=(RotationOrderForFilter=XZY)
    Child=(Type=Bone,Name="dyn_HairBraid_l_MSTR")
    Filter=(bX=True,bY=True,bZ=True)
    Parents=((Item=(Type=Bone,Name="spine_05"),Weight=0.5),(Item=(Type=Bone,Name="head"),Weight=0.5))
    Weight=1
    bMaintainOffset=true
   )
   > &RigUnit_RotationConstraintLocalSpaceOffset(
    AdvancedSettings=(RotationOrderForFilter=XZY)
    Child=(Type=Bone,Name="dyn_HairBraid_r_MSTR")
    Filter=(bX=True,bY=True,bZ=True)
    Parents=((Item=(Type=Bone,Name="spine_05"),Weight=0.5),(Item=(Type=Bone,Name="head"),Weight=0.5))
    Weight=1
    bMaintainOffset=true
   )
   > &RigUnit_ModifyTransforms(
    ItemToModify=((Item=(Type=Bone,Name="dyn_HairBraid_r_MSTR"),Transform=(Rotation=(X=0.2519,Y=-0.0615,Z=-0.1633,W=0.9519),Translation=(X=0,Y=0,Z=0),Scale3D=(X=1,Y=1,Z=1))))
    Weight=1
   )
   > &RigUnit_ModifyTransforms(
    ItemToModify=((Item=(Type=Bone,Name="dyn_HairBraid_r_02_jnt"),Transform=(Rotation=(X=0,Y=0,Z=-0.0872,W=0.9962),Translation=(X=0,Y=0,Z=0),Scale3D=(X=1,Y=1,Z=1))))
    Weight=1
   )
   > n1 = &RigUnit_SphericalPoseReader(
    ActiveRegionScaleFactors=(PositiveWidth=1,NegativeWidth=1.5,PositiveHeight=1,NegativeHeight=1)
    ActiveRegionSize=0.1
    Debug=(bDrawDebug=false,DebugScale=25,DebugSegments=20,DebugThickness=0.25)
    DriverAxis=(X=0,Y=-1,Z=0)
    DriverItem=(Type=Bone,Name="head")
    FalloffRegionScaleFactors=(PositiveWidth=2,NegativeWidth=2,PositiveHeight=2,NegativeHeight=2)
    FalloffSize=0.2
    FlipWidthScaling=false
    OptionalParentItem=(Type=Bone,Name="neck_01")
    RotationOffset=(X=-90,Y=0,Z=-55)
   )
   > &RigUnit_ModifyTransforms(
    ItemToModify=((Item=(Type=Bone,Name="dyn_HairBraid_l_MSTR"),Transform=(Rotation=(X=0,Y=0,Z=0,W=1),Translation=(X=0,Y=0,Z=0),Scale3D=(X=1,Y=1,Z=1))))
    ItemToModify.0.Transform.Rotation=&RigVMFunction_MathQuaternionFromEuler(Euler=(X=0,Y=0,Z=-20)).Result
    Weight=@n1.OutputParam
   )
   > &RigUnit_ModifyTransforms(
    ItemToModify=((Item=(Type=Bone,Name="dyn_HairBraid_r_MSTR"),Transform=(Rotation=(X=0,Y=0,Z=0,W=1),Translation=(X=0,Y=0,Z=0),Scale3D=(X=1,Y=1,Z=1))))
    ItemToModify.0.Transform.Rotation=&RigVMFunction_MathQuaternionFromEuler(Euler=(X=0,Y=0,Z=20)).Result
    Weight=@n1.OutputParam
   )
   > &RigUnit_ModifyTransforms(
    ItemToModify=((Item=(Type=Bone,Name="dyn_Hair_bk_r_MSTR"),Transform=(Rotation=(X=0,Y=0,Z=0,W=1),Translation=(X=0,Y=0,Z=0),Scale3D=(X=1,Y=1,Z=1))))
    ItemToModify.0.Transform.Rotation=&RigVMFunction_MathQuaternionFromEuler(Euler=(X=0,Y=-30,Z=0)).Result
    Weight=@n1.OutputParam
   )
   > &RigUnit_ModifyTransforms(
    ItemToModify=((Item=(Type=Bone,Name="dyn_Hair_bk_l_MSTR"),Transform=(Rotation=(X=0,Y=0,Z=0,W=1),Translation=(X=0,Y=0,Z=0),Scale3D=(X=1,Y=1,Z=1))))
    ItemToModify.0.Transform.Rotation=&RigVMFunction_MathQuaternionFromEuler(Euler=(X=0,Y=-30,Z=0)).Result
    Weight=@n1.OutputParam
   )
   > &RigUnit_ModifyTransforms(
    ItemToModify=((Item=(Type=Bone,Name="dyn_Hair_bk_c_MSTR"),Transform=(Rotation=(X=0,Y=0,Z=0,W=1),Translation=(X=0,Y=0,Z=0),Scale3D=(X=1,Y=1,Z=1))))
    ItemToModify.0.Transform.Rotation=&RigVMFunction_MathQuaternionFromEuler(Euler=(X=0,Y=-30,Z=0)).Result
    Weight=@n1.OutputParam
   )
   > &RigUnit_ModifyTransforms(
    ItemToModify=((Item=(Type=Bone,Name="dyn_Hair_bk_r_02_jnt"),Transform=(Rotation=(X=0,Y=0,Z=0,W=1),Translation=(X=0,Y=0,Z=0),Scale3D=(X=1,Y=1,Z=1))))
    ItemToModify.0.Transform.Rotation=&RigVMFunction_MathQuaternionFromEuler(Euler=(X=0,Y=30,Z=0)).Result
    Weight=@n1.OutputParam
   )
   > &RigUnit_ModifyTransforms(
    ItemToModify=((Item=(Type=Bone,Name="dyn_Hair_bk_l_02_jnt"),Transform=(Rotation=(X=0,Y=0,Z=0,W=1),Translation=(X=0,Y=0,Z=0),Scale3D=(X=1,Y=1,Z=1))))
    ItemToModify.0.Transform.Rotation=&RigVMFunction_MathQuaternionFromEuler(Euler=(X=0,Y=30,Z=0)).Result
    Weight=@n1.OutputParam
   )
   > &RigUnit_ModifyTransforms(
    ItemToModify=((Item=(Type=Bone,Name="dyn_Hair_bk_c_02_jnt"),Transform=(Rotation=(X=0,Y=0,Z=0,W=1),Translation=(X=0,Y=0,Z=0),Scale3D=(X=1,Y=1,Z=1))))
    ItemToModify.0.Transform.Rotation=&RigVMFunction_MathQuaternionFromEuler(Euler=(X=0,Y=30,Z=0)).Result
    Weight=@n1.OutputParam
   )
   > &RigUnit_ModifyTransforms(
    ItemToModify=((Item=(Type=Bone,Name="Hair_MSTR"),Transform=(Rotation=(X=0,Y=0,Z=0.2588,W=0.9659),Translation=(X=0,Y=0,Z=0),Scale3D=(X=1,Y=1,Z=1))))
    Weight=@n1.OutputParam
   )
   > n2 = &RigUnit_SphericalPoseReader(
    ActiveRegionScaleFactors=(PositiveWidth=0.25,NegativeWidth=0.25,PositiveHeight=1,NegativeHeight=1)
    ActiveRegionSize=0.1
    Debug=(bDrawDebug=false,DebugScale=25,DebugSegments=20,DebugThickness=0.25)
    DriverAxis=(X=0,Y=1,Z=0)
    DriverItem=(Type=Bone,Name="head")
    FalloffRegionScaleFactors=(PositiveWidth=0.75,NegativeWidth=0.5,PositiveHeight=4,NegativeHeight=4)
    FalloffSize=0.2
    OptionalParentItem=(Type=Bone,Name="spine_05")
    RotationOffset=(X=90,Y=0,Z=25)
   )
   > &RigUnit_ModifyTransforms(
    ItemToModify=((Item=(Type=Bone,Name="dyn_HairBraid_l_MSTR"),Transform=(Rotation=(X=0,Y=0,Z=0,W=1),Translation=(X=0,Y=0,Z=0),Scale3D=(X=1,Y=1,Z=1))))
    ItemToModify.0.Transform.Rotation=&RigVMFunction_MathQuaternionFromEuler(Euler=(X=0,Y=-10,Z=-10)).Result
    Weight=@n2.OutputParam
   )
   > &RigUnit_ModifyTransforms(
    ItemToModify=((Item=(Type=Bone,Name="dyn_HairBraid_r_MSTR"),Transform=(Rotation=(X=0,Y=0,Z=0,W=1),Translation=(X=0,Y=0,Z=0),Scale3D=(X=1,Y=1,Z=1))))
    ItemToModify.0.Transform.Rotation=&RigVMFunction_MathQuaternionFromEuler(Euler=(X=0,Y=-10,Z=10)).Result
    Weight=@n2.OutputParam
   )
   > &RigUnit_ModifyTransforms(
    ItemToModify=((Item=(Type=Bone,Name="dyn_Hair_bk_r_MSTR"),Transform=(Rotation=(X=0,Y=0,Z=0,W=1),Translation=(X=0,Y=0,Z=0),Scale3D=(X=1,Y=1,Z=1))))
    ItemToModify.0.Transform.Rotation=&RigVMFunction_MathQuaternionFromEuler(Euler=(X=0,Y=-5,Z=0)).Result
    Weight=@n2.OutputParam
   )
   > &RigUnit_ModifyTransforms(
    ItemToModify=((Item=(Type=Bone,Name="dyn_Hair_bk_l_MSTR"),Transform=(Rotation=(X=0,Y=0,Z=0,W=1),Translation=(X=0,Y=0,Z=0),Scale3D=(X=1,Y=1,Z=1))))
    ItemToModify.0.Transform.Rotation=&RigVMFunction_MathQuaternionFromEuler(Euler=(X=0,Y=-5,Z=0)).Result
    Weight=@n2.OutputParam
   )
   > &RigUnit_ModifyTransforms(
    ItemToModify=((Item=(Type=Bone,Name="dyn_Hair_bk_c_MSTR"),Transform=(Rotation=(X=0,Y=0,Z=0,W=1),Translation=(X=0,Y=0,Z=0),Scale3D=(X=1,Y=1,Z=1))))
    ItemToModify.0.Transform.Rotation=&RigVMFunction_MathQuaternionFromEuler(Euler=(X=0,Y=-5,Z=0)).Result
    Weight=@n2.OutputParam
   )
   > &RigUnit_ModifyTransforms(
    ItemToModify=((Item=(Type=Bone,Name="Hair_MSTR"),Transform=(Rotation=(X=0,Y=0,Z=-0.2588,W=0.9659),Translation=(X=0,Y=0,Z=0),Scale3D=(X=1,Y=1,Z=1))))
    Weight=@n2.OutputParam
   )
   > n3 = &RigUnit_SphericalPoseReader(
    ActiveRegionScaleFactors=(PositiveWidth=1,NegativeWidth=0.5,PositiveHeight=2.5,NegativeHeight=1)
    ActiveRegionSize=0.1
    Debug=(bDrawDebug=false,DebugScale=25,DebugSegments=20,DebugThickness=0.25)
    DriverAxis=(X=0,Y=-1,Z=0)
    DriverItem=(Type=Bone,Name="head")
    FalloffRegionScaleFactors=(PositiveWidth=0.5,NegativeWidth=0.5,PositiveHeight=1,NegativeHeight=0.5)
    FalloffSize=0.3
    OptionalParentItem=(Type=Bone,Name="neck_02")
    RotationOffset=(X=-55,Y=0,Z=-10)
   )
   > &RigUnit_ModifyTransforms(
    ItemToModify=((Item=(Type=Bone,Name="dyn_HairBraid_r_MSTR"),Transform=(Rotation=(X=0,Y=0,Z=0,W=1),Translation=(X=0,Y=0,Z=0),Scale3D=(X=1,Y=1,Z=1))))
    ItemToModify.0.Transform.Rotation=&RigVMFunction_MathQuaternionFromEuler(Euler=(X=0,Y=10,Z=18)).Result
    Weight=@n3.OutputParam
   )
   > &RigUnit_ModifyTransforms(
    ItemToModify=((Item=(Type=Bone,Name="dyn_HairBraid_l_MSTR"),Transform=(Rotation=(X=0,Y=0,Z=0,W=1),Translation=(X=0,Y=0,Z=0),Scale3D=(X=1,Y=1,Z=1))))
    ItemToModify.0.Transform.Rotation=&RigVMFunction_MathQuaternionFromEuler(Euler=(X=0,Y=10,Z=10)).Result
    Weight=@n3.OutputParam
   )
   > &RigUnit_ModifyTransforms(
    ItemToModify=((Item=(Type=Bone,Name="dyn_Hair_bk_l_MSTR"),Transform=(Rotation=(X=0,Y=0,Z=0,W=1),Translation=(X=0,Y=0,Z=0),Scale3D=(X=1,Y=1,Z=1))))
    ItemToModify.0.Transform.Rotation=&RigVMFunction_MathQuaternionFromEuler(Euler=(X=0,Y=10,Z=0)).Result
    Weight=@n3.OutputParam
   )
   > &RigUnit_ModifyTransforms(
    ItemToModify=((Item=(Type=Bone,Name="dyn_Hair_bk_r_MSTR"),Transform=(Rotation=(X=0,Y=0,Z=0,W=1),Translation=(X=0,Y=0,Z=0),Scale3D=(X=1,Y=1,Z=1))))
    ItemToModify.0.Transform.Rotation=&RigVMFunction_MathQuaternionFromEuler(Euler=(X=0,Y=-18,Z=0)).Result
    Weight=@n3.OutputParam
   )
   > n4 = &RigUnit_SphericalPoseReader(
    ActiveRegionScaleFactors=(PositiveWidth=1,NegativeWidth=0.5,PositiveHeight=1,NegativeHeight=2.5)
    ActiveRegionSize=0.1
    Debug=(bDrawDebug=false,DebugScale=25,DebugSegments=20,DebugThickness=0.25)
    DriverAxis=(X=0,Y=-1,Z=0)
    DriverItem=(Type=Bone,Name="head")
    FalloffRegionScaleFactors=(PositiveWidth=0.5,NegativeWidth=0.5,PositiveHeight=0.5,NegativeHeight=1)
    FalloffSize=0.3
    OptionalParentItem=(Type=Bone,Name="neck_02")
    RotationOffset=(X=235,Y=0,Z=-10)
   )
   > &RigUnit_ModifyTransforms(
    ItemToModify=((Item=(Type=Bone,Name="dyn_HairBraid_l_MSTR"),Transform=(Rotation=(X=0,Y=0,Z=0,W=1),Translation=(X=0,Y=0,Z=0),Scale3D=(X=1,Y=1,Z=1))))
    ItemToModify.0.Transform.Rotation=&RigVMFunction_MathQuaternionFromEuler(Euler=(X=0,Y=-10,Z=-18)).Result
    Weight=@n4.OutputParam
   )
   > &RigUnit_ModifyTransforms(
    ItemToModify=((Item=(Type=Bone,Name="dyn_HairBraid_r_MSTR"),Transform=(Rotation=(X=0,Y=0,Z=0,W=1),Translation=(X=0,Y=0,Z=0),Scale3D=(X=1,Y=1,Z=1))))
    ItemToModify.0.Transform.Rotation=&RigVMFunction_MathQuaternionFromEuler(Euler=(X=0,Y=-10,Z=-10)).Result
    Weight=@n4.OutputParam
   )
   > &RigUnit_ModifyTransforms(
    ItemToModify=((Item=(Type=Bone,Name="dyn_Hair_bk_l_MSTR"),Transform=(Rotation=(X=0,Y=0,Z=0,W=1),Translation=(X=0,Y=0,Z=0),Scale3D=(X=1,Y=1,Z=1))))
    ItemToModify.0.Transform.Rotation=&RigVMFunction_MathQuaternionFromEuler(Euler=(X=0,Y=-18,Z=0)).Result
    Weight=@n4.OutputParam
   )
   > &RigUnit_ModifyTransforms(
    ItemToModify=((Item=(Type=Bone,Name="dyn_Hair_bk_r_MSTR"),Transform=(Rotation=(X=0,Y=0,Z=0,W=1),Translation=(X=0,Y=0,Z=0),Scale3D=(X=1,Y=1,Z=1))))
    ItemToModify.0.Transform.Rotation=&RigVMFunction_MathQuaternionFromEuler(Euler=(X=0,Y=10,Z=0)).Result
    Weight=@n4.OutputParam
   )
   > n5 = &RigUnit_SphericalPoseReader(
    ActiveRegionScaleFactors=(PositiveWidth=1,NegativeWidth=1,PositiveHeight=0.25,NegativeHeight=1)
    ActiveRegionSize=0.1
    Debug=(bDrawDebug=false,bDraw2D=false,bDrawLocalAxes=false,DebugScale=25,DebugSegments=20,DebugThickness=0.25)
    DriverAxis=(X=-1,Y=0,Z=0)
    DriverItem=(Type=Bone,Name="upperarm_r")
    FalloffRegionScaleFactors=(PositiveWidth=0.5,NegativeWidth=0.5,PositiveHeight=0.5,NegativeHeight=0.5)
    FalloffSize=0.45
    OptionalParentItem=(Type=Bone,Name="spine_05")
    RotationOffset=(X=-90,Y=40,Z=45)
   )
   Multiply = &RigVMFunction_MathDoubleMul(A=@n5.OutputParam, B=RigVMVariableNode() {variable=Is Pistol Targeting}.Value)
   n6 = &RigVMFunction_AlphaInterp(InterpSpeedDecreasing=4, InterpSpeedIncreasing=10, Scale=1, bClampResult=False, bInterpResult=False, Value=@Multiply.Result)
   > &RigUnit_ModifyTransforms(
    ItemToModify=((Item=(Type=Bone,Name="dyn_Hair_bk_r_MSTR"),Transform=(Rotation=(X=0,Y=0,Z=0,W=1),Translation=(X=0,Y=0,Z=6),Scale3D=(X=1,Y=1,Z=1))))
    Weight=@n6.Result
    ItemToModify.0.Transform.Rotation=&RigVMFunction_MathQuaternionFromEuler(Euler=(X=0,Y=0,Z=0)).Result
   )
   > &RigUnit_ModifyTransforms(
    ItemToModify=((Item=(Type=Bone,Name="dyn_Hair_bk_l_MSTR"),Transform=(Rotation=(X=0,Y=0,Z=0,W=1),Translation=(X=0,Y=0,Z=0),Scale3D=(X=1,Y=1,Z=1))))
    Weight=@n6.Result
    ItemToModify.0.Transform.Rotation=&RigVMFunction_MathQuaternionFromEuler(Euler=(X=0,Y=10,Z=0)).Result
   )
   > &RigUnit_ModifyTransforms(
    ItemToModify=((Item=(Type=Bone,Name="dyn_Hair_bk_c_MSTR"),Transform=(Rotation=(X=0,Y=0,Z=0,W=1),Translation=(X=0,Y=0,Z=0),Scale3D=(X=1,Y=1,Z=1))))
    Weight=@n6.Result
    ItemToModify.0.Transform.Rotation=&RigVMFunction_MathQuaternionFromEuler(Euler=(X=0,Y=10,Z=0)).Result
   )
   > &RigUnit_ModifyTransforms(
    ItemToModify=((Item=(Type=Bone,Name="dyn_HairBraid_l_MSTR"),Transform=(Rotation=(X=0,Y=0,Z=0,W=1),Translation=(X=0,Y=0,Z=0),Scale3D=(X=1,Y=1,Z=1))))
    Weight=@n6.Result
    ItemToModify.0.Transform.Rotation=&RigVMFunction_MathQuaternionFromEuler(Euler=(X=0,Y=0,Z=10)).Result
   )
   > &RigUnit_ModifyTransforms(
    ItemToModify=((Item=(Type=Bone,Name="dyn_HairBraid_r_MSTR"),Transform=(Rotation=(X=0,Y=0,Z=0,W=1),Translation=(X=0,Y=0,Z=0),Scale3D=(X=1,Y=1,Z=1))))
    Weight=@n6.Result
    ItemToModify.0.Transform.Rotation=&RigVMFunction_MathQuaternionFromEuler(Euler=(X=0,Y=-30,Z=20)).Result
   )
   > &RigUnit_ModifyTransforms(
    ItemToModify=((Item=(Type=Bone,Name="dyn_Hair_bk_r_02_jnt"),Transform=(Rotation=(X=0,Y=0,Z=0,W=1),Translation=(X=0,Y=0,Z=0),Scale3D=(X=1,Y=1,Z=1))))
    Weight=@n6.Result
    ItemToModify.0.Transform.Rotation=&RigVMFunction_MathQuaternionFromEuler(Euler=(X=0,Y=20,Z=0)).Result
   )
   > &RigUnit_ModifyTransforms(
    ItemToModify=((Item=(Type=Bone,Name="Hair_MSTR"),Transform=(Rotation=(X=0,Y=0,Z=0.2588,W=0.9659),Translation=(X=0,Y=0,Z=0),Scale3D=(X=1,Y=1,Z=1))))
    Weight=@n6.Result
   )
   > n7 = &RigUnit_SphericalPoseReader(
    ActiveRegionScaleFactors=(PositiveWidth=1,NegativeWidth=0.25,PositiveHeight=1,NegativeHeight=1)
    ActiveRegionSize=0.1
    Debug=(bDrawDebug=false,bDraw2D=false,bDrawLocalAxes=false,DebugScale=25,DebugSegments=20,DebugThickness=0.25)
    DriverAxis=(X=-1,Y=0,Z=0)
    DriverItem=(Type=Bone,Name="upperarm_r")
    FalloffRegionScaleFactors=(PositiveWidth=0.5,NegativeWidth=0.25,PositiveHeight=0.5,NegativeHeight=0.5)
    FalloffSize=0.35
    OptionalParentItem=(Type=Bone,Name="spine_05")
    RotationOffset=(X=-45,Y=0,Z=-15)
   )
   AlphaInterp = &RigVMFunction_AlphaInterp(
    InterpSpeedDecreasing=4
    InterpSpeedIncreasing=10
    Scale=1
    bClampResult=False
    bInterpResult=False
    Value=&RigVMFunction_MathDoubleMul(A=@n7.OutputParam, B=RigVMVariableNode() {variable=Is Rifle Targeting}.Value).Result
   )
   > &RigUnit_ModifyTransforms(
    ItemToModify=((Item=(Type=Bone,Name="dyn_Hair_bk_r_MSTR"),Transform=(Rotation=(X=0,Y=0,Z=0,W=1),Translation=(X=0,Y=0,Z=0),Scale3D=(X=1,Y=1,Z=1))))
    Weight=@AlphaInterp.Result
    ItemToModify.0.Transform.Rotation=&RigVMFunction_MathQuaternionFromEuler(Euler=(X=0,Y=-30,Z=-10)).Result
   )
   > &RigUnit_ModifyTransforms(
    ItemToModify=((Item=(Type=Bone,Name="dyn_Hair_bk_l_MSTR"),Transform=(Rotation=(X=0,Y=0,Z=0,W=1),Translation=(X=0,Y=0,Z=0),Scale3D=(X=1,Y=1,Z=1))))
    Weight=@AlphaInterp.Result
    ItemToModify.0.Transform.Rotation=&RigVMFunction_MathQuaternionFromEuler(Euler=(X=0,Y=10,Z=0)).Result
   )
   > &RigUnit_ModifyTransforms(
    ItemToModify=((Item=(Type=Bone,Name="dyn_Hair_bk_c_MSTR"),Transform=(Rotation=(X=0,Y=0,Z=0,W=1),Translation=(X=0,Y=0,Z=0),Scale3D=(X=1,Y=1,Z=1))))
    Weight=@AlphaInterp.Result
    ItemToModify.0.Transform.Rotation=&RigVMFunction_MathQuaternionFromEuler(Euler=(X=0,Y=0,Z=-10)).Result
   )
   > &RigUnit_ModifyTransforms(
    ItemToModify=((Item=(Type=Bone,Name="dyn_HairBraid_l_MSTR"),Transform=(Rotation=(X=0,Y=0,Z=0,W=1),Translation=(X=0,Y=0,Z=0),Scale3D=(X=1,Y=1,Z=1))))
    Weight=@AlphaInterp.Result
    ItemToModify.0.Transform.Rotation=&RigVMFunction_MathQuaternionFromEuler(Euler=(X=0,Y=10,Z=10)).Result
   )
   > &RigUnit_ModifyTransforms(
    ItemToModify=((Item=(Type=Bone,Name="dyn_HairBraid_r_MSTR"),Transform=(Rotation=(X=0,Y=0,Z=0,W=1),Translation=(X=0,Y=0,Z=0),Scale3D=(X=1,Y=1,Z=1))))
    Weight=@AlphaInterp.Result
    ItemToModify.0.Transform.Rotation=&RigVMFunction_MathQuaternionFromEuler(Euler=(X=0,Y=-30,Z=10)).Result
   )
   > &RigUnit_ModifyTransforms(
    ItemToModify=((Item=(Type=Bone,Name="dyn_Hair_bk_r_02_jnt"),Transform=(Rotation=(X=0,Y=0,Z=0,W=1),Translation=(X=0,Y=0,Z=0),Scale3D=(X=1,Y=1,Z=1))))
    Weight=@AlphaInterp.Result
    ItemToModify.0.Transform.Rotation=&RigVMFunction_MathQuaternionFromEuler(Euler=(X=0,Y=20,Z=0)).Result
   )
   > &RigUnit_ModifyTransforms(
    ItemToModify=((Item=(Type=Bone,Name="Hair_MSTR"),Transform=(Rotation=(X=0,Y=0,Z=0.2588,W=0.9659),Translation=(X=0,Y=0,Z=0),Scale3D=(X=1,Y=1,Z=1))))
    Weight=@AlphaInterp.Result
   )
   > n8 = &RigUnit_SphericalPoseReader(
    ActiveRegionScaleFactors=(PositiveWidth=0.25,NegativeWidth=1,PositiveHeight=1,NegativeHeight=1)
    ActiveRegionSize=0.1
    Debug=(bDrawDebug=false,DebugScale=25,DebugSegments=20,DebugThickness=0.25)
    DriverAxis=(X=0,Y=1,Z=0)
    DriverItem=(Type=Bone,Name="head")
    FalloffRegionScaleFactors=(PositiveWidth=0.25,NegativeWidth=1,PositiveHeight=1,NegativeHeight=1)
    FalloffSize=0.2
    OptionalParentItem=(Type=Bone,Name="spine_05")
    RotationOffset=(X=110,Y=0,Z=5)
   )
   n9 = &RigVMFunction_AlphaInterp(
    InterpSpeedDecreasing=4
    InterpSpeedIncreasing=10
    Scale=1
    bClampResult=False
    bInterpResult=False
    Value=&RigVMFunction_MathDoubleMul(A=@n8.OutputParam, B=RigVMVariableNode() {variable=Is Rifle Targeting}.Value).Result
   )
   > &RigUnit_ModifyTransforms(
    ItemToModify=((Item=(Type=Bone,Name="dyn_Hair_bk_r_MSTR"),Transform=(Rotation=(X=0,Y=0,Z=0,W=1),Translation=(X=0,Y=0,Z=0),Scale3D=(X=1,Y=1,Z=1))))
    Weight=@n9.Result
    ItemToModify.0.Transform.Rotation=&RigVMFunction_MathQuaternionFromEuler(Euler=(X=0,Y=0,Z=0)).Result
   )
   > &RigUnit_ModifyTransforms(
    ItemToModify=((Item=(Type=Bone,Name="dyn_Hair_bk_l_MSTR"),Transform=(Rotation=(X=0,Y=0,Z=0,W=1),Translation=(X=0,Y=0,Z=0),Scale3D=(X=1,Y=1,Z=1))))
    Weight=@n9.Result
    ItemToModify.0.Transform.Rotation=&RigVMFunction_MathQuaternionFromEuler(Euler=(X=0,Y=0,Z=0)).Result
   )
   > &RigUnit_ModifyTransforms(
    ItemToModify=((Item=(Type=Bone,Name="dyn_Hair_bk_c_MSTR"),Transform=(Rotation=(X=0,Y=0,Z=0,W=1),Translation=(X=0,Y=0,Z=0),Scale3D=(X=1,Y=1,Z=1))))
    Weight=@n9.Result
    ItemToModify.0.Transform.Rotation=&RigVMFunction_MathQuaternionFromEuler(Euler=(X=0,Y=0,Z=0)).Result
   )
   > &RigUnit_ModifyTransforms(
    ItemToModify=((Item=(Type=Bone,Name="dyn_HairBraid_l_MSTR"),Transform=(Rotation=(X=0,Y=0,Z=0,W=1),Translation=(X=0,Y=0,Z=0),Scale3D=(X=1,Y=1,Z=1))))
    Weight=@n9.Result
    ItemToModify.0.Transform.Rotation=&RigVMFunction_MathQuaternionFromEuler(Euler=(X=0,Y=0,Z=0)).Result
   )
   > &RigUnit_ModifyTransforms(
    ItemToModify=((Item=(Type=Bone,Name="dyn_HairBraid_r_MSTR"),Transform=(Rotation=(X=0,Y=0,Z=0,W=1),Translation=(X=0,Y=0,Z=0),Scale3D=(X=1,Y=1,Z=1))))
    Weight=@n9.Result
    ItemToModify.0.Transform.Rotation=&RigVMFunction_MathQuaternionFromEuler(Euler=(X=0,Y=0,Z=0)).Result
   )
   > &RigUnit_ModifyTransforms(
    ItemToModify=((Item=(Type=Bone,Name="dyn_Hair_bk_r_02_jnt"),Transform=(Rotation=(X=0,Y=0,Z=0,W=1),Translation=(X=0,Y=0,Z=0),Scale3D=(X=1,Y=1,Z=1))))
    Weight=@n9.Result
    ItemToModify.0.Transform.Rotation=&RigVMFunction_MathQuaternionFromEuler(Euler=(X=0,Y=0,Z=0)).Result
   )
  ]}
]
hierarchy [
 root {loc=<0 0 0>, rot=<0 0 0 1>}
 pelvis {parent=root, loc=<0 -0.5301 98.0571>, rot=<0.0127 0.707 -0.0127 -0.707>}
 spine_01 {parent=pelvis, loc=<6.1208 -0.9328 0>, rot=<0 0 0.034 -0.9994>}
 spine_02 {parent=spine_01, loc=<6.5996 -0.2095 0>, rot=<0 0 0.0238 -0.9997>}
 spine_03 {parent=spine_02, loc=<6.4914 -0.1237 0>, rot=<0 0 0.0149 -0.9999>}
 spine_04 {parent=spine_03, loc=<6.9598 0.0794 0>, rot=<0 0 -0.0132 -0.9999>}
 spine_05 {parent=spine_04, loc=<6.9987 0 0>, rot=<0 0 -0.0015 -1>}
 neck_01 {parent=spine_05, loc=<11.3138 0.1529 0>, rot=<0 0 -0.2003 -0.9797>}
 neck_02 {parent=neck_01, loc=<4.5627 0 0>, rot=<0 0 0.0504 -0.9987>}
 head {parent=neck_02, loc=<5.3914 0 0>, rot=<0 0 0.115 -0.9934>}
 dyn_HairBraid_l_MSTR {parent=head, loc=<3.7955 -0.7336 -8.4792>, rot=<0.2426 -0.9624 -0.0339 -0.1174>}
 dyn_HairBraid_l_01_jnt {parent=dyn_HairBraid_l_MSTR, loc=<0 0 0>, rot=<0 0 0 1>}
 dyn_HairBraid_l_02_jnt {parent=dyn_HairBraid_l_01_jnt, loc=<5.3762 0 0>, rot=<0 0 0 1>}
 dyn_HairBraid_bk_l_01_jnt {parent=dyn_HairBraid_l_02_jnt, loc=<1.9964 -0.9175 -0.0968>, rot=<0.0019 0.022 0.0839 -0.9962>}
 dyn_HairBraid_bk_l_02_jnt {parent=dyn_HairBraid_bk_l_01_jnt, loc=<3.9394 0 0>, rot=<0 -0.0038 -0.0056 -1>}
 dyn_HairBraid_bk_l_03_jnt {parent=dyn_HairBraid_bk_l_02_jnt, loc=<4.5216 0 0>, rot=<0.0027 -0.0219 -0.1208 -0.9924>}
 dyn_HairBraid_bk_l_04_jnt {parent=dyn_HairBraid_bk_l_03_jnt, loc=<3.6571 0 0>, rot=<0.0011 -0.0529 -0.021 -0.9984>}
 dyn_HairBraid_bk_l_05_jnt {parent=dyn_HairBraid_bk_l_04_jnt, loc=<2.0521 0 0>, rot=<0 0 0 1>}
 dyn_HairBraid_bk_l_06_jnt {parent=dyn_HairBraid_bk_l_05_jnt, loc=<2.1967 0 0>, rot=<0.0018 0.0251 0.0722 -0.9971>}
 dyn_HairBraid_ft_l_01_jnt {parent=dyn_HairBraid_l_02_jnt, loc=<2.0507 0.1679 0>, rot=<-0.0576 0.0508 -0.0212 -0.9968>}
 dyn_HairBraid_ft_l_02_jnt {parent=dyn_HairBraid_ft_l_01_jnt, loc=<3.9373 0 0>, rot=<0 0 0.0099 -1>}
 dyn_HairBraid_ft_l_03_jnt {parent=dyn_HairBraid_ft_l_02_jnt, loc=<3.9714 0 0>, rot=<0 -0.0471 0 -0.9989>}
 dyn_HairBraid_ft_l_04_jnt {parent=dyn_HairBraid_ft_l_03_jnt, loc=<4.1286 0 0>, rot=<0.0006 -0.0283 0.0228 -0.9993>}
 dyn_HairBraid_ft_l_05_jnt {parent=dyn_HairBraid_ft_l_04_jnt, loc=<2.9446 0 0>, rot=<0 0 -0.0564 -0.9984>}
 dyn_HairBraid_ft_l_06_jnt {parent=dyn_HairBraid_ft_l_05_jnt, loc=<2.8788 0 0>, rot=<0 0 -0.0492 -0.9988>}
 dyn_Hair_bk_c_MSTR {parent=head, loc=<9.9034 -7.8655 -0.3914>, rot=<0.0865 -0.7018 0.7018 -0.0865>}
 dyn_Hair_bk_c_01_jnt {parent=dyn_Hair_bk_c_MSTR, loc=<0 0 0>, rot=<0 0 0 1>}
 dyn_Hair_bk_c_02_jnt {parent=dyn_Hair_bk_c_01_jnt, loc=<6.9574 0 0>, rot=<0 0 0 1>}
 dyn_Hair_bk_c_03_jnt {parent=dyn_Hair_bk_c_02_jnt, loc=<6.9574 0 0>, rot=<0 0 0 1>}
 dyn_Hair_bk_c_04_jnt {parent=dyn_Hair_bk_c_03_jnt, loc=<6.9574 0 0>, rot=<0 0 0 1>}
 dyn_Hair_bk_c_05_jnt {parent=dyn_Hair_bk_c_04_jnt, loc=<6.9574 0 0>, rot=<0 0 0 1>}
 dyn_Hair_bk_c_06_jnt {parent=dyn_Hair_bk_c_05_jnt, loc=<6.9574 0 0>, rot=<0 0 0 1>}
 dyn_Hair_bk_l_MSTR {parent=head, loc=<9.5735 -4.311 -5.93>, rot=<0.0482 -0.7055 0.6961 -0.1245>}
 dyn_Hair_bk_l_01_jnt {parent=dyn_Hair_bk_l_MSTR, loc=<0 0 0>, rot=<0 0 0 1>}
 dyn_Hair_bk_l_02_jnt {parent=dyn_Hair_bk_l_01_jnt, loc=<6.9574 0 0>, rot=<0 0 0 1>}
 dyn_Hair_bk_l_03_jnt {parent=dyn_Hair_bk_l_02_jnt, loc=<6.9574 0 0>, rot=<0 0 0 1>}
 dyn_Hair_bk_l_04_jnt {parent=dyn_Hair_bk_l_03_jnt, loc=<6.9574 0 0>, rot=<0 0 0 1>}
 dyn_Hair_bk_l_05_jnt {parent=dyn_Hair_bk_l_04_jnt, loc=<6.9574 0 0>, rot=<0 0 0 1>}
 dyn_Hair_bk_l_06_jnt {parent=dyn_Hair_bk_l_05_jnt, loc=<6.9574 0 0>, rot=<0 0 0 1>}
 dyn_Hair_bk_r_MSTR {parent=head, loc=<9.5737 -4.311 5.9296>, rot=<0.1245 -0.6961 0.7055 -0.0482>}
 dyn_Hair_bk_r_01_jnt {parent=dyn_Hair_bk_r_MSTR, loc=<0 0 0>, rot=<0 0 0 1>}
 dyn_Hair_bk_r_02_jnt {parent=dyn_Hair_bk_r_01_jnt, loc=<6.9577 0 0>, rot=<0 0 0 1>}
 dyn_Hair_bk_r_03_jnt {parent=dyn_Hair_bk_r_02_jnt, loc=<6.9577 0 0>, rot=<0 0 0 1>}
 dyn_Hair_bk_r_04_jnt {parent=dyn_Hair_bk_r_03_jnt, loc=<6.9577 0 0>, rot=<0 0 0 1>}
 dyn_Hair_bk_r_05_jnt {parent=dyn_Hair_bk_r_04_jnt, loc=<6.9577 0 0>, rot=<0 0 0 1>}
 dyn_Hair_bk_r_06_jnt {parent=dyn_Hair_bk_r_05_jnt, loc=<6.9577 0 0>, rot=<0 0 0 1>}
 dyn_HairBraid_r_MSTR {parent=head, loc=<3.9642 -0.7141 9.2748>, rot=<0.1149 0.0416 0.9764 -0.1783>}
 dyn_HairBraid_r_01_jnt {parent=dyn_HairBraid_r_MSTR, loc=<0 0 0>, rot=<0 0 0 1>}
 dyn_HairBraid_r_02_jnt {parent=dyn_HairBraid_r_01_jnt, loc=<5.3673 0.224 0.0794>, rot=<-0.0005 0.0129 -0.0366 -0.9992>}
 dyn_HairBraid_r_03_jnt {parent=dyn_HairBraid_r_02_jnt, loc=<2.5018 0.1497 -0.0787>, rot=<-0.0006 0 -0.0236 -0.9997>}
 dyn_HairBraid_ft_r_01_jnt {parent=dyn_HairBraid_r_03_jnt, loc=<1.4463 -0.4599 -0.0075>, rot=<0.0581 0.0146 -0.0019 -0.9982>}
 dyn_HairBraid_ft_r_02_jnt {parent=dyn_HairBraid_ft_r_01_jnt, loc=<4.0705 0 -0.0001>, rot=<-0.0001 -0.0112 0.0099 -0.9999>}
 dyn_HairBraid_ft_r_03_jnt {parent=dyn_HairBraid_ft_r_02_jnt, loc=<3.2264 0.0002 0>, rot=<0 -0.0342 0 -0.9994>}
 dyn_HairBraid_ft_r_04_jnt {parent=dyn_HairBraid_ft_r_03_jnt, loc=<3.0157 0 0>, rot=<-0.0004 -0.0377 -0.0229 -0.999>}
 dyn_HairBraid_ft_r_05_jnt {parent=dyn_HairBraid_ft_r_04_jnt, loc=<2.8052 0.0156 -0.0001>, rot=<0 0 0.0564 -0.9984>}
 dyn_HairBraid_bk_r_01_jnt {parent=dyn_HairBraid_r_03_jnt, loc=<1.3523 0.4822 -0.1283>, rot=<0.0571 -0.0011 -0.1086 -0.9924>}
 dyn_HairBraid_bk_r_02_jnt {parent=dyn_HairBraid_bk_r_01_jnt, loc=<4.1749 0 0>, rot=<0.0001 -0.0038 0.0407 -0.9992>}
 dyn_HairBraid_bk_r_03_jnt {parent=dyn_HairBraid_bk_r_02_jnt, loc=<3.8502 0 0.0002>, rot=<-0.0031 -0.0219 0.1011 -0.9946>}
 dyn_HairBraid_bk_r_04_jnt {parent=dyn_HairBraid_bk_r_03_jnt, loc=<2.9562 -0.0003 -0.0001>, rot=<0.0003 0.0164 0.0211 -0.9996>}
 dyn_HairBraid_bk_r_05_jnt {parent=dyn_HairBraid_bk_r_04_jnt, loc=<2.9024 0 0>, rot=<0 0 0 -1>}
 dyn_Earring_l_MSTR {parent=head, loc=<-0.3026 -0.0238 -7.0071>, rot=<0.0031 -0.9568 -0.2907 -0.0001>}
 dyn_Earring_l_01_jnt {parent=dyn_Earring_l_MSTR, loc=<0 0 0>, rot=<0 0 0 1>}
 dyn_Earring_l_02_jnt {parent=dyn_Earring_l_01_jnt, loc=<0.4757 0 0>, rot=<0 0 0 1>}
 dyn_Earring_l_03_jnt {parent=dyn_Earring_l_02_jnt, loc=<0.9323 0 0>, rot=<0 0 0 1>}
 dyn_Earring_r_MSTR {parent=head, loc=<-0.3033 -0.0138 6.892>, rot=<-0.012 0.2907 0.9567 -0.0067>}
 dyn_Earring_r_01_jnt {parent=dyn_Earring_r_MSTR, loc=<0 0 0>, rot=<0 0 0 1>}
 dyn_Earring_r_02_jnt {parent=dyn_Earring_r_01_jnt, loc=<0.4634 0 0>, rot=<0 0 0 1>}
 dyn_Earring_r_03_jnt {parent=dyn_Earring_r_02_jnt, loc=<0.933 0 0>, rot=<0 0 0 1>}
 dyn_HairStrand_l_MSTR {parent=head, loc=<3.9299 2.5486 -7.049>, rot=<0.1387 -0.9899 -0.0146 0.026>}
 dyn_HairStrand_l_01_jnt {parent=dyn_HairStrand_l_MSTR, loc=<0 0 0>, rot=<0 0 0 1>}
 dyn_HairStrand_l_02_jnt {parent=dyn_HairStrand_l_01_jnt, loc=<4.3017 0 0>, rot=<-0.0132 -0.1338 -0.097 -0.9862>}
 dyn_HairStrand_l_03_jnt {parent=dyn_HairStrand_l_02_jnt, loc=<4.3017 0 0>, rot=<0.004 -0.1145 -0.0343 -0.9928>}
 dyn_HairStrand_c_MSTR {parent=head, loc=<9.957 7.2022 -3.4683>, rot=<-0.3015 -0.8552 -0.4195 -0.0426>}
 dyn_HairStrand_c_01_jnt {parent=dyn_HairStrand_c_MSTR, loc=<0 0 0>, rot=<0 0 0 1>}
 dyn_HairStrand_c_02_jnt {parent=dyn_HairStrand_c_01_jnt, loc=<2.7601 0 0>, rot=<-0.0216 -0.183 -0.0586 -0.9811>}
 dyn_HairStrand_c_03_jnt {parent=dyn_HairStrand_c_02_jnt, loc=<2.7187 0 0>, rot=<0 -0.1692 0 -0.9856>}
 dyn_Bangs_c_MSTR {parent=head, loc=<10.0985 8.4735 -1.7778>, rot=<-0.3201 -0.7439 -0.5753 0.1145>}
 dyn_Bangs_c_01_jnt {parent=dyn_Bangs_c_MSTR, loc=<0 0 0>, rot=<0 0 0 1>}
 dyn_Bangs_c_02_jnt {parent=dyn_Bangs_c_01_jnt, loc=<4.3017 0 0>, rot=<0 -0.2069 0 -0.9784>}
 dyn_Bangs_c_03_jnt {parent=dyn_Bangs_c_02_jnt, loc=<4.3017 0 0>, rot=<0 0 0 1>}
 dyn_Hairstrand01_r_MSTR {parent=head, loc=<7.2622 8.5997 6.0758>, rot=<-0.0342 0.4283 0.9028 -0.0156>}
 dyn_Hairstrand01_r_01_jnt {parent=dyn_Hairstrand01_r_MSTR, loc=<0 0 0>, rot=<0 0 0 1>}
 dyn_Hairstrand01_r_02_jnt {parent=dyn_Hairstrand01_r_01_jnt, loc=<4.0092 0.03 0.0518>, rot=<-0.0034 -0.0871 0.0508 -0.9949>}
 dyn_Hairstrand01_r_03_jnt {parent=dyn_Hairstrand01_r_02_jnt, loc=<4.0068 0.0465 0>, rot=<0 0 0.0782 -0.9969>}
 dyn_Hairstrand02_r_MSTR {parent=head, loc=<7.2427 6.0391 7.6452>, rot=<0.0028 0.2768 0.9609 -0.0076>}
 dyn_Hairstrand02_r_01_jnt {parent=dyn_Hairstrand02_r_MSTR, loc=<0 0 0>, rot=<0 0 0 1>}
 dyn_Hairstrand02_r_02_jnt {parent=dyn_Hairstrand02_r_01_jnt, loc=<3.743 0 0.1148>, rot=<0.0099 -0.1007 0.0969 -0.9901>}
 dyn_Hairstrand02_r_03_jnt {parent=dyn_Hairstrand02_r_02_jnt, loc=<5.4635 0 -0.276>, rot=<0.0058 -0.1162 0.0499 -0.992>}
 dyn_Hairstrand03_r_MSTR {parent=head, loc=<5.4164 3.712 8.8655>, rot=<-0.0712 0.2657 0.9525 -0.1307>}
 dyn_Hairstrand03_r_01_jnt {parent=dyn_Hairstrand03_r_MSTR, loc=<0 0 0>, rot=<0 0 0 1>}
 dyn_Hairstrand03_r_02_jnt {parent=dyn_Hairstrand03_r_01_jnt, loc=<3.7186 0 0>, rot=<0 -0.091 0 -0.9959>}
 dyn_Hairstrand03_r_03_jnt {parent=dyn_Hairstrand03_r_02_jnt, loc=<3.7186 0 0>, rot=<0 -0.0281 0 -0.9996>}
 Hair_MSTR {parent=head, loc=<0 0 0>, rot=<0 0 0 1>}
 clavicle_l {parent=spine_05, loc=<5.4831 -0.1053 -3.2946>, rot=<0.0465 -0.7289 -0.009 -0.6829>}
 upperarm_l {parent=clavicle_l, loc=<9.8658 -1.2136 -0.4105>, rot=<0.0164 -0.3245 -0.0269 -0.9454>}
 upperarm_twist_01_l {parent=upperarm_l, loc=<7.2135 -0.1134 0>, rot=<0 0 0 1>}
 deform_upperarm_bk_base_l {parent=upperarm_twist_01_l, loc=<-6.4053 -4.3746 0.314>, rot=<0 0 0 -1>}
 deform_upperarm_bk_l {parent=deform_upperarm_bk_base_l, loc=<0 0 0>, rot=<0 0 0 1>}
 deform_upperarm_fr_base_l {parent=upperarm_twist_01_l, loc=<-5.6312 5.092 0.3354>, rot=<0 0 0 -1>}
 deform_upperarm_fr_l {parent=deform_upperarm_fr_base_l, loc=<0 0 0>, rot=<0 0 0 1>}
 deform_upperarm_up_base_l {parent=upperarm_twist_01_l, loc=<-8.891 0.4825 3.9517>, rot=<0 0 0 -1>}
 deform_upperarm_up_l {parent=deform_upperarm_up_base_l, loc=<0 0 0>, rot=<0 0 0 1>}
 clavicle_r {parent=spine_05, loc=<5.4831 -0.1053 3.2946>, rot=<0.7289 0.0465 -0.6829 0.009>}
 upperarm_r {parent=clavicle_r, loc=<-9.8658 1.2136 0.4105>, rot=<0.0164 -0.3245 -0.0269 -0.9454>}
 upperarm_twist_01_r {parent=upperarm_r, loc=<-7.2135 0.1134 0>, rot=<0 0 0 -1>}
 deform_upperarm_bk_base_r {parent=upperarm_twist_01_r, loc=<6.4053 4.3746 -0.314>, rot=<0 0 0 -1>}
 deform_upperarm_bk_r {parent=deform_upperarm_bk_base_r, loc=<0 0 0>, rot=<0 0 0 1>}
 deform_upperarm_fr_base_r {parent=upperarm_twist_01_r, loc=<5.6312 -5.092 -0.3354>, rot=<0 0 0 -1>}
 deform_upperarm_fr_r {parent=deform_upperarm_fr_base_r, loc=<0 0 0>, rot=<0 0 0 1>}
 deform_upperarm_up_base_r {parent=upperarm_twist_01_r, loc=<8.891 -0.4825 -3.9517>, rot=<0 0 0 -1>}
 deform_upperarm_up_r {parent=deform_upperarm_up_base_r, loc=<0 0 0>, rot=<0 0 0 1>}
 root {parent=root, type=Socket, loc=<0 0 0>, rot=<0 0 0 1>}
]
variables [
 Is Pistol Targeting {type=double, default=0}
 Is Rifle Targeting {type=double, default=0}
]
