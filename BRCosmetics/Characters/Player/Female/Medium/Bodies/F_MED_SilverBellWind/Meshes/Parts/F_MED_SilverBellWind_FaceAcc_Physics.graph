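# Physics asset for the SilverBellWind medium female cosmetic's FaceAcc part, defining dynamic hair and braids with capsules on head, neck, upper spine, clavicles and upper arms.
# Zero-size MSTR bodies link two braids, each forking into front and back chains, and three back-hair chains to the head.
# Back-hair joints bend mostly on one swing axis, with free links joining side roots to center.
# Constraints also name dyn_Flower_01_l and dyn_Hair_bk_MSTR, which have no bodies here.
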PhysicsAsset /BRCosmetics/Characters/Player/Female/Medium/Bodies/F_MED_SilverBellWind/Meshes/Parts/F_MED_SilverBellWind_FaceAcc_Physics
bodies {
 clavicle_l [{shapeType=Capsule, length=12, radius=6.25, shapeName=clavicle_l_capsule, center=<3.5 -1.3 -1.3>, rotation=<rot 80 -0.0003 -0.0003>}]
 clavicle_r [{shapeType=Capsule, length=12, radius=6.25, shapeName=clavicle_r_capsule, center=<-3.5 1.3 1.3>, rotation=<rot 80 -0.0003 -0.0004>}]
 dyn_HairBraid_bk_l_01_jnt [{shapeType=Capsule, length=2, radius=0.6, shapeName=dyn_HairBraid_bk_l_01_jnt_capsule, center=<2.5 0 0>, rotation=<rot 90 0 0>}]
 dyn_HairBraid_bk_l_02_jnt [{shapeType=Capsule, length=3, radius=0.6, shapeName=dyn_HairBraid_bk_l_02_jnt_capsule, center=<2.25 0 0>, rotation=<rot 90 0 0>}]
 dyn_HairBraid_bk_l_03_jnt [{shapeType=Capsule, length=2, radius=0.6, shapeName=dyn_HairBraid_bk_l_03_jnt_capsule, center=<2 0 0>, rotation=<rot 90 0 0>}]
 dyn_HairBraid_bk_l_04_jnt [{shapeType=Capsule, length=2, radius=0.6, shapeName=dyn_HairBraid_bk_l_04_jnt_capsule, center=<1.75 0 0>, rotation=<rot 90 0 0>}]
 dyn_HairBraid_bk_l_05_jnt [{shapeType=Capsule, length=4, radius=0.6, shapeName=dyn_HairBraid_bk_l_05_jnt_capsule, center=<3 0 0>, rotation=<rot 90 0 0>}]
 dyn_HairBraid_bk_r_01_jnt [{shapeType=Capsule, length=2, radius=0.6, shapeName=dyn_HairBraid_bk_r_01_jnt_capsule, center=<2.5 0 0>, rotation=<rot 90 0 0>}]
 dyn_HairBraid_bk_r_02_jnt [{shapeType=Capsule, length=2.25, radius=0.6, shapeName=dyn_HairBraid_bk_r_02_jnt_capsule, center=<2 0 0>, rotation=<rot 90 0 0>}]
 dyn_HairBraid_bk_r_03_jnt [{shapeType=Capsule, length=2.25, radius=0.6, shapeName=dyn_HairBraid_bk_r_03_jnt_capsule, center=<2 0 0>, rotation=<rot 90 0 0>}]
 dyn_HairBraid_bk_r_04_jnt [{shapeType=Capsule, length=4, radius=0.6, shapeName=dyn_HairBraid_bk_r_04_jnt_capsule, center=<3 0 0>, rotation=<rot 90 0 0>}]
 dyn_HairBraid_ft_l_01_jnt [{shapeType=Capsule, length=2, radius=0.6, shapeName=dyn_HairBraid_ft_l_01_jnt_capsule, center=<2.5 0 0>, rotation=<rot 90 0 0>}]
 dyn_HairBraid_ft_l_02_jnt [{shapeType=Capsule, length=3, radius=0.6, shapeName=dyn_HairBraid_ft_l_02_jnt_capsule, center=<2.25 0 0>, rotation=<rot 90 0 0>}]
 dyn_HairBraid_ft_l_03_jnt [{shapeType=Capsule, length=2, radius=0.6, shapeName=dyn_HairBraid_ft_l_03_jnt_capsule, center=<2 0 0>, rotation=<rot 90 0 0>}]
 dyn_HairBraid_ft_l_04_jnt [{shapeType=Capsule, length=2, radius=0.6, shapeName=dyn_HairBraid_ft_l_04_jnt_capsule, center=<1.75 0 0>, rotation=<rot 90 0 0>}]
 dyn_HairBraid_ft_l_05_jnt [{shapeType=Capsule, length=4, radius=0.6, shapeName=dyn_HairBraid_ft_l_05_jnt_capsule, center=<3 0 0>, rotation=<rot 90 0 0>}]
 dyn_HairBraid_ft_r_01_jnt [{shapeType=Capsule, length=2, radius=0.6, shapeName=dyn_HairBraid_ft_r_01_jnt_capsule, center=<2.5 0 0>, rotation=<rot 90 0 0>}]
 dyn_HairBraid_ft_r_02_jnt [{shapeType=Capsule, length=2.25, radius=0.6, shapeName=dyn_HairBraid_ft_r_02_jnt_capsule, center=<2 0 0>, rotation=<rot 90 0 0>}]
 dyn_HairBraid_ft_r_03_jnt [{shapeType=Capsule, length=2.25, radius=0.6, shapeName=dyn_HairBraid_ft_r_03_jnt_capsule, center=<2 0 0>, rotation=<rot 90 0 0>}]
 dyn_HairBraid_ft_r_04_jnt [{shapeType=Capsule, length=4, radius=0.6, shapeName=dyn_HairBraid_ft_r_04_jnt_capsule, center=<3 0 0>, rotation=<rot 90 0 0>}]
 dyn_HairBraid_l_01_jnt [{shapeType=Capsule, length=3, radius=1, shapeName=dyn_HairBraid_l_01_jnt_capsule, center=<3 0 0>, rotation=<rot 90 0 0>}]
 dyn_HairBraid_l_02_jnt [{shapeType=Capsule, length=3, radius=1, shapeName=dyn_HairBraid_l_02_jnt_capsule, center=<3 0 0>, rotation=<rot 90 0 0>}]
 dyn_HairBraid_l_MSTR [{shapeType=Capsule, length=0, radius=0, shapeName=dyn_HairBraid_l_MSTR_capsule}]
 dyn_HairBraid_r_01_jnt [{shapeType=Capsule, length=3, radius=1, shapeName=dyn_HairBraid_r_01_jnt_capsule, center=<3 0 0>, rotation=<rot 90 0 0>}]
 dyn_HairBraid_r_02_jnt [{shapeType=Capsule, length=2, radius=1, shapeName=dyn_HairBraid_r_02_jnt_capsule, center=<2.5 0 0>, rotation=<rot 90 0 0>}]
 dyn_HairBraid_r_03_jnt [{shapeType=Capsule, length=0.5, radius=1, shapeName=dyn_HairBraid_r_03_jnt_capsule, center=<1.5 0 0>, rotation=<rot 90 0 0>}]
 dyn_HairBraid_r_MSTR [{shapeType=Capsule, length=0, radius=0, shapeName=dyn_HairBraid_r_MSTR_capsule}]
 dyn_Hair_bk_c_01_jnt [{shapeType=Capsule, length=6, radius=3.5, shapeName=dyn_Hair_bk_c_01_jnt_capsule, center=<4 0 -2>, rotation=<rot 0 0 90>}]
 dyn_Hair_bk_c_02_jnt [{shapeType=Capsule, length=3, radius=3.5, shapeName=dyn_Hair_bk_c_02_jnt_capsule, center=<4 0 -1>, rotation=<rot 0 0 90>}]
 dyn_Hair_bk_c_03_jnt [{shapeType=Capsule, length=3, radius=4, shapeName=dyn_Hair_bk_c_03_jnt_capsule, center=<4 0 -1>, rotation=<rot 0 0 90>}]
 dyn_Hair_bk_c_04_jnt [{shapeType=Capsule, length=3, radius=4, shapeName=dyn_Hair_bk_c_04_jnt_capsule, center=<4 0 -1>, rotation=<rot -0.0002 0 90>}]
 dyn_Hair_bk_c_05_jnt [{shapeType=Capsule, length=4, radius=5, shapeName=dyn_Hair_bk_c_05_jnt_capsule, center=<8 0 1>, rotation=<rot 90 0 0>}]
 dyn_Hair_bk_c_MSTR [{shapeType=Capsule, length=0, radius=0, shapeName=dyn_Hair_bk_c_MSTR_capsule}]
 dyn_Hair_bk_l_01_jnt [{shapeType=Capsule, length=6, radius=3.5, shapeName=dyn_Hair_bk_l_01_jnt_capsule, center=<4 -3 0>, rotation=<rot -0.0002 0 110>}]
 dyn_Hair_bk_l_02_jnt [{shapeType=Capsule, length=6, radius=3.5, shapeName=dyn_Hair_bk_l_02_jnt_capsule, center=<4 -2 0>, rotation=<rot -0.0002 0 110>}]
 dyn_Hair_bk_l_03_jnt [{shapeType=Capsule, length=4, radius=3.5, shapeName=dyn_Hair_bk_l_03_jnt_capsule, center=<4 -1 1>, rotation=<rot -1.7283 9.8511 109.851>}]
 dyn_Hair_bk_l_04_jnt [{shapeType=Capsule, length=4, radius=3.5, shapeName=dyn_Hair_bk_l_04_jnt_capsule, center=<4 0 2>, rotation=<rot -6.7182 18.8817 108.8816>}]
 dyn_Hair_bk_l_05_jnt [{shapeType=Capsule, length=4, radius=5, shapeName=dyn_Hair_bk_l_05_jnt_capsule, center=<8 0 3>, rotation=<rot 90 0 0>}]
 dyn_Hair_bk_l_MSTR [{shapeType=Capsule, length=0, radius=0, shapeName=dyn_Hair_bk_l_MSTR_capsule}]
 dyn_Hair_bk_r_01_jnt [{shapeType=Capsule, length=6, radius=3.5, shapeName=dyn_Hair_bk_r_01_jnt_capsule, center=<4 3 0>, rotation=<rot -0.0002 0 70>}]
 dyn_Hair_bk_r_02_jnt [{shapeType=Capsule, length=6, radius=3.5, shapeName=dyn_Hair_bk_r_02_jnt_capsule, center=<4 2 0>, rotation=<rot -0.0002 0 70>}]
 dyn_Hair_bk_r_03_jnt [{shapeType=Capsule, length=4, radius=3.5, shapeName=dyn_Hair_bk_r_03_jnt_capsule, center=<4 1 1>, rotation=<rot -1.7284 -9.8511 70.149>}]
 dyn_Hair_bk_r_04_jnt [{shapeType=Capsule, length=4, radius=3.5, shapeName=dyn_Hair_bk_r_04_jnt_capsule, center=<4 0 2>, rotation=<rot -6.7183 -18.8817 71.1185>}]
 dyn_Hair_bk_r_05_jnt [{shapeType=Capsule, length=4, radius=5, shapeName=dyn_Hair_bk_r_05_jnt_capsule, center=<8 0 3>, rotation=<rot 90 0 0>}]
 dyn_Hair_bk_r_MSTR [{shapeType=Capsule, length=0, radius=0, shapeName=dyn_Hair_bk_r_MSTR_capsule}]
 head [{shapeType=Capsule, length=8, radius=6.5, shapeName=head_capsule, center=<2.5 3 0>, rotation=<rot 90 1.9744 1.9744>}]
 neck_01 [{shapeType=Capsule, length=8, radius=4.5, shapeName=neck_01_capsule, center=<3.6683 -0.2876 0>, rotation=<rot 90 0 0>}]
 spine_03 [{shapeType=Capsule, length=15, radius=10, shapeName=spine_04_capsule, center=<0.229 2 0>, rotation=<rot 90 0 10.0001>}]
 spine_05 [
  {shapeType=Capsule, length=12, radius=9.25, shapeName=spine_05_capsule, center=<-2.5963 0.209 0>}
  {shapeType=Capsule, length=12, radius=7.75, shapeName=spine_05_capsule, center=<-4.8394 3.7445 0>}
  {shapeType=Capsule, length=13, radius=2.5, shapeName=neck_01_capsule, center=<8 4.5 -4>, rotation=<rot 0 52 87>}
  {shapeType=Capsule, length=13, radius=2.5, shapeName=neck_01_capsule, center=<8 4.5 4>, rotation=<rot 0 52 -87>}
 ]
 upperarm_l [{shapeType=Capsule, length=20, radius=5.5, shapeName=upperarm_l_capsule, center=<12 -0.5 0>, rotation=<rot 90 0 0>}]
 upperarm_r [{shapeType=Capsule, length=20, radius=5.5, shapeName=upperarm_r_capsule, center=<-12 0.5 0>, rotation=<rot 90 0 0>}]
}
constraints [
 {bone1Name=dyn_HairBraid_l_MSTR, bone2Name=head, swing1LimitDegrees=35, swing2LimitDegrees=35, twistLimitDegrees=2}
 {bone1Name=dyn_HairBraid_r_MSTR, bone2Name=head, swing1LimitDegrees=35, swing2LimitDegrees=35, twistLimitDegrees=2}
 {
  bone1Name=dyn_Hair_bk_MSTR
  bone2Name=head
  swing1LimitDegrees=45
  swing1Motion=Free
  swing2LimitDegrees=45
  swing2Motion=Free
  twistLimitDegrees=45
  twistMotion=Free
 }
 {
  bone1Name=dyn_Hair_bk_l_MSTR
  bone2Name=head
  swing1LimitDegrees=45
  swing1Motion=Free
  swing2LimitDegrees=45
  swing2Motion=Free
  twistLimitDegrees=45
  twistMotion=Free
 }
 {
  bone1Name=dyn_Hair_bk_r_MSTR
  bone2Name=head
  swing1LimitDegrees=45
  swing1Motion=Free
  swing2LimitDegrees=45
  swing2Motion=Free
  twistLimitDegrees=45
  twistMotion=Free
 }
 {
  bone1Name=head
  bone2Name=spine_05
  swing1LimitDegrees=45
  swing1Motion=Free
  swing2LimitDegrees=45
  swing2Motion=Free
  twistLimitDegrees=45
  twistMotion=Free
 }
 {bone1Name=dyn_Flower_01_l, bone2Name=head, swing1LimitDegrees=25, swing2LimitDegrees=25, twistLimitDegrees=2}
 {bone1Name=dyn_HairBraid_l_01_jnt, bone2Name=dyn_HairBraid_l_MSTR, swing1LimitDegrees=25, swing2LimitDegrees=15, twistLimitDegrees=2}
 {bone1Name=dyn_HairBraid_l_02_jnt, bone2Name=dyn_HairBraid_l_01_jnt, swing1LimitDegrees=25, swing2LimitDegrees=15, twistLimitDegrees=2}
 {bone1Name=dyn_HairBraid_bk_l_01_jnt, bone2Name=dyn_HairBraid_l_02_jnt, swing1LimitDegrees=25, swing2LimitDegrees=15, twistLimitDegrees=2}
 {bone1Name=dyn_HairBraid_bk_l_02_jnt, bone2Name=dyn_HairBraid_bk_l_01_jnt, swing1LimitDegrees=25, swing2LimitDegrees=15, twistLimitDegrees=2}
 {bone1Name=dyn_HairBraid_bk_l_03_jnt, bone2Name=dyn_HairBraid_bk_l_02_jnt, swing1LimitDegrees=25, swing2LimitDegrees=15, twistLimitDegrees=2}
 {bone1Name=dyn_HairBraid_bk_l_04_jnt, bone2Name=dyn_HairBraid_bk_l_03_jnt, swing1LimitDegrees=25, swing2LimitDegrees=15, twistLimitDegrees=2}
 {bone1Name=dyn_HairBraid_bk_l_05_jnt, bone2Name=dyn_HairBraid_bk_l_04_jnt, swing1LimitDegrees=25, swing2LimitDegrees=15, twistLimitDegrees=2}
 {bone1Name=dyn_HairBraid_ft_l_01_jnt, bone2Name=dyn_HairBraid_l_02_jnt, swing1LimitDegrees=25, swing2LimitDegrees=15, twistLimitDegrees=2}
 {bone1Name=dyn_HairBraid_ft_l_02_jnt, bone2Name=dyn_HairBraid_ft_l_01_jnt, swing1LimitDegrees=25, swing2LimitDegrees=15, twistLimitDegrees=2}
 {bone1Name=dyn_HairBraid_ft_l_03_jnt, bone2Name=dyn_HairBraid_ft_l_02_jnt, swing1LimitDegrees=25, swing2LimitDegrees=15, twistLimitDegrees=2}
 {bone1Name=dyn_HairBraid_ft_l_04_jnt, bone2Name=dyn_HairBraid_ft_l_03_jnt, swing1LimitDegrees=25, swing2LimitDegrees=15, twistLimitDegrees=2}
 {bone1Name=dyn_HairBraid_ft_l_05_jnt, bone2Name=dyn_HairBraid_ft_l_04_jnt, swing1LimitDegrees=25, swing2LimitDegrees=15, twistLimitDegrees=2}
 {bone1Name=dyn_HairBraid_r_01_jnt, bone2Name=dyn_HairBraid_r_MSTR, swing1LimitDegrees=8, swing2LimitDegrees=8, twistLimitDegrees=2}
 {bone1Name=dyn_HairBraid_r_02_jnt, bone2Name=dyn_HairBraid_r_01_jnt, swing1LimitDegrees=15, swing2LimitDegrees=15, twistLimitDegrees=2}
 {bone1Name=dyn_HairBraid_r_03_jnt, bone2Name=dyn_HairBraid_r_02_jnt, swing1LimitDegrees=15, swing2LimitDegrees=15, twistLimitDegrees=2}
 {bone1Name=dyn_HairBraid_bk_r_01_jnt, bone2Name=dyn_HairBraid_r_03_jnt, swing1LimitDegrees=25, swing2LimitDegrees=25, twistLimitDegrees=2}
 {bone1Name=dyn_HairBraid_bk_r_02_jnt, bone2Name=dyn_HairBraid_bk_r_01_jnt, swing1LimitDegrees=25, swing2LimitDegrees=25, twistLimitDegrees=2}
 {bone1Name=dyn_HairBraid_bk_r_03_jnt, bone2Name=dyn_HairBraid_bk_r_02_jnt, swing1LimitDegrees=25, swing2LimitDegrees=25, twistLimitDegrees=2}
 {bone1Name=dyn_HairBraid_bk_r_04_jnt, bone2Name=dyn_HairBraid_bk_r_03_jnt, swing1LimitDegrees=25, swing2LimitDegrees=25, twistLimitDegrees=2}
 {bone1Name=dyn_HairBraid_ft_r_01_jnt, bone2Name=dyn_HairBraid_r_03_jnt, swing1LimitDegrees=25, swing2LimitDegrees=25, twistLimitDegrees=2}
 {bone1Name=dyn_HairBraid_ft_r_02_jnt, bone2Name=dyn_HairBraid_ft_r_01_jnt, swing1LimitDegrees=25, swing2LimitDegrees=25, twistLimitDegrees=2}
 {bone1Name=dyn_HairBraid_ft_r_03_jnt, bone2Name=dyn_HairBraid_ft_r_02_jnt, swing1LimitDegrees=25, swing2LimitDegrees=25, twistLimitDegrees=2}
 {bone1Name=dyn_HairBraid_ft_r_04_jnt, bone2Name=dyn_HairBraid_ft_r_03_jnt, swing1LimitDegrees=25, swing2LimitDegrees=25, twistLimitDegrees=2}
 {
  bone1Name=dyn_Hair_bk_c_MSTR
  bone2Name=head
  swing1LimitDegrees=45
  swing1Motion=Free
  swing2LimitDegrees=45
  swing2Motion=Free
  twistLimitDegrees=45
  twistMotion=Free
 }
 {bone1Name=dyn_Hair_bk_c_01_jnt, bone2Name=dyn_Hair_bk_c_MSTR, swing1LimitDegrees=2, swing2LimitDegrees=35, twistLimitDegrees=2}
 {bone1Name=dyn_Hair_bk_c_02_jnt, bone2Name=dyn_Hair_bk_c_01_jnt, swing1LimitDegrees=2, swing2LimitDegrees=24, twistLimitDegrees=2}
 {bone1Name=dyn_Hair_bk_c_03_jnt, bone2Name=dyn_Hair_bk_c_02_jnt, swing1LimitDegrees=2, swing2LimitDegrees=24, twistLimitDegrees=2}
 {bone1Name=dyn_Hair_bk_c_04_jnt, bone2Name=dyn_Hair_bk_c_03_jnt, swing1LimitDegrees=2, swing2LimitDegrees=35, twistLimitDegrees=2}
 {bone1Name=dyn_Hair_bk_c_05_jnt, bone2Name=dyn_Hair_bk_c_04_jnt, swing1LimitDegrees=2, swing2LimitDegrees=35, twistLimitDegrees=2}
 {bone1Name=dyn_Hair_bk_l_01_jnt, bone2Name=dyn_Hair_bk_l_MSTR, swing1LimitDegrees=2, swing2LimitDegrees=35, twistLimitDegrees=2}
 {bone1Name=dyn_Hair_bk_l_02_jnt, bone2Name=dyn_Hair_bk_l_01_jnt, swing1LimitDegrees=2, swing2LimitDegrees=24, twistLimitDegrees=2}
 {bone1Name=dyn_Hair_bk_l_03_jnt, bone2Name=dyn_Hair_bk_l_02_jnt, swing1LimitDegrees=2, swing2LimitDegrees=24, twistLimitDegrees=2}
 {bone1Name=dyn_Hair_bk_l_04_jnt, bone2Name=dyn_Hair_bk_l_03_jnt, swing1LimitDegrees=2, swing2LimitDegrees=35, twistLimitDegrees=2}
 {bone1Name=dyn_Hair_bk_l_05_jnt, bone2Name=dyn_Hair_bk_l_04_jnt, swing1LimitDegrees=2, swing2LimitDegrees=35, twistLimitDegrees=2}
 {bone1Name=dyn_Hair_bk_r_01_jnt, bone2Name=dyn_Hair_bk_r_MSTR, swing1LimitDegrees=2, swing2LimitDegrees=35, twistLimitDegrees=2}
 {bone1Name=dyn_Hair_bk_r_02_jnt, bone2Name=dyn_Hair_bk_r_01_jnt, swing1LimitDegrees=2, swing2LimitDegrees=24, twistLimitDegrees=2}
 {bone1Name=dyn_Hair_bk_r_03_jnt, bone2Name=dyn_Hair_bk_r_02_jnt, swing1LimitDegrees=2, swing2LimitDegrees=24, twistLimitDegrees=2}
 {bone1Name=dyn_Hair_bk_r_04_jnt, bone2Name=dyn_Hair_bk_r_03_jnt, swing1LimitDegrees=2, swing2LimitDegrees=35, twistLimitDegrees=2}
 {bone1Name=dyn_Hair_bk_r_05_jnt, bone2Name=dyn_Hair_bk_r_04_jnt, swing1LimitDegrees=2, swing2LimitDegrees=35, twistLimitDegrees=2}
 {
  bone1Name=dyn_Hair_bk_r_01_jnt
  bone2Name=dyn_Hair_bk_c_01_jnt
  swing1LimitDegrees=45
  swing1Motion=Free
  swing2LimitDegrees=45
  swing2Motion=Free
  twistLimitDegrees=45
  twistMotion=Free
 }
 {
  bone1Name=dyn_Hair_bk_l_01_jnt
  bone2Name=dyn_Hair_bk_c_01_jnt
  swing1LimitDegrees=45
  swing1Motion=Free
  swing2LimitDegrees=45
  swing2Motion=Free
  twistLimitDegrees=45
  twistMotion=Free
 }
]
