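# Animation Blueprint for the SilverBellWind face accessory part, adding secondary motion to hair and earrings over the pose copied from the parent mesh.
# A Control Rig takes pistol/rifle alphas derived from the player's weapon type.
# A backbling check switches back hair from AnimDynamics and RigidBody physics to weighted bone constraints with hair-pin RigidBody when worn.
# Shared AnimDynamics chains then spring earrings, bangs and strands; EventGraph also feeds gravity override to RigidBody.
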AnimBlueprint /BRCosmetics/Characters/Player/Female/Medium/Bodies/F_MED_SilverBellWind/Meshes/Parts/F_MED_SilverBellWind_FaceAcc_AnimBP
target_skeleton=&/F_MED_SilverBellWind_FaceAcc_Skeleton
graphs [
 AnimGraph {nodes [
   n1 = Control Rig!(
    Source=Copy Pose From Mesh() {bUseAttachedParent=true}
    Alpha
    Is Pistol Targeting:double=Get Is Pistol Alpha() -> (Is Pistol Alpha:float)
    Is Rifle Targeting:double=Get Is Rifle Alpha() -> (Is Rifle Alpha:float)
   ) -> (Pose) {
    alphaBoolBlend {blendOption=Linear}
    controlRigAssetReference {blueprintRigClass=&/F_MED_SilverBellWind_FaceAcc_CtrlRig.F_MED_SilverBellWind_FaceAcc_CtrlRig_C}
   }
   n2 = "Transform (Modify) Bone - Bone: dyn_HairBraid_l_MSTR"(ComponentPose=Local To Component(LocalPose=@n1), Alpha=1, Translation="0,0,0", Rotation="10,0,0", Scale="1,1,1") {
    rotationMode=BMM_Additive
    rotationSpace=BCS_BoneSpace
    translationMode=BMM_Additive
    translationSpace=BCS_BoneSpace
    alphaBoolBlend {blendOption=Linear}
    boneToModify {boneName=dyn_HairBraid_l_MSTR}
    rotation {pitch=10}
   }
   n3 = "Transform (Modify) Bone - Bone: dyn_HairBraid_r_MSTR"(ComponentPose=@n2, Alpha=1, Translation="0,0,0", Rotation="10,0,0", Scale="1,1,1") {
    rotationMode=BMM_Additive
    rotationSpace=BCS_BoneSpace
    translationMode=BMM_Additive
    translationSpace=BCS_BoneSpace
    alphaBoolBlend {blendOption=Linear}
    boneToModify {boneName=dyn_HairBraid_r_MSTR}
    rotation {pitch=10}
   }
   n4 = Anim Dynamics - Bone: dyn_Hair_bk_c_01_jnt!(ComponentPose=@n3, Alpha=0.5, GravityOverride:Vector="0,0,0", ExternalForce="0,0,0") -> (Pose) {
    angularDampingOverride=1
    angularSpringConstant=50
    bAngularSpring=true
    bOverrideAngularDamping=true
    bUseGravityOverride=true
    simulationSpace=RootRelative
    alphaBoolBlend {blendOption=Linear}
    boundBone {boneName=dyn_Hair_bk_c_01_jnt}
    componentAppliedLinearAccClamp=<6000 6000 6000>
    componentLinearAccScale=<0.3 0.3 0.3>
    componentLinearVelScale=<0.3 0.3 0.3>
    physicsBodyDefinitions [{
      boundBone {boneName=dyn_Hair_bk_c_01_jnt}
      boxExtents=<20 20 20>
      constraintSetup {angularLimitsMax=<1 1 5>, angularLimitsMin=<-1 -30 -5>, angularTarget=<1 0 0>}
      localJointOffset=<8 0 0>
     }]
   }
   n5 = Anim Dynamics - Bone: dyn_Hair_bk_l_01_jnt!(ComponentPose=@n4, Alpha=0.5, GravityOverride:Vector="0,0,0", ExternalForce="0,0,0") -> (Pose) {
    angularDampingOverride=1
    angularSpringConstant=50
    bAngularSpring=true
    bOverrideAngularDamping=true
    bUseGravityOverride=true
    simulationSpace=RootRelative
    alphaBoolBlend {blendOption=Linear}
    boundBone {boneName=dyn_Hair_bk_l_01_jnt}
    componentAppliedLinearAccClamp=<6000 6000 6000>
    componentLinearAccScale=<0.3 0.3 0.3>
    componentLinearVelScale=<0.3 0.3 0.3>
    physicsBodyDefinitions [{
      boundBone {boneName=dyn_Hair_bk_l_01_jnt}
      boxExtents=<20 20 20>
      constraintSetup {angularLimitsMax=<1 1 5>, angularLimitsMin=<-1 -30 -5>, angularTarget=<1 0 0>}
      localJointOffset=<8 0 0>
     }]
   }
   n6 = Anim Dynamics - Bone: dyn_Hair_bk_r_01_jnt!(ComponentPose=@n5, Alpha=0.5, GravityOverride:Vector="0,0,0", ExternalForce="0,0,0") -> (Pose) {
    angularDampingOverride=1
    angularSpringConstant=50
    bAngularSpring=true
    bOverrideAngularDamping=true
    bUseGravityOverride=true
    simulationSpace=RootRelative
    alphaBoolBlend {blendOption=Linear}
    boundBone {boneName=dyn_Hair_bk_r_01_jnt}
    componentAppliedLinearAccClamp=<6000 6000 6000>
    componentLinearAccScale=<0.3 0.3 0.3>
    componentLinearVelScale=<0.3 0.3 0.3>
    physicsBodyDefinitions [{
      boundBone {boneName=dyn_Hair_bk_r_01_jnt}
      boxExtents=<20 20 20>
      constraintSetup {angularLimitsMax=<1 1 5>, angularLimitsMin=<-1 -30 -5>, angularTarget=<1 0 0>}
      localJointOffset=<8 0 0>
     }]
   }
   n7 = Anim Dynamics - Bone: dyn_Hair_bk_c_02_jnt!(ComponentPose=@n6, Alpha=0.5, GravityOverride:Vector="0,0,0", ExternalForce="0,0,0") -> (Pose) {
    angularDampingOverride=1
    angularSpringConstant=50
    bAngularSpring=true
    bOverrideAngularDamping=true
    bUseGravityOverride=true
    simulationSpace=RootRelative
    alphaBoolBlend {blendOption=Linear}
    boundBone {boneName=dyn_Hair_bk_c_02_jnt}
    componentAppliedLinearAccClamp=<6000 6000 6000>
    componentLinearAccScale=<0.3 0.3 0.3>
    componentLinearVelScale=<0.3 0.3 0.3>
    physicsBodyDefinitions [{
      boundBone {boneName=dyn_Hair_bk_c_02_jnt}
      boxExtents=<20 20 20>
      constraintSetup {angularLimitsMax=<1 1 5>, angularLimitsMin=<-1 -30 -5>, angularTarget=<1 0 0>}
      localJointOffset=<8 0 0>
     }]
   }
   n8 = Anim Dynamics - Bone: dyn_Hair_bk_l_02_jnt!(ComponentPose=@n7, Alpha=0.5, GravityOverride:Vector="0,0,0", ExternalForce="0,0,0") -> (Pose) {
    angularDampingOverride=1
    angularSpringConstant=50
    bAngularSpring=true
    bOverrideAngularDamping=true
    bUseGravityOverride=true
    simulationSpace=RootRelative
    alphaBoolBlend {blendOption=Linear}
    boundBone {boneName=dyn_Hair_bk_l_02_jnt}
    componentAppliedLinearAccClamp=<6000 6000 6000>
    componentLinearAccScale=<0.3 0.3 0.3>
    componentLinearVelScale=<0.3 0.3 0.3>
    physicsBodyDefinitions [{
      boundBone {boneName=dyn_Hair_bk_l_02_jnt}
      boxExtents=<20 20 20>
      constraintSetup {angularLimitsMax=<1 1 5>, angularLimitsMin=<-1 -30 -5>, angularTarget=<1 0 0>}
      localJointOffset=<8 0 0>
     }]
   }
   n9 = Anim Dynamics - Bone: dyn_Hair_bk_r_02_jnt!(ComponentPose=@n8, Alpha=0.5, GravityOverride:Vector="0,0,0", ExternalForce="0,0,0") -> (Pose) {
    angularDampingOverride=1
    angularSpringConstant=50
    bAngularSpring=true
    bOverrideAngularDamping=true
    bUseGravityOverride=true
    simulationSpace=RootRelative
    alphaBoolBlend {blendOption=Linear}
    boundBone {boneName=dyn_Hair_bk_r_02_jnt}
    componentAppliedLinearAccClamp=<6000 6000 6000>
    componentLinearAccScale=<0.3 0.3 0.3>
    componentLinearVelScale=<0.3 0.3 0.3>
    physicsBodyDefinitions [{
      boundBone {boneName=dyn_Hair_bk_r_02_jnt}
      boxExtents=<20 20 20>
      constraintSetup {angularLimitsMax=<1 1 5>, angularLimitsMin=<-1 -30 -5>, angularTarget=<1 0 0>}
      localJointOffset=<8 0 0>
     }]
   }
   n10 = RigidBody!(ComponentPose=@n9, Alpha, OverrideWorldGravity:Vector=Get Gravity Override(), ExternalForce) -> (Pose) {
    bClampLinearTranslationLimitToRefPose=true
    bOverrideWorldGravity=true
    overridePhysicsAsset=&/F_MED_SilverBellWind_FaceAcc_Physics
    simulationSpace=BaseBoneSpace
    alphaBoolBlend {blendOption=Linear}
    alphaScaleBiasClamp {bInterpResult=true, interpSpeedIncreasing=2}
    baseBoneRef {boneName=neck_01}
    componentAppliedLinearAccClamp=<6000 6000 6000>
    componentLinearAccScale=<0.4 0.4 0.3>
    componentLinearVelScale=<0.4 0.4 0.3>
    simSpaceSettings {
     maxAngularAcceleration=1000
     maxAngularVelocity=3
     maxLinearAcceleration=700
     maxLinearVelocity=500
     velocityScaleZ=0.5
     worldAlpha=0.5
     externalLinearDragV=<0.2 0.2>
    }
   }
   n11 = Component To Local(ComponentPose=@n10)
   n12 = Control Rig(Source=Copy Pose From Mesh() {bUseAttachedParent=true}, Is Pistol Targeting=Get Is Pistol Alpha(), Is Rifle Targeting=Get Is Rifle Alpha()) {
    alphaBoolBlend {blendOption=Linear}
    controlRigAssetReference {blueprintRigClass=&F_MED_SilverBellWind_FaceAcc_CtrlRig_C}
   }
   n13 = "Transform (Modify) Bone - Bone: dyn_HairBraid_l_MSTR"(ComponentPose=Local To Component(LocalPose=@n12), Alpha=1, Translation="0,0,0", Rotation="0,0,0", Scale="1,1,1") {
    rotationMode=BMM_Additive
    rotationSpace=BCS_BoneSpace
    translationMode=BMM_Additive
    translationSpace=BCS_BoneSpace
    alphaBoolBlend {blendOption=Linear}
    boneToModify {boneName=dyn_HairBraid_l_MSTR}
   }
   n14 = "Transform (Modify) Bone - Bone: dyn_HairBraid_r_MSTR"(ComponentPose=@n13, Alpha=1, Translation="0,0,0", Rotation="0,0,0", Scale="1,1,1") {
    rotationMode=BMM_Additive
    rotationSpace=BCS_BoneSpace
    translationMode=BMM_Additive
    translationSpace=BCS_BoneSpace
    alphaBoolBlend {blendOption=Linear}
    boneToModify {boneName=dyn_HairBraid_r_MSTR}
   }
   n15 = "Constraint - Hair_MSTR (PP)"!(ComponentPose=@n14, Alpha=1, ConstraintWeights_0=0.8, ConstraintWeights_1:float=0.2) -> (Pose) {
    alphaBoolBlend {blendOption=Linear}
    boneToModify {boneName=Hair_MSTR}
    constraintSetup [{targetBone {boneName=spine_05}}, {targetBone {boneName=head}}]
    constraintWeights [1,1]
   }
   n16 = "Constraint - dyn_Hair_bk_c_01_jnt (PP)"!(ComponentPose=@n15, Alpha=1, ConstraintWeights_0=0, ConstraintWeights_1:float=1) -> (Pose) {
    alphaBoolBlend {blendOption=Linear}
    boneToModify {boneName=dyn_Hair_bk_c_01_jnt}
    constraintSetup [{targetBone {boneName=neck_02}}, {targetBone {boneName=head}}]
    constraintWeights [1,1]
   }
   n17 = "Constraint - dyn_Hair_bk_l_01_jnt (PP)"!(ComponentPose=@n16, Alpha=1, ConstraintWeights_0=0, ConstraintWeights_1:float=1) -> (Pose) {
    alphaBoolBlend {blendOption=Linear}
    boneToModify {boneName=dyn_Hair_bk_l_01_jnt}
    constraintSetup [{targetBone {boneName=neck_02}}, {targetBone {boneName=head}}]
    constraintWeights [1,1]
   }
   n18 = "Constraint - dyn_Hair_bk_r_01_jnt (PP)"!(ComponentPose=@n17, Alpha=1, ConstraintWeights_0=0, ConstraintWeights_1:float=1) -> (Pose) {
    alphaBoolBlend {blendOption=Linear}
    boneToModify {boneName=dyn_Hair_bk_r_01_jnt}
    constraintSetup [{targetBone {boneName=neck_02}}, {targetBone {boneName=head}}]
    constraintWeights [1,1]
   }
   n19 = "Constraint - dyn_Hair_bk_c_02_jnt (PP)"!(ComponentPose=@n18, Alpha=1, ConstraintWeights_0=0.6, ConstraintWeights_1:float=0.4) -> (Pose) {
    alphaBoolBlend {blendOption=Linear}
    boneToModify {boneName=dyn_Hair_bk_c_02_jnt}
    constraintSetup [{targetBone {boneName=head}}, {targetBone {boneName=neck_01}}]
    constraintWeights [1,1]
   }
   n20 = "Constraint - dyn_Hair_bk_l_02_jnt (PP)"!(ComponentPose=@n19, Alpha=1, ConstraintWeights_0=0.6, ConstraintWeights_1:float=0.4) -> (Pose) {
    alphaBoolBlend {blendOption=Linear}
    boneToModify {boneName=dyn_Hair_bk_l_02_jnt}
    constraintSetup [{targetBone {boneName=head}}, {targetBone {boneName=neck_01}}]
    constraintWeights [1,1]
   }
   n21 = "Constraint - dyn_Hair_bk_r_02_jnt (PP)"!(ComponentPose=@n20, Alpha=1, ConstraintWeights_0=0.6, ConstraintWeights_1:float=0.4) -> (Pose) {
    alphaBoolBlend {blendOption=Linear}
    boneToModify {boneName=dyn_Hair_bk_r_02_jnt}
    constraintSetup [{targetBone {boneName=head}}, {targetBone {boneName=neck_01}}]
    constraintWeights [1,1]
   }
   n22 = "Constraint - dyn_Hair_bk_c_03_jnt (PP)"!(ComponentPose=@n21, Alpha=1, ConstraintWeights_0=0.4, ConstraintWeights_1:float=0.6) -> (Pose) {
    alphaBoolBlend {blendOption=Linear}
    boneToModify {boneName=dyn_Hair_bk_c_03_jnt}
    constraintSetup [{targetBone {boneName=spine_05}}, {targetBone {boneName=neck_01}}]
    constraintWeights [1,1]
   }
   n23 = "Constraint - dyn_Hair_bk_l_03_jnt (PP)"!(ComponentPose=@n22, Alpha=1, ConstraintWeights_0=0.4, ConstraintWeights_1:float=0.6) -> (Pose) {
    alphaBoolBlend {blendOption=Linear}
    boneToModify {boneName=dyn_Hair_bk_l_03_jnt}
    constraintSetup [{targetBone {boneName=clavicle_l}}, {targetBone {boneName=neck_01}}]
    constraintWeights [1,1]
   }
   n24 = "Constraint - dyn_Hair_bk_r_03_jnt (PP)"!(ComponentPose=@n23, Alpha=1, ConstraintWeights_0=0.4, ConstraintWeights_1:float=0.6) -> (Pose) {
    alphaBoolBlend {blendOption=Linear}
    boneToModify {boneName=dyn_Hair_bk_r_03_jnt}
    constraintSetup [{targetBone {boneName=clavicle_r}}, {targetBone {boneName=neck_01}}]
    constraintWeights [1,1]
   }
   n25 = "Transform (Modify) Bone - Bone: dyn_Hair_bk_c_03_jnt"(ComponentPose=@n24, Alpha=1, Translation="-1,0,0", Rotation="15,0,0", Scale="1,1,1") {
    rotationMode=BMM_Replace
    rotationSpace=BCS_BoneSpace
    translationMode=BMM_Additive
    translationSpace=BCS_BoneSpace
    alphaBoolBlend {blendOption=Linear}
    alphaScaleBiasClamp {bInterpResult=true}
    boneToModify {boneName=dyn_Hair_bk_c_03_jnt}
    rotation {pitch=15}
    translation {x=-1}
   }
   n26 = "Transform (Modify) Bone - Bone: dyn_Hair_bk_l_03_jnt"(ComponentPose=@n25, Alpha=1, Translation="-1,-0.5,0", Rotation="15,0,0", Scale="1,1,1") {
    rotationMode=BMM_Replace
    rotationSpace=BCS_BoneSpace
    translationMode=BMM_Additive
    translationSpace=BCS_BoneSpace
    alphaBoolBlend {blendOption=Linear}
    alphaScaleBiasClamp {bInterpResult=true}
    boneToModify {boneName=dyn_Hair_bk_l_03_jnt}
    rotation {pitch=15}
    translation=<-1 -0.5>
   }
   n27 = "Transform (Modify) Bone - Bone: dyn_Hair_bk_r_03_jnt"(ComponentPose=@n26, Alpha=1, Translation="-1,-0.5,0", Rotation="15,0,0", Scale="1,1,1") {
    rotationMode=BMM_Replace
    rotationSpace=BCS_BoneSpace
    translationMode=BMM_Additive
    translationSpace=BCS_BoneSpace
    alphaBoolBlend {blendOption=Linear}
    alphaScaleBiasClamp {bInterpResult=true}
    boneToModify {boneName=dyn_Hair_bk_r_03_jnt}
    rotation {pitch=15}
    translation=<-1 -0.5>
   }
   n28 = "Constraint - dyn_Hair_bk_c_04_jnt (PP)"!(ComponentPose=@n27, Alpha=1, ConstraintWeights_0=0.5, ConstraintWeights_1:float=0.5) -> (Pose) {
    alphaBoolBlend {blendOption=Linear}
    boneToModify {boneName=dyn_Hair_bk_c_04_jnt}
    constraintSetup [{targetBone {boneName=spine_04}}, {targetBone {boneName=spine_05}}]
    constraintWeights [1,1]
   }
   n29 = "Constraint - dyn_Hair_bk_l_04_jnt (PP)"!(ComponentPose=@n28, Alpha=1, ConstraintWeights_0=0.5, ConstraintWeights_1:float=0.5) -> (Pose) {
    alphaBoolBlend {blendOption=Linear}
    boneToModify {boneName=dyn_Hair_bk_l_04_jnt}
    constraintSetup [{targetBone {boneName=spine_04}}, {targetBone {boneName=spine_05}}]
    constraintWeights [1,1]
   }
   n30 = "Constraint - dyn_Hair_bk_r_04_jnt (PP)"!(ComponentPose=@n29, Alpha=1, ConstraintWeights_0=0.5, ConstraintWeights_1:float=0.5) -> (Pose) {
    alphaBoolBlend {blendOption=Linear}
    boneToModify {boneName=dyn_Hair_bk_r_04_jnt}
    constraintSetup [{targetBone {boneName=spine_04}}, {targetBone {boneName=spine_05}}]
    constraintWeights [1,1]
   }
   n31 = "Transform (Modify) Bone - Bone: dyn_Hair_bk_c_04_jnt"(ComponentPose=@n30, Alpha=1, Translation="0,0,0", Rotation="-10,0,0", Scale="1,1,1") {
    rotationMode=BMM_Replace
    rotationSpace=BCS_BoneSpace
    translationMode=BMM_Additive
    translationSpace=BCS_BoneSpace
    alphaBoolBlend {blendOption=Linear}
    alphaScaleBiasClamp {bInterpResult=true}
    boneToModify {boneName=dyn_Hair_bk_c_04_jnt}
    rotation {pitch=-10}
   }
   n32 = "Transform (Modify) Bone - Bone: dyn_Hair_bk_l_04_jnt"(ComponentPose=@n31, Alpha=1, Translation="0,0,0", Rotation="-10,0,0", Scale="1,1,1") {
    rotationMode=BMM_Replace
    rotationSpace=BCS_BoneSpace
    translationMode=BMM_Additive
    translationSpace=BCS_BoneSpace
    alphaBoolBlend {blendOption=Linear}
    alphaScaleBiasClamp {bInterpResult=true}
    boneToModify {boneName=dyn_Hair_bk_l_04_jnt}
    rotation {pitch=-10}
   }
   n33 = "Transform (Modify) Bone - Bone: dyn_Hair_bk_r_04_jnt"(ComponentPose=@n32, Alpha=1, Translation="0,0,0", Rotation="-10,0,0", Scale="1,1,1") {
    rotationMode=BMM_Replace
    rotationSpace=BCS_BoneSpace
    translationMode=BMM_Additive
    translationSpace=BCS_BoneSpace
    alphaBoolBlend {blendOption=Linear}
    alphaScaleBiasClamp {bInterpResult=true}
    boneToModify {boneName=dyn_Hair_bk_r_04_jnt}
    rotation {pitch=-10}
   }
   n34 = RigidBody(ComponentPose=@n33, OverrideWorldGravity=Get Gravity Override()) {
    bClampLinearTranslationLimitToRefPose=true
    bOverrideWorldGravity=true
    overridePhysicsAsset=&/F_MED_SilverBellWind_FaceAcc_HairPin_Physics
    simulationSpace=BaseBoneSpace
    alphaBoolBlend {blendOption=Linear}
    alphaScaleBiasClamp {bInterpResult=true, interpSpeedIncreasing=2}
    baseBoneRef {boneName=neck_02}
    componentAppliedLinearAccClamp=<6000 6000 6000>
    componentLinearAccScale=<0.4 0.4 0.3>
    componentLinearVelScale=<0.4 0.4 0.3>
    simSpaceSettings {
     maxAngularAcceleration=1000
     maxAngularVelocity=3
     maxLinearAcceleration=700
     maxLinearVelocity=500
     velocityScaleZ=0.5
     worldAlpha=0.5
     externalLinearDragV=<0.2 0.2>
    }
   }
   n35 = Component To Local(ComponentPose=@n34)
   n36 = Get Is Wearing Backbling()
   n37 = Anim Dynamics - Bone: dyn_Earring_l_01_jnt!(
    ComponentPose=Local To Component(LocalPose=Blend Poses by bool(BlendPose_0=@n35, BlendPose_1=@n11, bActiveValue=@n36) {blendType=Linear})
    Alpha=1
    GravityOverride:Vector="0,0,0"
    ExternalForce="0,0,0"
   ) -> (Pose) {
    angularDampingOverride=1
    angularSpringConstant=200
    bAngularSpring=true
    bOverrideAngularDamping=true
    bUseGravityOverride=true
    simulationSpace=RootRelative
    alphaBoolBlend {blendOption=Linear}
    boundBone {boneName=dyn_Earring_l_01_jnt}
    componentAppliedLinearAccClamp=<6000 6000 6000>
    componentLinearAccScale=<0.3 0.3 0.3>
    componentLinearVelScale=<0.3 0.3 0.3>
    physicsBodyDefinitions [{
      boundBone {boneName=dyn_Earring_l_01_jnt}
      boxExtents=<10 10 10>
      constraintSetup {angularLimitsMax=<1 30 30>, angularLimitsMin=<-1 -30 -30>, angularTarget=<1 0 0>}
      localJointOffset=<3 0 0>
     }]
   }
   n38 = Anim Dynamics - Bone: dyn_Earring_l_02_jnt!(ComponentPose=@n37, Alpha=1, GravityOverride:Vector="0,0,0", ExternalForce="0,0,0") -> (Pose) {
    angularDampingOverride=1
    angularSpringConstant=200
    bAngularSpring=true
    bOverrideAngularDamping=true
    bUseGravityOverride=true
    simulationSpace=RootRelative
    alphaBoolBlend {blendOption=Linear}
    boundBone {boneName=dyn_Earring_l_02_jnt}
    componentAppliedLinearAccClamp=<6000 6000 6000>
    componentLinearAccScale=<0.3 0.3 0.3>
    componentLinearVelScale=<0.3 0.3 0.3>
    physicsBodyDefinitions [{
      boundBone {boneName=dyn_Earring_l_02_jnt}
      boxExtents=<10 10 10>
      constraintSetup {angularLimitsMax=<1 30 30>, angularLimitsMin=<-1 -30 -30>, angularTarget=<1 0 0>}
      localJointOffset=<3 0 0>
     }]
   }
   n39 = Anim Dynamics - Bone: dyn_Earring_l_03_jnt!(ComponentPose=@n38, Alpha=1, GravityOverride:Vector="0,0,0", ExternalForce="0,0,0") -> (Pose) {
    angularDampingOverride=1
    angularSpringConstant=200
    bAngularSpring=true
    bOverrideAngularDamping=true
    bUseGravityOverride=true
    simulationSpace=RootRelative
    alphaBoolBlend {blendOption=Linear}
    boundBone {boneName=dyn_Earring_l_03_jnt}
    componentAppliedLinearAccClamp=<6000 6000 6000>
    componentLinearAccScale=<0.3 0.3 0.3>
    componentLinearVelScale=<0.3 0.3 0.3>
    physicsBodyDefinitions [{
      boundBone {boneName=dyn_Earring_l_03_jnt}
      boxExtents=<10 10 10>
      constraintSetup {angularLimitsMax=<1 30 30>, angularLimitsMin=<-1 -30 -30>, angularTarget=<1 0 0>}
      localJointOffset=<3 0 0>
     }]
   }
   n40 = Anim Dynamics - Bone: dyn_Earring_r_01_jnt!(ComponentPose=@n39, Alpha=1, GravityOverride:Vector="0,0,0", ExternalForce="0,0,0") -> (Pose) {
    angularDampingOverride=1
    angularSpringConstant=200
    bAngularSpring=true
    bOverrideAngularDamping=true
    bUseGravityOverride=true
    simulationSpace=RootRelative
    alphaBoolBlend {blendOption=Linear}
    boundBone {boneName=dyn_Earring_r_01_jnt}
    componentAppliedLinearAccClamp=<6000 6000 6000>
    componentLinearAccScale=<0.3 0.3 0.3>
    componentLinearVelScale=<0.3 0.3 0.3>
    physicsBodyDefinitions [{
      boundBone {boneName=dyn_Earring_r_01_jnt}
      boxExtents=<10 10 10>
      constraintSetup {angularLimitsMax=<1 30 30>, angularLimitsMin=<-1 -30 -30>, angularTarget=<1 0 0>}
      localJointOffset=<3 0 0>
     }]
   }
   n41 = Anim Dynamics - Bone: dyn_Earring_r_02_jnt!(ComponentPose=@n40, Alpha=1, GravityOverride:Vector="0,0,0", ExternalForce="0,0,0") -> (Pose) {
    angularDampingOverride=1
    angularSpringConstant=200
    bAngularSpring=true
    bOverrideAngularDamping=true
    bUseGravityOverride=true
    simulationSpace=RootRelative
    alphaBoolBlend {blendOption=Linear}
    boundBone {boneName=dyn_Earring_r_02_jnt}
    componentAppliedLinearAccClamp=<6000 6000 6000>
    componentLinearAccScale=<0.3 0.3 0.3>
    componentLinearVelScale=<0.3 0.3 0.3>
    physicsBodyDefinitions [{
      boundBone {boneName=dyn_Earring_r_02_jnt}
      boxExtents=<10 10 10>
      constraintSetup {angularLimitsMax=<1 30 30>, angularLimitsMin=<-1 -30 -30>, angularTarget=<1 0 0>}
      localJointOffset=<3 0 0>
     }]
   }
   n42 = Anim Dynamics - Bone: dyn_Earring_r_03_jnt!(ComponentPose=@n41, Alpha=1, GravityOverride:Vector="0,0,0", ExternalForce="0,0,0") -> (Pose) {
    angularDampingOverride=1
    angularSpringConstant=200
    bAngularSpring=true
    bOverrideAngularDamping=true
    bUseGravityOverride=true
    simulationSpace=RootRelative
    alphaBoolBlend {blendOption=Linear}
    boundBone {boneName=dyn_Earring_r_03_jnt}
    componentAppliedLinearAccClamp=<6000 6000 6000>
    componentLinearAccScale=<0.3 0.3 0.3>
    componentLinearVelScale=<0.3 0.3 0.3>
    physicsBodyDefinitions [{
      boundBone {boneName=dyn_Earring_r_03_jnt}
      boxExtents=<10 10 10>
      constraintSetup {angularLimitsMax=<1 30 30>, angularLimitsMin=<-1 -30 -30>, angularTarget=<1 0 0>}
      localJointOffset=<3 0 0>
     }]
   }
   n43 = Anim Dynamics - Bone: dyn_HairStrand_l_01_jnt!(ComponentPose=@n42, Alpha=0.8, GravityOverride:Vector="0,0,0", ExternalForce="0,0,0") -> (Pose) {
    angularDampingOverride=1
    angularSpringConstant=500
    bAngularSpring=true
    bOverrideAngularDamping=true
    bUseGravityOverride=true
    simulationSpace=RootRelative
    alphaBoolBlend {blendOption=Linear}
    boundBone {boneName=dyn_HairStrand_l_01_jnt}
    componentAppliedLinearAccClamp=<6000 6000 6000>
    componentLinearAccScale=<0.3 0.3 0.3>
    componentLinearVelScale=<0.3 0.3 0.3>
    physicsBodyDefinitions [{
      boundBone {boneName=dyn_HairStrand_l_01_jnt}
      boxExtents=<20 20 20>
      constraintSetup {angularLimitsMax=<1 0 15>, angularLimitsMin=<-1 -30 -10>, angularTarget=<1 0 0>}
      localJointOffset=<3 0 0>
     }]
   }
   n44 = Anim Dynamics - Bone: dyn_HairStrand_l_02_jnt!(ComponentPose=@n43, Alpha=0.8, GravityOverride:Vector="0,0,0", ExternalForce="0,0,0") -> (Pose) {
    angularDampingOverride=1
    angularSpringConstant=500
    bAngularSpring=true
    bOverrideAngularDamping=true
    bUseGravityOverride=true
    simulationSpace=RootRelative
    alphaBoolBlend {blendOption=Linear}
    boundBone {boneName=dyn_HairStrand_l_02_jnt}
    componentAppliedLinearAccClamp=<6000 6000 6000>
    componentLinearAccScale=<0.3 0.3 0.3>
    componentLinearVelScale=<0.3 0.3 0.3>
    physicsBodyDefinitions [{
      boundBone {boneName=dyn_HairStrand_l_02_jnt}
      boxExtents=<20 20 20>
      constraintSetup {angularLimitsMax=<1 0 15>, angularLimitsMin=<-1 -30 -10>, angularTarget=<1 0 0>}
      localJointOffset=<3 0 0>
     }]
   }
   n45 = Anim Dynamics - Bone: dyn_HairStrand_l_03_jnt!(ComponentPose=@n44, Alpha=0.8, GravityOverride:Vector="0,0,0", ExternalForce="0,0,0") -> (Pose) {
    angularDampingOverride=1
    angularSpringConstant=500
    bAngularSpring=true
    bOverrideAngularDamping=true
    bUseGravityOverride=true
    simulationSpace=RootRelative
    alphaBoolBlend {blendOption=Linear}
    boundBone {boneName=dyn_HairStrand_l_03_jnt}
    componentAppliedLinearAccClamp=<6000 6000 6000>
    componentLinearAccScale=<0.3 0.3 0.3>
    componentLinearVelScale=<0.3 0.3 0.3>
    physicsBodyDefinitions [{
      boundBone {boneName=dyn_HairStrand_l_03_jnt}
      boxExtents=<20 20 20>
      constraintSetup {angularLimitsMax=<1 0 15>, angularLimitsMin=<-1 -30 -10>, angularTarget=<1 0 0>}
      localJointOffset=<3 0 0>
     }]
   }
   n46 = Anim Dynamics - Bone: dyn_HairStrand_c_01_jnt!(ComponentPose=@n45, Alpha=1, GravityOverride:Vector="0,0,0", ExternalForce="0,0,0") -> (Pose) {
    angularDampingOverride=1
    angularSpringConstant=600
    bAngularSpring=true
    bOverrideAngularDamping=true
    bUseGravityOverride=true
    simulationSpace=RootRelative
    alphaBoolBlend {blendOption=Linear}
    boundBone {boneName=dyn_HairStrand_c_01_jnt}
    componentAppliedLinearAccClamp=<6000 6000 6000>
    componentLinearAccScale=<0.3 0.3 0.3>
    componentLinearVelScale=<0.3 0.3 0.3>
    physicsBodyDefinitions [{
      boundBone {boneName=dyn_HairStrand_c_01_jnt}
      boxExtents=<10 10 10>
      constraintSetup {angularLimitsMax=<1 0 10>, angularLimitsMin=<-1 -30 -10>, angularTarget=<1 0 0>}
      localJointOffset=<3 0 0>
     }]
   }
   n47 = Anim Dynamics - Bone: dyn_HairStrand_c_02_jnt!(ComponentPose=@n46, Alpha=1, GravityOverride:Vector="0,0,0", ExternalForce="0,0,0") -> (Pose) {
    angularDampingOverride=1
    angularSpringConstant=600
    bAngularSpring=true
    bOverrideAngularDamping=true
    bUseGravityOverride=true
    simulationSpace=RootRelative
    alphaBoolBlend {blendOption=Linear}
    boundBone {boneName=dyn_HairStrand_c_02_jnt}
    componentAppliedLinearAccClamp=<6000 6000 6000>
    componentLinearAccScale=<0.3 0.3 0.3>
    componentLinearVelScale=<0.3 0.3 0.3>
    physicsBodyDefinitions [{
      boundBone {boneName=dyn_HairStrand_c_02_jnt}
      boxExtents=<10 10 10>
      constraintSetup {angularLimitsMax=<1 0 5>, angularLimitsMin=<-1 -30 -5>, angularTarget=<1 0 0>}
      localJointOffset=<3 0 0>
     }]
   }
   n48 = Anim Dynamics - Bone: dyn_HairStrand_c_03_jnt!(ComponentPose=@n47, Alpha=1, GravityOverride:Vector="0,0,0", ExternalForce="0,0,0") -> (Pose) {
    angularDampingOverride=1
    angularSpringConstant=600
    bAngularSpring=true
    bOverrideAngularDamping=true
    bUseGravityOverride=true
    simulationSpace=RootRelative
    alphaBoolBlend {blendOption=Linear}
    boundBone {boneName=dyn_HairStrand_c_03_jnt}
    componentAppliedLinearAccClamp=<6000 6000 6000>
    componentLinearAccScale=<0.3 0.3 0.3>
    componentLinearVelScale=<0.3 0.3 0.3>
    physicsBodyDefinitions [{
      boundBone {boneName=dyn_HairStrand_c_03_jnt}
      boxExtents=<10 10 10>
      constraintSetup {angularLimitsMax=<1 0 5>, angularLimitsMin=<-1 -30 -5>, angularTarget=<1 0 0>}
      localJointOffset=<3 0 0>
     }]
   }
   n49 = Anim Dynamics - Bone: dyn_Bangs_c_01_jnt!(ComponentPose=@n48, Alpha=1, GravityOverride:Vector="0,0,0", ExternalForce="0,0,0") -> (Pose) {
    angularDampingOverride=1
    angularSpringConstant=200
    bAngularSpring=true
    bOverrideAngularDamping=true
    bUseGravityOverride=true
    simulationSpace=RootRelative
    alphaBoolBlend {blendOption=Linear}
    boundBone {boneName=dyn_Bangs_c_01_jnt}
    componentAppliedLinearAccClamp=<6000 6000 6000>
    componentLinearAccScale=<0.3 0.3 0.3>
    componentLinearVelScale=<0.3 0.3 0.3>
    physicsBodyDefinitions [{
      boundBone {boneName=dyn_Bangs_c_01_jnt}
      boxExtents=<10 10 10>
      constraintSetup {angularLimitsMax=<1 0 15>, angularLimitsMin=<-1 -30 -15>, angularTarget=<1 0 0>}
      localJointOffset=<3 0 0>
     }]
   }
   n50 = Anim Dynamics - Bone: dyn_Bangs_c_02_jnt!(ComponentPose=@n49, Alpha=1, GravityOverride:Vector="0,0,0", ExternalForce="0,0,0") -> (Pose) {
    angularDampingOverride=1
    angularSpringConstant=200
    bAngularSpring=true
    bOverrideAngularDamping=true
    bUseGravityOverride=true
    simulationSpace=RootRelative
    alphaBoolBlend {blendOption=Linear}
    boundBone {boneName=dyn_Bangs_c_02_jnt}
    componentAppliedLinearAccClamp=<6000 6000 6000>
    componentLinearAccScale=<0.3 0.3 0.3>
    componentLinearVelScale=<0.3 0.3 0.3>
    physicsBodyDefinitions [{
      boundBone {boneName=dyn_Bangs_c_02_jnt}
      boxExtents=<10 10 10>
      constraintSetup {angularLimitsMax=<1 0 15>, angularLimitsMin=<-1 -30 -15>, angularTarget=<1 0 0>}
      localJointOffset=<3 0 0>
     }]
   }
   n51 = Anim Dynamics - Bone: dyn_Bangs_c_03_jnt!(ComponentPose=@n50, Alpha=1, GravityOverride:Vector="0,0,0", ExternalForce="0,0,0") -> (Pose) {
    angularDampingOverride=1
    angularSpringConstant=200
    bAngularSpring=true
    bOverrideAngularDamping=true
    bUseGravityOverride=true
    simulationSpace=RootRelative
    alphaBoolBlend {blendOption=Linear}
    boundBone {boneName=dyn_Bangs_c_03_jnt}
    componentAppliedLinearAccClamp=<6000 6000 6000>
    componentLinearAccScale=<0.3 0.3 0.3>
    componentLinearVelScale=<0.3 0.3 0.3>
    physicsBodyDefinitions [{
      boundBone {boneName=dyn_Bangs_c_03_jnt}
      boxExtents=<10 10 10>
      constraintSetup {angularLimitsMax=<1 0 15>, angularLimitsMin=<-1 -30 -15>, angularTarget=<1 0 0>}
      localJointOffset=<3 0 0>
     }]
   }
   n52 = Anim Dynamics - Bone: dyn_Hairstrand01_r_01_jnt!(ComponentPose=@n51, Alpha=1, GravityOverride:Vector="0,0,0", ExternalForce="0,0,0") -> (Pose) {
    angularDampingOverride=1
    angularSpringConstant=400
    bAngularSpring=true
    bOverrideAngularDamping=true
    bUseGravityOverride=true
    simulationSpace=RootRelative
    alphaBoolBlend {blendOption=Linear}
    boundBone {boneName=dyn_Hairstrand01_r_01_jnt}
    componentAppliedLinearAccClamp=<6000 6000 6000>
    componentLinearAccScale=<0.3 0.3 0.3>
    componentLinearVelScale=<0.3 0.3 0.3>
    physicsBodyDefinitions [{
      boundBone {boneName=dyn_Hairstrand01_r_01_jnt}
      boxExtents=<15 15 15>
      constraintSetup {angularLimitsMax=<1 0 15>, angularLimitsMin=<-1 -20 -10>, angularTarget=<1 0 0>}
      localJointOffset=<3 0 0>
     }]
   }
   n53 = Anim Dynamics - Bone: dyn_Hairstrand01_r_02_jnt!(ComponentPose=@n52, Alpha=1, GravityOverride:Vector="0,0,0", ExternalForce="0,0,0") -> (Pose) {
    angularDampingOverride=1
    angularSpringConstant=400
    bAngularSpring=true
    bOverrideAngularDamping=true
    bUseGravityOverride=true
    simulationSpace=RootRelative
    alphaBoolBlend {blendOption=Linear}
    boundBone {boneName=dyn_Hairstrand01_r_02_jnt}
    componentAppliedLinearAccClamp=<6000 6000 6000>
    componentLinearAccScale=<0.3 0.3 0.3>
    componentLinearVelScale=<0.3 0.3 0.3>
    physicsBodyDefinitions [{
      boundBone {boneName=dyn_Hairstrand01_r_02_jnt}
      boxExtents=<15 15 15>
      constraintSetup {angularLimitsMax=<1 0 15>, angularLimitsMin=<-1 -20 -10>, angularTarget=<1 0 0>}
      localJointOffset=<3 0 0>
     }]
   }
   n54 = Anim Dynamics - Bone: dyn_Hairstrand01_r_03_jnt!(ComponentPose=@n53, Alpha=1, GravityOverride:Vector="0,0,0", ExternalForce="0,0,0") -> (Pose) {
    angularDampingOverride=1
    angularSpringConstant=400
    bAngularSpring=true
    bOverrideAngularDamping=true
    bUseGravityOverride=true
    simulationSpace=RootRelative
    alphaBoolBlend {blendOption=Linear}
    boundBone {boneName=dyn_Hairstrand01_r_03_jnt}
    componentAppliedLinearAccClamp=<6000 6000 6000>
    componentLinearAccScale=<0.3 0.3 0.3>
    componentLinearVelScale=<0.3 0.3 0.3>
    physicsBodyDefinitions [{
      boundBone {boneName=dyn_Hairstrand01_r_03_jnt}
      boxExtents=<15 15 15>
      constraintSetup {angularLimitsMax=<1 0 15>, angularLimitsMin=<-1 -20 -10>, angularTarget=<1 0 0>}
      localJointOffset=<3 0 0>
     }]
   }
   n55 = Anim Dynamics - Bone: dyn_Hairstrand02_r_01_jnt!(ComponentPose=@n54, Alpha=1, GravityOverride:Vector="0,0,0", ExternalForce="0,0,0") -> (Pose) {
    angularDampingOverride=1
    angularSpringConstant=400
    bAngularSpring=true
    bOverrideAngularDamping=true
    bUseGravityOverride=true
    simulationSpace=RootRelative
    alphaBoolBlend {blendOption=Linear}
    boundBone {boneName=dyn_Hairstrand02_r_01_jnt}
    componentAppliedLinearAccClamp=<6000 6000 6000>
    componentLinearAccScale=<0.3 0.3 0.3>
    componentLinearVelScale=<0.3 0.3 0.3>
    physicsBodyDefinitions [{
      boundBone {boneName=dyn_Hairstrand02_r_01_jnt}
      boxExtents=<15 15 15>
      constraintSetup {angularLimitsMax=<1 0 15>, angularLimitsMin=<-1 -20 -10>, angularTarget=<1 0 0>}
      localJointOffset=<3 0 0>
     }]
   }
   n56 = Anim Dynamics - Bone: dyn_Hairstrand02_r_02_jnt!(ComponentPose=@n55, Alpha=1, GravityOverride:Vector="0,0,0", ExternalForce="0,0,0") -> (Pose) {
    angularDampingOverride=1
    angularSpringConstant=400
    bAngularSpring=true
    bOverrideAngularDamping=true
    bUseGravityOverride=true
    simulationSpace=RootRelative
    alphaBoolBlend {blendOption=Linear}
    boundBone {boneName=dyn_Hairstrand02_r_02_jnt}
    componentAppliedLinearAccClamp=<6000 6000 6000>
    componentLinearAccScale=<0.3 0.3 0.3>
    componentLinearVelScale=<0.3 0.3 0.3>
    physicsBodyDefinitions [{
      boundBone {boneName=dyn_Hairstrand02_r_02_jnt}
      boxExtents=<15 15 15>
      constraintSetup {angularLimitsMax=<1 0 15>, angularLimitsMin=<-1 -20 -10>, angularTarget=<1 0 0>}
      localJointOffset=<3 0 0>
     }]
   }
   n57 = Anim Dynamics - Bone: dyn_Hairstrand02_r_03_jnt!(ComponentPose=@n56, Alpha=1, GravityOverride:Vector="0,0,0", ExternalForce="0,0,0") -> (Pose) {
    angularDampingOverride=1
    angularSpringConstant=400
    bAngularSpring=true
    bOverrideAngularDamping=true
    bUseGravityOverride=true
    simulationSpace=RootRelative
    alphaBoolBlend {blendOption=Linear}
    boundBone {boneName=dyn_Hairstrand02_r_03_jnt}
    componentAppliedLinearAccClamp=<6000 6000 6000>
    componentLinearAccScale=<0.3 0.3 0.3>
    componentLinearVelScale=<0.3 0.3 0.3>
    physicsBodyDefinitions [{
      boundBone {boneName=dyn_Hairstrand02_r_03_jnt}
      boxExtents=<15 15 15>
      constraintSetup {angularLimitsMax=<1 0 15>, angularLimitsMin=<-1 -20 -10>, angularTarget=<1 0 0>}
      localJointOffset=<3 0 0>
     }]
   }
   n58 = Anim Dynamics - Bone: dyn_Hairstrand03_r_01_jnt!(ComponentPose=@n57, Alpha=1, GravityOverride:Vector="0,0,0", ExternalForce="0,0,0") -> (Pose) {
    angularDampingOverride=1
    angularSpringConstant=200
    bAngularSpring=true
    bOverrideAngularDamping=true
    bUseGravityOverride=true
    simulationSpace=RootRelative
    alphaBoolBlend {blendOption=Linear}
    boundBone {boneName=dyn_Hairstrand03_r_01_jnt}
    componentAppliedLinearAccClamp=<6000 6000 6000>
    componentLinearAccScale=<0.3 0.3 0.3>
    componentLinearVelScale=<0.3 0.3 0.3>
    physicsBodyDefinitions [{
      boundBone {boneName=dyn_Hairstrand03_r_01_jnt}
      boxExtents=<20 20 20>
      constraintSetup {angularLimitsMax=<1 0 10>, angularLimitsMin=<-1 -20 -10>, angularTarget=<1 0 0>}
      localJointOffset=<3 0 0>
     }]
   }
   n59 = Anim Dynamics - Bone: dyn_Hairstrand03_r_02_jnt!(ComponentPose=@n58, Alpha=1, GravityOverride:Vector="0,0,0", ExternalForce="0,0,0") -> (Pose) {
    angularDampingOverride=1
    angularSpringConstant=200
    bAngularSpring=true
    bOverrideAngularDamping=true
    bUseGravityOverride=true
    simulationSpace=RootRelative
    alphaBoolBlend {blendOption=Linear}
    boundBone {boneName=dyn_Hairstrand03_r_02_jnt}
    componentAppliedLinearAccClamp=<6000 6000 6000>
    componentLinearAccScale=<0.3 0.3 0.3>
    componentLinearVelScale=<0.3 0.3 0.3>
    physicsBodyDefinitions [{
      boundBone {boneName=dyn_Hairstrand03_r_02_jnt}
      boxExtents=<20 20 20>
      constraintSetup {angularLimitsMax=<1 0 10>, angularLimitsMin=<-1 -20 -10>, angularTarget=<1 0 0>}
      localJointOffset=<3 0 0>
     }]
   }
   n60 = Anim Dynamics - Bone: dyn_Hairstrand03_r_03_jnt!(ComponentPose=@n59, Alpha=1, GravityOverride:Vector="0,0,0", ExternalForce="0,0,0") -> (Pose) {
    angularDampingOverride=1
    angularSpringConstant=200
    bAngularSpring=true
    bOverrideAngularDamping=true
    bUseGravityOverride=true
    simulationSpace=RootRelative
    alphaBoolBlend {blendOption=Linear}
    boundBone {boneName=dyn_Hairstrand03_r_03_jnt}
    componentAppliedLinearAccClamp=<6000 6000 6000>
    componentLinearAccScale=<0.3 0.3 0.3>
    componentLinearVelScale=<0.3 0.3 0.3>
    physicsBodyDefinitions [{
      boundBone {boneName=dyn_Hairstrand03_r_03_jnt}
      boxExtents=<20 20 20>
      constraintSetup {angularLimitsMax=<1 0 10>, angularLimitsMin=<-1 -20 -10>, angularTarget=<1 0 0>}
      localJointOffset=<3 0 0>
     }]
   }
   Output Pose(Result=Component To Local(ComponentPose=@n60)) {layerGroup=None}
   Get Is Pistol Targeting Alpha() -> (Is Pistol Targeting Alpha:float)
   Get Is Rifle Targeting Alpha() -> (Is Rifle Targeting Alpha:float)
  ]}
 EventGraph {nodes [
   Event BlueprintUpdateAnimation()
   > n1 = Sequence!(execute) -> (then_0, then_1, then_2:Exec, then_3:Exec, then_4:Exec, then_5:Exec, then_6:Exec, then_7:Exec)
   > n2 = Gravity Override State Function DEPRECATED(CharacterPartAnimInstance=Self-Reference(), GameTimeInSeconds=GetGameTimeInSeconds(), GravityOverrideParams=Get Gravity Override Params())
   > Set Gravity Override(Gravity Override=@n2)
   n3 = Sequence!(execute=@n1.then_1) -> (then_0, then_1, then_2, then_3, then_4)
   > n4 = Cast To FortPlayerAnimInstance(Object=GetAnimInstance(self=GetOwnerSkeletalMeshComponent()))
   > Set isPistol(isPistol=false)
   > Set isRifle(isRifle:bool=false) -> (Output_Get:bool)
   > n5 = Switch on EFortWeaponCoreAnimation!(execute, Selection:EFortWeaponCoreAnimation Enum=Get WeaponCoreAnim(self=@n4)) -> (
    Melee:Exec
    Pistol:Exec
    Shotgun:Exec
    PaperBlueprint:Exec
    Rifle:Exec
    MeleeOneHand:Exec
    MachinePistol:Exec
    RocketLauncher:Exec
    GrenadeLauncher:Exec
    GoingCommando:Exec
    AssaultRifle:Exec
    TacticalShotgun:Exec
    SniperRifle:Exec
    TrapPlacement:Exec
    ShoulderLauncher:Exec
    AbilityDecoTool:Exec
    Crossbow:Exec
    C4:Exec
    RemoteControl:Exec
    DualWield:Exec
    AR_BullPup:Exec
    AR_ForwardGrip:Exec
    MedPackPaddles:Exec
    SMG_P90:Exec
    AR_DrumGun:Exec
    Consumable_Small:Exec
    Consumable_Large:Exec
    Balloon:Exec
    MountedTurret:Exec
    CreativeTool:Exec
    ExplosiveBow:Exec
    AshtonIndigo:Exec
    AshtonChicago:Exec
    MeleeDualWield:Exec
    Unarmed:Exec
   )
   Set isRifle(execute=[@n5.AssaultRifle, @n5.Rifle, @n5.RocketLauncher, @n5.Shotgun, @n5.SniperRifle, @n5.TacticalShotgun], isRifle=true)
   Set isPistol(execute=@n5.Pistol, isPistol=true)
   Set Is Rifle Alpha(execute=@n3.then_1, Is Rifle Alpha:float="To Float (Boolean)"(InBool=Get isRifle() -> (isRifle:bool))) -> (Output_Get:float)
   n6 = Get bIsTargeting()
   n7 = Get isRifle()
   Set Is Rifle Targeting Alpha(execute=@n3.then_2, Is Rifle Targeting Alpha:float="To Float (Boolean)"(InBool=AND Boolean(A=@n7, B=@n6))) -> (Output_Get:float)
   Set Is Pistol Alpha(execute=@n3.then_3, Is Pistol Alpha:float="To Float (Boolean)"(InBool=Get isPistol() -> (isPistol:bool))) -> (Output_Get:float)
   n8 = Get isPistol()
   n9 = Get bIsTargeting()
   Set Is Pistol Targeting Alpha(execute=@n3.then_4, Is Pistol Targeting Alpha:float="To Float (Boolean)"(InBool=AND Boolean(A=@n8, B=@n9))) -> (Output_Get:float)
   Set Is Skydiving Alpha(execute=@n1.then_3, Is Skydiving Alpha:double="To Float (Boolean)"(InBool=Get bIsSkyDiving())) -> (Output_Get:double)
   Set Hair Alpha(
    execute=@n1.then_4
    Hair Alpha:double=float * float(A=1, B=Get Skydive Hair Alpha() -> (Skydive Hair Alpha:double)) {is=title K2Node_PromotableOperator}
   ) -> (Output_Get:double)
   n10 = Branch(execute=@n1.then_5, Condition=Get bIsSkyDiving())
   > Set Skydive Hair Alpha(Skydive Hair Alpha:double=0.75) -> (Output_Get:double)
   Set Skydive Hair Alpha(execute=@n10.else, Skydive Hair Alpha=1)
   Set Is Crouching Alpha(execute=@n1.then_6, Is Crouching Alpha:double="To Float (Boolean)"(InBool=Get bIsCrouching())) -> (Output_Get:double)
   Event OnCharacterCustomizationLoaded()
   > n11 = Sequence!(execute) -> (then_0, then_1, then_2)
   > Cast To PlayerPawn_Athena(Object=TryGetPawnOwner())
   n12 = CheckForBackbling()
   Set Is Wearing Backbling(execute=@n11.then_1, Is Wearing Backbling=OR Boolean(A=@n12, B=@n12.bIsBackblingLong))
  ]}
]
variables [
 Gravity Override {type=Vector}
 Gravity Override Params {type=GravityOverrideParamsStruct}
 Hair Alpha {type=number}
 Is Crouching Alpha {type=number}
 Is Pistol Alpha {type=number}
 Is Pistol Targeting Alpha {type=number}
 Is Rifle Alpha {type=number}
 Is Rifle Targeting Alpha {type=number}
 Is Skydiving Alpha {type=number}
 Is Wearing Backbling {type=boolean}
 isPistol {type=boolean}
 isRifle {type=boolean}
 Skydive Hair Alpha {type=number}
]
